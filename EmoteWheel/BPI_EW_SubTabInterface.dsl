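; Blueprint interface for sub-tab navigation in the Emote Wheel UI, declaring four functions without implementations.
; Next and previous sub-tab calls each take an interaction widget reference.
; The other functions pass a category top bar reference and trigger a tab-change animation by integer index.
; References use generic Widget and Object types rather than specific widget classes.
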(fn GoToNextSubTab
  (params (InteractionWidget :type Widget%20Object%20Reference)))

(fn GoToPreviousSubTab
  (params (InteractionWidget :type Widget%20Object%20Reference)))

(fn SetCategoryTopBarRef
  (params (TopBarRef :type Object%20Reference)))

(fn TriggerChangeTabsAnimation
  (params (Index :type Integer)))
