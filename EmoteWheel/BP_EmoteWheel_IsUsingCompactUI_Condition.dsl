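; BP_EmoteWheel_IsUsingCompactUI_Condition is a Sequencer condition for the emote wheel that checks whether the compact UI layout is in use.
; It overrides OnEvaluateCondition and returns the result of DisplayHelper's IsUsingCompactUI directly, ignoring the passed condition context.
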(fn OnEvaluateCondition
  (params (ConditionContext :type Movie%20Scene%20Condition%20Context%20Structure%20%28by%20ref%29))
  (return (DisplayHelper|IsUsingCompactUI)))
