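# Translucent UI material for the emote screen background, with a radial opacity mask controlled at the middle by Center Opacity.
# At Epic quality it shows a slightly darkened blurred backdrop sampled from the Slate post buffer.
# Aspect-scaled UVs feed circle and linear gradient functions to build the mask.
# Quality and shading path switches give other tiers and mobile the flat LowPerf Color at half that opacity.
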Material /EmoteWheel/Materials/M_UI_EmoteScreen_Background
graph {
 properties {
  bAutomaticallySetUsageInEditor=false
  bUsedWithStaticMesh=false
  blendMode=BLEND_Translucent
  materialDomain=MD_UI
  usedShadingModels=Default Lit
 }
 nodes [
  n1 = ScalarParameter() {desc=Black, parameterName=LowPerf Color}
  n2 = NamedRerouteDeclaration(None=MaterialFunctionCall!("UVs (Optional)") {materialFunction=/Engine/Functions/UserInterface/GetSlatePost0}.LinearRGB) {name=BlurredBackground, nodeColor=<rgba 1 0.1172 0 1>}
  ShadingPathSwitch(Default=QualitySwitch(Default=@n1, Low=@n1, High=@n1, Medium=@n1, Epic=LinearInterpolate(A=@n2) {constAlpha=0.2, constB=0}), Mobile=@n1)
  n3 = MaterialFunctionCall(UV=TextureCoordinate()) {materialFunction=&MF_UI_UVScaling}
  n4 = Constant() {r=0.01}
  n5 = Constant() {r=0.5}
  n6 = LinearInterpolate(
   A=Constant() {r=1}
   B=ScalarParameter() {parameterName=Center Opacity}
   Alpha=MaterialFunctionCall(
    UVs=MaterialFunctionCall(UVs=@n3.Scaled to Fit, Size=@n4, GlowMin=@n4, GlowMax=@n5) {materialFunction=&MF_UI_Circle}.Fill
    GradientStart=@n4
    GradientEnd=@n5
   ) {materialFunction=&MF_UI_LinearGradient}.Result
  ) {constB=0}
  n7 = Multiply(A=@n6) {constB=0.5}
  ShadingPathSwitch(Default=QualitySwitch(Default=@n7, Low=@n7, High=@n7, Medium=@n7, Epic=@n6), Mobile=@n7)
 ]
}
