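# Comms overlay widget for the emote wheel that shows a small reticle dot centered in a grid panel.
# Behind the dot sits a circle progress-bar material image, tinted fully transparent.
# Construct plays the short Intro opacity fade on both images forward and Destruct plays it in reverse, stopping it first if already playing.
# A CloseWheel member is declared with an empty graph and is not called here.
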WidgetBlueprint /EmoteWheel/WBP_EW_CommsOverlay
animations [Intro {
  binding_count=2
  length_seconds=0.15
  object=&self:Intro
  sequence {bindings [
    Image_CenterReticle {
     id=#1
     possessed_class=Image
     tracks [{
       type=MovieSceneFloatTrack
       display_name=Render Opacity
       sections [{type=MovieSceneFloatSection, end_frame=3, start_frame=0, channels [None {type=MovieSceneScriptingFloatChannel, keys [[0,0],[3,1]]}]}]
      }]
    }
    Reticle_Backer {
     id=#2
     possessed_class=Image
     tracks [{
       type=MovieSceneFloatTrack
       display_name=Render Opacity
       sections [{type=MovieSceneFloatSection, end_frame=3, start_frame=0, channels [None {type=MovieSceneScriptingFloatChannel, keys [[0,0],[3,1]]}]}]
      }]
    }
   ]}
 }]
graphs [
 CloseWheel {nodes [CloseWheel() {is=title K2Node_FunctionEntry}]}
 EventGraph {nodes [
   Event Construct()
   n1 = Get Intro()
   > n2 = Branch(Condition=IsAnimationPlaying(InAnimation=@n1))
   > n3 = StopAnimation(InAnimation=@n1)
   PlayAnimationForward(execute=[@n3.then, @n2.else], InAnimation=@n1)
   Event Destruct()
   n4 = Get Intro()
   > n5 = Branch(Condition=IsAnimationPlaying(InAnimation=@n4))
   > n6 = StopAnimation(InAnimation=@n4)
   PlayAnimationReverse(execute=[@n6.then, @n5.else], InAnimation=@n4)
  ]}
]
variables [CloseWheel {}]
widgets {
 info {inheritedWidgetCount=0, parentClass=&UserWidget, rootWidgetClass=&GridPanel, widgetCount=3}
 widgets [
  GridPanel_Main {widgetClassPath=&GridPanel, widget=GridPanel_Main, properties {bOverride_Cursor=true, columnFill [1], rowFill [1]}}
  Reticle_Backer {
   widgetClassPath=&Image
   bIsVariable=true
   parent=GridPanel_Main
   widget=Reticle_Backer
   properties {
    visibility=SelfHitTestInvisible
    brush {
     bIsDynamicallyLoaded=false
     drawAs=Image
     imageType=FullColor
     mirroring=NoMirror
     resourceName=None
     resourceObject=/FortUILibrary/Foundation/Art_ToolKit/Tools/MI_UI_Circle_ProgressBar
     tiling=NoTile
     imageSize=<112 112>
     margin=<ltrb 0 0 0 0>
     outlineSettings {
      bUseBrushTransparency=false
      roundingType=HalfHeightRadius
      width=0
      color {colorUseRule=UseColor_Specified, specifiedColor=<rgba 0 0 0 0>}
      cornerRadii=<0 0 0 0>
     }
     tintColor {colorUseRule=UseColor_Specified, specifiedColor=<rgba 1 1 1 1>}
     uVRegion {bIsValid=false, max=<0 0>, min=<0 0>}
    }
    colorAndOpacity=<rgba 0 0 0 0>
   }
   slot {class=GridSlot, properties {horizontalAlignment=HAlign_Center, verticalAlignment=VAlign_Center}}
  }
  Image_CenterReticle {
   widgetClassPath=&Image
   bIsVariable=true
   parent=GridPanel_Main
   widget=Image_CenterReticle
   properties {brush {
     bIsDynamicallyLoaded=false
     drawAs=Image
     imageType=NoImage
     mirroring=NoMirror
     resourceName=None
     resourceObject=/Game/UI/Friends_UI/Social/circle
     tiling=NoTile
     imageSize=<5 5>
     margin=<ltrb 0 0 0 0>
     outlineSettings {
      bUseBrushTransparency=false
      roundingType=HalfHeightRadius
      width=0
      color {colorUseRule=UseColor_Specified, specifiedColor=<rgba 0 0 0 0>}
      cornerRadii=<0 0 0 0>
     }
     tintColor {colorUseRule=UseColor_Specified, specifiedColor=<rgba 1 1 1 1>}
     uVRegion {bIsValid=false, max=<0 0>, min=<0 0>}
    }}
   slot {class=GridSlot, properties {horizontalAlignment=HAlign_Center, verticalAlignment=VAlign_Center}}
  }
 ]
}
