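# Widget Blueprint overlay for the emote wheel that shows a follower's icon and text, with a dismiss action.
# SetFollowerIcon feeds the portrait into the lazy image's dynamic material; SetFollowersCategoryVM stores the cast category view model.
# A collapsed decoy button bound to DismissFollower with hold data calls DismissAlly on that view model, broadcasts CloseWheel and fires a Flow analytics event.
# A short opacity fade plays forward on Construct, reversed on Destruct.
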WidgetBlueprint /EmoteWheel/WBP_EW_FollowersOverlay
animations [Intro {
  binding_count=1
  length_seconds=0.15
  object=&self:Intro
  sequence {bindings [Overlay_FollowerInfo {
     id=#1
     possessed_class=Overlay
     tracks [{
       type=MovieSceneFloatTrack
       display_name=Render Opacity
       sections [{type=MovieSceneFloatSection, end_frame=3, start_frame=0, channels [None {type=MovieSceneScriptingFloatChannel, keys [[0,0],[3,1]]}]}]
      }]
    }]}
 }]
graphs [
 CloseWheel {nodes [CloseWheel() {is=title K2Node_FunctionEntry}]}
 EventGraph {nodes [
   n1 = "On Clicked (Dismiss_BottomBarDecoyButton)"!() -> (OutputDelegate, then, Button:Common Button Base Object Reference) {is=title K2Node_ComponentBoundEvent}
   > n2 = Get!(execute:Exec) -> (then:Exec, else:Exec, FollowersEmoteWheelCategoryVM:Followers Emote Wheel Category VM Object Reference)
   > DismissAlly(self:Followers Emote Wheel Category VM Object Reference=@n2)
   > Call Close Wheel!(execute:Exec, self:WBP EW Followers Overlay Object Reference) -> (then:Exec) {is=title K2Node_CallDelegate}
   > TriggerFlowriderInteraction!(execute, self, Element:Widget Object Reference=@n1.Button, Value:String=Dismiss Follower) -> (then)
   Event Construct()
   n3 = Get Intro()
   > n4 = Branch(Condition=IsAnimationPlaying(InAnimation=@n3))
   > n5 = StopAnimation(InAnimation=@n3)
   PlayAnimationForward(execute=[@n5.then, @n4.else], InAnimation=@n3)
   Event Destruct()
   n6 = Get Intro()
   > n7 = Branch(Condition=IsAnimationPlaying(InAnimation=@n6))
   > n8 = StopAnimation(InAnimation=@n6)
   PlayAnimationReverse(execute=[@n8.then, @n7.else], InAnimation=@n6)
  ]}
 SetFollowerIcon {nodes [
   n1 = SetFollowerIcon!() -> (then, Texture:Texture 2D Soft Object Reference) {is=title K2Node_FunctionEntry}
   > n2 = GetDynamicMaterial(self=Get FollowerIcon_Lazy() -> (FollowerIcon_Lazy:Fort Lazy Image Object Reference))
   > SetTextureParameterValue(
    self=@n2
    ParameterName=NPC_Icon
    Value=Resolve Soft Reference!(Input:Texture 2D Soft Object Reference=@n1) -> (Output:Texture 2D Object Reference)
   )
  ]}
 SetFollowersCategoryVM {nodes [
   n1 = SetFollowersCategoryVM!() -> (then, VM:Emote Wheel Category VM Object Reference) {is=title K2Node_FunctionEntry}
   > n2 = Cast To FollowersEmoteWheelCategoryVM!(execute, Object=@n1) -> (then, CastFailed, AsFollowers Emote Wheel Category VM:Followers Emote Wheel Category VM Object Reference) {is=title K2Node_DynamicCast}
   > Set FollowersEmoteWheelCategoryVM(FollowersEmoteWheelCategoryVM:Followers Emote Wheel Category VM Object Reference=@n2) -> (Output_Get:Followers Emote Wheel Category VM Object Reference)
  ]}
 TriggerFlowriderInteraction {nodes [
   n1 = TriggerFlowriderInteraction!() -> (then, Element:Widget Object Reference, Value:String) {is=title K2Node_FunctionEntry}
   > Fire Event Element Interacted(
    Element=@n1
    AdditionalInfo="Add String (Non-Localized)"!(
     Info:"Flow Analytics Additional Info Structure (by ref)"=Make Flow Analytics Additional Info!() -> (FlowAnalyticsAdditionalInfo:Flow Analytics Additional Info Structure) {is=title K2Node_MakeStruct}
     Key:String=Button Name
     Value:String=@n1.Value
    ) -> (ReturnValue:Flow Analytics Additional Info Structure)
   )
  ]}
]
variables [CloseWheel {}]
widgets {
 info {inheritedWidgetCount=0, parentClass=&UserWidget, rootWidgetClass=&GridPanel, widgetCount=8}
 widgets [
  GridPanel_Main {widgetClassPath=&GridPanel, widget=GridPanel_Main, properties {bOverride_Cursor=true, columnFill [1], rowFill [1]}}
  Overlay_FollowerInfo {
   widgetClassPath=&Overlay
   bIsVariable=true
   parent=GridPanel_Main
   widget=Overlay_FollowerInfo
   slot {class=GridSlot, properties {horizontalAlignment=HAlign_Center, verticalAlignment=VAlign_Center}}
  }
  Follower_Backer {
   widgetClassPath=&Image
   bIsVariable=true
   parent=Overlay_FollowerInfo
   widget=Follower_Backer
   properties {
    visibility=SelfHitTestInvisible
    brush {
     bIsDynamicallyLoaded=false
     drawAs=Image
     imageType=NoImage
     mirroring=NoMirror
     resourceName=None
     resourceObject=/FortUILibrary/Foundation/Art_ToolKit/Tools/M_UI_Circle_White
     tiling=NoTile
     imageSize=<110 110>
     margin=<ltrb 0 0 0 0>
     outlineSettings {
      bUseBrushTransparency=false
      roundingType=HalfHeightRadius
      width=0
      color {colorUseRule=UseColor_Specified, specifiedColor=<rgba 0 0 0 0>}
      cornerRadii=<0 0 0 0>
     }
     tintColor {colorUseRule=UseColor_Specified, specifiedColor=<rgba 1 1 1 1>}
     uVRegion {bIsValid=false, max=<0 0>, min=<0 0>}
    }
    colorAndOpacity=<rgba 0 0 0 0.6>
   }
   slot {class=OverlaySlot, properties {horizontalAlignment=HAlign_Center, verticalAlignment=VAlign_Center}}
  }
  FollowerIcon_Lazy {
   widgetClassPath=&FortLazyImage
   bIsVariable=true
   parent=Overlay_FollowerInfo
   widget=FollowerIcon_Lazy
   properties {brush {
     bIsDynamicallyLoaded=false
     drawAs=Image
     imageType=FullColor
     mirroring=NoMirror
     resourceName=None
     resourceObject=/FortniteCoreUI/Athena/HUD/Art/squadArt/MI_UI_NPCCom_NPCIcon_Inst_Faded
     tiling=NoTile
     imageSize=<97 97>
     margin=<ltrb 0 0 0 0>
     outlineSettings {
      bUseBrushTransparency=false
      roundingType=HalfHeightRadius
      width=0
      color {colorUseRule=UseColor_Specified, specifiedColor=<rgba 0 0 0 0>}
      cornerRadii=<0 0 0 0>
     }
     tintColor {colorUseRule=UseColor_Specified, specifiedColor=<rgba 1 1 1 1>}
     uVRegion {bIsValid=false, max=<0 0>, min=<0 0>}
    }}
   slot {class=OverlaySlot, properties {horizontalAlignment=HAlign_Center, verticalAlignment=VAlign_Center}}
  }
  SizeBox_FollowerText {
   widgetClassPath=&SizeBox
   parent=Overlay_FollowerInfo
   widget=SizeBox_FollowerText
   properties {bOverride_MaxDesiredWidth=true, maxDesiredWidth=280}
   slot {class=OverlaySlot, properties {horizontalAlignment=HAlign_Center, verticalAlignment=VAlign_Center, padding=<ltrb 0 50 0 0>}}
  }
  Follower_Text {
   widgetClassPath=&CommonTextBlock
   bIsVariable=true
   parent=SizeBox_FollowerText
   widget=Follower_Text
   properties {
    scrollStyle=&TextScrollStyle-Base_C
    text=Text Block
    font {
     bForceMonospaced=false
     bMaterialIsStencil=false
     fontMaterial=null
     fontObject=&HeadingNow
     letterSpacing=0
     monospacedWidth=1
     size=22.5
     skewAmount=0
     typefaceFontName=86 Bold Italic
     outlineSettings {
      bApplyOutlineToDropShadows=false
      bMiteredCorners=true
      bSeparateFillAlpha=false
      outlineMaterial=null
      outlineSize=2
      outlineColor=<rgba 0 0 0 0.4>
     }
    }
    shadowOffset=<0 0>
    strikeBrush {
     bIsDynamicallyLoaded=false
     drawAs=NoDrawType
     imageType=NoImage
     mirroring=NoMirror
     resourceName=None
     resourceObject=null
     tiling=NoTile
     imageSize=<32 32>
     margin=<ltrb 0 0 0 0>
     outlineSettings {
      bUseBrushTransparency=false
      roundingType=HalfHeightRadius
      width=0
      color {colorUseRule=UseColor_Specified, specifiedColor=<rgba 0 0 0 0>}
      cornerRadii=<0 0 0 0>
     }
     tintColor {colorUseRule=UseColor_Specified, specifiedColor=<rgba 1 1 1 1>}
     uVRegion {bIsValid=false, max=<0 0>, min=<0 0>}
    }
   }
   slot {class=SizeBoxSlot, properties {horizontalAlignment=HAlign_Center}}
  }
  DecoyBottomButtons {
   widgetClassPath=&HorizontalBox
   parent=GridPanel_Main
   widget=DecoyBottomButtons
   slot {class=GridSlot, properties {horizontalAlignment=HAlign_Right, verticalAlignment=VAlign_Bottom}}
  }
  Dismiss_BottomBarDecoyButton {
   widgetClassPath=/Game/UI/Frontend/BottomBar/WBP_BottomBarDecoyButton
   bIsVariable=true
   parent=DecoyBottomButtons
   widget=Dismiss_BottomBarDecoyButton
   properties {
    holdData=&CommonUI_HoldData_Base_C
    visibility=Collapsed
    triggeringInputAction {dataTable=&InputActionDataTable, rowName=DismissFollower}
   }
   slot {class=HorizontalBoxSlot}
  }
 ]
}
