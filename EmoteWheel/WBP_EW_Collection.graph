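# Emote wheel collection screen, an activatable widget where the player browses locker items by category and plays a chosen emote.
# Item list events update the selection and hover view models; using an item plays it and requests menu close, or focuses search on gamepad when hovered.
# Use and clear-search buttons follow input type and report analytics.
# Intro and SubMenu animations slide and fade panels, blocking input through bound callbacks; closing reverses Intro before deactivating.
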WidgetBlueprint /EmoteWheel/WBP_EW_Collection
animations [
 Intro {
  binding_count=2
  length_seconds=0.25
  object=&self:Intro
  sequence {bindings [
    WBP_ItemSelection_ItemDetails {
     id=#1
     possessed_class=WBP_ItemSelection_ItemDetails_C
     tracks [
      {
       type=MovieSceneFloatTrack
       display_name=Render Opacity
       sections [{type=MovieSceneFloatSection, end_frame=5, start_frame=0, channels [None {type=MovieSceneScriptingFloatChannel, keys [[0,0],[5,1]]}]}]
      }
      {
       type=MovieScene2DTransformTrack
       sections [{
         type=MovieScene2DTransformSection
         end_frame=5
         start_frame=0
         channels [
          Translation.X {type=MovieSceneScriptingFloatChannel, const [0,2]}
          Translation.Y {type=MovieSceneScriptingFloatChannel, keys [[0,200],[2,0]]}
          Angle {type=MovieSceneScriptingFloatChannel}
          Scale.X {type=MovieSceneScriptingFloatChannel}
          Scale.Y {type=MovieSceneScriptingFloatChannel}
          Shear.X {type=MovieSceneScriptingFloatChannel}
          Shear.Y {type=MovieSceneScriptingFloatChannel}
         ]
        }]
      }
     ]
    }
    WBP_ItemCollectionList {
     id=#2
     possessed_class=WBP_ItemCollectionList_C
     tracks [
      {
       type=MovieSceneFloatTrack
       display_name=Render Opacity
       sections [{type=MovieSceneFloatSection, end_frame=5, start_frame=0, channels [None {type=MovieSceneScriptingFloatChannel, keys [[0,0],[5,1]]}]}]
      }
      {
       type=MovieScene2DTransformTrack
       sections [{
         type=MovieScene2DTransformSection
         end_frame=5
         start_frame=0
         channels [
          Translation.X {type=MovieSceneScriptingFloatChannel, const [0,2]}
          Translation.Y {type=MovieSceneScriptingFloatChannel, keys [[0,200],[2,0]]}
          Angle {type=MovieSceneScriptingFloatChannel}
          Scale.X {type=MovieSceneScriptingFloatChannel}
          Scale.Y {type=MovieSceneScriptingFloatChannel}
          Shear.X {type=MovieSceneScriptingFloatChannel}
          Shear.Y {type=MovieSceneScriptingFloatChannel}
         ]
        }]
      }
     ]
    }
   ]}
 }
 SubMenu {
  binding_count=2
  length_seconds=0.25
  object=&self:SubMenu
  sequence {bindings [
    WBP_ItemSelection_ItemDetails {
     id=#1
     possessed_class=WBP_ItemSelection_ItemDetails_C
     tracks [
      {
       type=MovieSceneFloatTrack
       display_name=Render Opacity
       sections [{type=MovieSceneFloatSection, end_frame=5, start_frame=0, channels [None {type=MovieSceneScriptingFloatChannel, keys [[0,1],[5,0]]}]}]
      }
      {
       type=MovieScene2DTransformTrack
       sections [{
         type=MovieScene2DTransformSection
         end_frame=5
         start_frame=0
         channels [
          Translation.X {type=MovieSceneScriptingFloatChannel, keys [[0,0],[5,-200]]}
          Translation.Y {type=MovieSceneScriptingFloatChannel}
          Angle {type=MovieSceneScriptingFloatChannel}
          Scale.X {type=MovieSceneScriptingFloatChannel}
          Scale.Y {type=MovieSceneScriptingFloatChannel}
          Shear.X {type=MovieSceneScriptingFloatChannel}
          Shear.Y {type=MovieSceneScriptingFloatChannel}
         ]
        }]
      }
     ]
    }
    WBP_ItemCollectionList {
     id=#2
     possessed_class=WBP_ItemCollectionList_C
     tracks [
      {
       type=MovieSceneFloatTrack
       display_name=Render Opacity
       sections [{type=MovieSceneFloatSection, end_frame=5, start_frame=0, channels [None {type=MovieSceneScriptingFloatChannel, keys [[0,1],[5,0]]}]}]
      }
      {
       type=MovieScene2DTransformTrack
       sections [{
         type=MovieScene2DTransformSection
         end_frame=5
         start_frame=0
         channels [
          Translation.X {type=MovieSceneScriptingFloatChannel, keys [[0,0],[5,-200]]}
          Translation.Y {type=MovieSceneScriptingFloatChannel}
          Angle {type=MovieSceneScriptingFloatChannel}
          Scale.X {type=MovieSceneScriptingFloatChannel}
          Scale.Y {type=MovieSceneScriptingFloatChannel}
          Shear.X {type=MovieSceneScriptingFloatChannel}
          Shear.Y {type=MovieSceneScriptingFloatChannel}
         ]
        }]
      }
     ]
    }
   ]}
 }
]
graphs [
 BP_GetDesiredInputConfig {nodes [
   Get Desired Input Config() {is=title K2Node_FunctionEntry}
   > Return Node!(
    execute
    ReturnValue:UIInput Config Structure=Make UIInput Config!(
     bIgnoreMoveInput:bool=true
     bIgnoreLookInput:bool=true
     InputMode:ECommonInputMode Enum=Menu
     MouseCaptureMode:EMouseCaptureMode Enum=NoCapture
     MouseLockMode:EMouseLockMode Enum=DoNotLock
     bHideCursorDuringViewportCapture:bool=True
    ) -> (UIInputConfig:UIInput Config Structure) {is=title K2Node_MakeStruct}
   )
  ]}
 EmoteWheelCategorySetup {nodes [EmoteWheelCategorySetup!() -> (then, Categories:Array of Emote Wheel Category VM Object References) {is=title K2Node_FunctionEntry}]}
 EventGraph {nodes [
   n1 = "Update Use Button Is Visible (WBP_ItemCollectionList)"!() -> (OutputDelegate, then, IsVisible:bool) {is=title K2Node_ComponentBoundEvent}
   > n2 = Set ShowUseButton(ShowUseButton:bool=@n1.IsVisible) -> (Output_Get:bool)
   > UpdateUseButtonVisibility!(execute, self, isVisible:bool=@n2) -> (then)
   n3 = "Trigger Activated Event (WBP_ItemCollectionList)"!() -> (OutputDelegate, then, LastActiveScreen:ELockerScreenList Enum=NewEnumerator0) {is=title K2Node_ComponentBoundEvent}
   > OnWidgetActivatedEvent(self:WBP Item Collection List Object Reference=Get WBP_ItemCollectionList() -> (WBP_ItemCollectionList:WBP Item Collection List Object Reference))
   > n4 = Branch(Condition="Equal (Enum)"(A:ELockerScreenList Enum=@n3.LastActiveScreen, B:ELockerScreenList Enum=NewEnumerator4))
   > PlayAnimationReverse(InAnimation=Get SubMenu() -> (SubMenu:Widget Animation Object Reference))
   PlayAnimationForward(execute=@n4.else, InAnimation=Get Intro())
   n5 = "Manual Update Selected Item (WBP_ItemCollectionList)"!() -> (OutputDelegate, then, SelectedItem:Fort Item VM Object Reference) {is=title K2Node_ComponentBoundEvent}
   > Is Valid(InputObject=@n5.SelectedItem)
   > UpdateSelectedVM!(execute, self, Item:Object Reference=@n5.SelectedItem) -> (then)
   > Set SelectedItem!(
    execute
    SelectedItem:Fort Item VM Object Reference=@n5.SelectedItem
    self:Fort Item Selection VM Object Reference=Get FortItemSelectionVM!(self:WBP Item Selection Item Details Object Reference=Get WBP_ItemSelection_ItemDetails() -> (WBP_ItemSelection_ItemDetails:WBP Item Selection Item Details Object Reference)) -> (FortItemSelectionVM:Fort Item Selection VM Object Reference)
   ) -> (then, Output_Get:Fort Item VM Object Reference)
   n6 = "Update Clear Button Is Visible (WBP_ItemCollectionList)"!() -> (OutputDelegate, then, IsVisible:bool) {is=title K2Node_ComponentBoundEvent}
   > Set ShowClearButton(ShowClearButton:bool=@n6.IsVisible) -> (Output_Get:bool)
   n7 = Get CommonInputSubsystem()
   > Branch(Condition="Equal (Enum)"(A:ECommonInputType Enum=GetCurrentInputType(self=@n7), B:ECommonInputType Enum=Gamepad))
   > SetBottomButtonVisibility!(
    execute
    self:WBP Bottom Bar Decoy Button Object Reference=Get ClearSearch_BottomBarDecoyButton() -> (ClearSearch_BottomBarDecoyButton:WBP Bottom Bar Decoy Button Object Reference)
    ShouldButtonBeVisible:bool=@n6.IsVisible
   ) -> (then)
   n8 = "Trigger Use Item (WBP_ItemCollectionList)"!() -> (OutputDelegate, then, Item:Object Reference) {is=title K2Node_ComponentBoundEvent}
   > UpdateSelectedVM(Item=@n8.Item)
   > UpdateLastSelectedCollectionItem!(execute, self, Item:Object Reference=@n8.Item) -> (then)
   > Delay(Duration=0.2)
   > UseItem()
   n9 = "On Item Hovered (WBP_ItemCollectionList)"!() -> (OutputDelegate, then, Item:Object Reference) {is=title K2Node_ComponentBoundEvent}
   > n10 = Cast To FortItemVM!(execute, Object=@n9.Item) -> (then, CastFailed, AsFort Item VM:Fort Item VM Object Reference) {is=title K2Node_DynamicCast}
   > Set with Broadcast HoveredFortItemVM!(execute, HoveredFortItemVM:Fort Item VM Object Reference=@n10) -> (then, Output_Get:Fort Item VM Object Reference)
   n11 = "Update Clear Button Is Enabled (WBP_ItemCollectionList)"!() -> (OutputDelegate, then, IsEnabled:bool) {is=title K2Node_ComponentBoundEvent}
   > SetIsEnabled(self=Get ClearSearch_BottomBarDecoyButton(), bInIsEnabled=@n11.IsEnabled)
   n12 = "On Clicked (ItemUse_BottomBarDecoyButton)"!() -> (OutputDelegate, then, Button:Common Button Base Object Reference) {is=title K2Node_ComponentBoundEvent}
   > TriggerFlowriderInteraction!(execute, self, Element:Widget Object Reference=@n12.Button, Value:String=Use Item) -> (then)
   n13 = Get HoveredFortItemVM() -> (HoveredFortItemVM:Fort Item VM Object Reference)
   > UpdateLastSelectedCollectionItem(Item=@n13)
   > UpdateSelectedVM(Item=@n13)
   > UseItem()
   n14 = "On Clicked (ClearSearch_BottomBarDecoyButton)"!() -> (OutputDelegate, then, Button:Common Button Base Object Reference) {is=title K2Node_ComponentBoundEvent}
   > TriggerFlowriderInteraction(Element=@n14.Button, Value=Clear Search)
   > ResetTextField(self:WBP Item Collection List Object Reference=Get WBP_ItemCollectionList())
   "Trigger Entry Right Clicked (WBP_ItemCollectionList)"() {is=title K2Node_ComponentBoundEvent}
   > Call On Close Menu!(execute:Exec, self:WBP EW Collection Object Reference) -> (then:Exec) {is=title K2Node_CallDelegate}
   CloseMenu() {is=title K2Node_CustomEvent}
   > SetAllInteractionsEnabled!(execute, self, IsInteractionEnabled:bool=false) -> (then)
   n15 = Get Intro()
   > Branch(Condition=NOT Boolean(A=IsAnimationPlaying(InAnimation=@n15)))
   > n16 = Play Animation with Finished event!(
    execute:Exec
    Widget:User Widget Object Reference=Self-Reference()
    InAnimation:Widget Animation Object Reference=@n15
    StartAtTime:float=0
    NumLoopsToPlay:int=1
    PlayMode:EUMGSequencePlayMode Enum=Reverse
    PlaybackSpeed:float=1
   ) -> (then:Exec, Result:Widget Animation Handle Structure, Finished:Exec) {is=title K2Node_PlayAnimation2}
   CloseMenu(execute=@n16.Finished, self:WBP Item Collection List Object Reference=Get WBP_ItemCollectionList()) {is=title K2Node_CallFunction}
   > DeactivateWidget()
   Event On Deactivated()
   > DeactivateWidget(self=Get WBP_CollectionNavigationWidget() -> (WBP_CollectionNavigationWidget:WBP Collection Navigation Widget Object Reference))
   Event Construct()
   n17 = Get CommonInputSubsystem()
   > Bind Event to On Input Method Changed(self=@n17, Delegate=Create Event())
   > HandleInputChanged!(execute, self, bNewInputType:ECommonInputType Enum=GetCurrentInputType(self=@n17)) -> (then)
   > SetupTransitionsBindings!(execute, self, Bind:bool=true) -> (then)
   Event On Activated()
   n18 = GetAnalyticsEWActivationCount!(self:Fort Picker Context Object Reference=Get FortPickerContext() -> (ReturnValue:Fort Picker Context Object Reference) {is=title K2Node_GetSubsystem}) -> (ReturnValue:int=0)
   > Fire Event Main View Opened!(
    execute
    View:Widget Object Reference=Self-Reference()
    AdditionalInfo:Flow Analytics Additional Info Structure="Add String (Non-Localized)"!(
     Info:"Flow Analytics Additional Info Structure (by ref)"=Make Flow Analytics Additional Info!() -> (FlowAnalyticsAdditionalInfo:Flow Analytics Additional Info Structure) {is=title K2Node_MakeStruct}
     Key:String=emote_wheel_session_id
     Value:String=GetAnalyticsSessionID!(ActivationCount:int=@n18) -> (ReturnValue:String)
    ) -> (ReturnValue:Flow Analytics Additional Info Structure)
    bIsOverlay:bool=false
   ) -> (then)
   > ActivateWidget(self=Get WBP_ItemCollectionList())
   > SetAllowDpadDownNavigation!(execute, self:Emote Wheel Collection Navigation Widget Object Reference=Get WBP_CollectionNavigationWidget(), InValue:bool=false) -> (then)
   n19 = GetCollectionLastSelectedItem!(self:Emote Wheel Item Selection VM Object Reference=Get EmoteWheelItemSelectionVM() -> (EmoteWheelItemSelectionVM:Emote Wheel Item Selection VM Object Reference)) -> (ReturnValue:Fort Item VM Object Reference)
   > Is Valid(InputObject=@n19)
   > Set CollectionLastSelectedItem!(
    execute
    CollectionLastSelectedItem:Fort Item VM Object Reference=@n19
    self:WBP Item Collection List Object Reference=Get WBP_ItemCollectionList()
   ) -> (then, Output_Get:Fort Item VM Object Reference)
   Event Destruct()
   > Unbind Event from On Input Method Changed(self=@n17, Delegate=Create Event())
   > SetupTransitionsBindings(Bind=false)
  ]}
 HandleInputChanged {nodes [
   HandleInputChanged!() -> (then, bNewInputType:ECommonInputType Enum) {is=title K2Node_FunctionEntry}
   > UpdateUseButtonVisibility(isVisible=Get ShowUseButton() -> (ShowUseButton:bool))
   n1 = Get bNewInputType() -> (bNewInputType:ECommonInputType Enum)
   > SetBottomButtonVisibility(
    self=Get ClearSearch_BottomBarDecoyButton()
    ShouldButtonBeVisible=AND Boolean(
     A=Get ShowClearButton() -> (ShowClearButton:bool)
     B="Not Equal (Enum)"!(A:ECommonInputType Enum=@n1, B:ECommonInputType Enum=MouseAndKeyboard) -> (ReturnValue:bool)
    )
   )
  ]}
 OnCloseMenu {nodes [OnCloseMenu() {is=title K2Node_FunctionEntry}]}
 OnKeyUp {nodes [
   n1 = OnKeyUp!() -> (then, MyGeometry:Geometry Structure, InKeyEvent:Key Event Structure) {is=title K2Node_FunctionEntry}
   > Branch(Condition="Equal (Key)"(
     A:Key Structure=GetKey!(Input:"Key Event Structure (by ref)"=@n1.InKeyEvent) -> (ReturnValue:Key Structure)
     B:Key Structure=Get GamepadNavDown() -> (GamepadNavDown:Key Structure)
    ) -> (ReturnValue:bool=~) {is=title K2Node_PromotableOperator})
   > SetAllowDpadDownNavigation(self=Get WBP_CollectionNavigationWidget(), InValue=true)
   > Return Node(ReturnValue:Event Reply Structure=Make Event Reply!() -> (EventReply:Event Reply Structure) {is=title K2Node_MakeStruct})
  ]}
 OnTransitionAnimationFinished {nodes [
   OnTransitionAnimationFinished() {is=title K2Node_FunctionEntry}
   > SetAllInteractionsEnabled(IsInteractionEnabled=true)
  ]}
 OnTransitionAnimationStarted {nodes [
   OnTransitionAnimationStarted() {is=title K2Node_FunctionEntry}
   > SetAllInteractionsEnabled(IsInteractionEnabled=false)
  ]}
 SetAllInteractionsEnabled {nodes [
   SetAllInteractionsEnabled!() -> (then, IsInteractionEnabled:bool=~) {is=title K2Node_FunctionEntry}
   > n1 = Set IgnoreInputs(IgnoreInputs:bool=Get IsInteractionEnabled() -> (IsInteractionEnabled:bool)) -> (Output_Get:bool)
   > SetAllInteractionsEnabled!(execute, self:WBP Item Collection List Object Reference=Get WBP_ItemCollectionList(), IsInteractionEnabled=@n1, IncludeSortFilter:bool=true) -> (then) {is=title K2Node_CallFunction}
  ]}
 SetupTransitionsBindings {nodes [
   n1 = SetupTransitionsBindings!() -> (then, Bind:bool=~) {is=title K2Node_FunctionEntry}
   > n2 = Branch(Condition=@n1)
   n3 = Create Event()
   n4 = Get Intro()
   > BindToAnimationStarted!(execute, self:User Widget Object Reference, Animation:Widget Animation Object Reference=@n4, Delegate:Delegate=@n3) -> (then)
   n5 = Create Event()
   n6 = Get SubMenu()
   > BindToAnimationStarted(Animation=@n6, Delegate=@n5)
   n7 = Create Event()
   n8 = Get Intro()
   > BindToAnimationFinished(Animation=@n8, Delegate=@n7)
   n9 = Create Event()
   n10 = Get SubMenu()
   > BindToAnimationFinished(Animation=@n10, Delegate=@n9)
   UnbindFromAnimationStarted!(execute=@n2.else, self:User Widget Object Reference, Animation:Widget Animation Object Reference=@n4, Delegate:Delegate=@n3) -> (then)
   > UnbindFromAnimationStarted(Animation=@n6, Delegate=@n5)
   > UnbindFromAnimationFinished!(execute, self:User Widget Object Reference, Animation:Widget Animation Object Reference=@n8, Delegate:Delegate=@n7) -> (then)
   > UnbindFromAnimationFinished(Animation=@n10, Delegate=@n9)
  ]}
 TriggerFlowriderInteraction {nodes [
   n1 = TriggerFlowriderInteraction!() -> (then, Element:Widget Object Reference, Value:String) {is=title K2Node_FunctionEntry}
   n2 = GetAnalyticsEWActivationCount(self=Get FortPickerContext()) -> (ReturnValue=0)
   n3 = Make Flow Analytics Additional Info()
   > Fire Event Element Interacted(
    Element=@n1
    AdditionalInfo="Add String (Non-Localized)"(
     Info="Add String (Non-Localized)"(Info=@n3, Key=Button Name, Value=@n1.Value)
     Key=emote_wheel_session_id
     Value=GetAnalyticsSessionID(ActivationCount=@n2)
    )
   )
  ]}
 UpdateEmoteCategory {nodes [
   n1 = UpdateEmoteCategory!() -> (then, Category:Emote Wheel Category VM Object Reference) {is=title K2Node_FunctionEntry}
   > n2 = Cast To EmoteWheelLockerCategoryVM!(execute, Object=@n1) -> (then, CastFailed, AsEmote Wheel Locker Category VM:Emote Wheel Locker Category VM Object Reference) {is=title K2Node_DynamicCast}
   > Set CategoryCustomClass!(
    execute
    CategoryCustomClass:User Widget Soft Class Reference=GetCustomCollectionListWidget!(self:Emote Wheel Category VM Object Reference=@n2) -> (ReturnValue:User Widget Soft Class Reference)
    self:WBP Item Collection List Object Reference=Get WBP_ItemCollectionList()
   ) -> (then, Output_Get:User Widget Soft Class Reference)
   > Set Category!(
    execute
    Category:Fort Locker Category VM Object Reference=GetLockerCategoryVM!(self:Emote Wheel Locker Category VM Object Reference=@n2) -> (ReturnValue:Fort Locker Category VM Object Reference)
    self:Fort Locker Category Filter VM Object Reference=Get FortLockerCategoryFilterVM!(self:WBP Item Collection List Object Reference=Get WBP_ItemCollectionList()) -> (FortLockerCategoryFilterVM:Fort Locker Category Filter VM Object Reference)
   ) -> (then, Output_Get:Fort Locker Category VM Object Reference)
  ]}
 UpdateLastSelectedCollectionItem {nodes [
   n1 = UpdateLastSelectedCollectionItem!() -> (then, Item:Object Reference) {is=title K2Node_FunctionEntry}
   > n2 = Cast To FortItemVM(Object=@n1)
   > SetCollectionLastSelectedItem!(execute, self:Emote Wheel Item Selection VM Object Reference=Get EmoteWheelItemSelectionVM(), InItemVM:Fort Item VM Object Reference=@n2) -> (then)
  ]}
 UpdateSelectedVM {nodes [
   n1 = UpdateSelectedVM!() -> (then, Item:Object Reference) {is=title K2Node_FunctionEntry}
   > n2 = Cast To FortItemVM(Object=@n1)
   > n3 = Get Emote Wheel Item VM From Item VM!(
    execute
    self:Emote Wheel Category VM Object Reference=Get SelectedCategory!(self:Emote Wheel Item Selection VM Object Reference=Get EmoteWheelItemSelectionVM()) -> (SelectedCategory:Emote Wheel Category VM Object Reference)
    InFortItemVM:Fort Item VM Object Reference=@n2
   ) -> (then, ReturnValue:Emote Wheel Item VM Object Reference) {is=title K2Node_CallFunction}
   > SetSelectedItem!(execute, self:Emote Wheel Item Selection VM Object Reference=Get EmoteWheelItemSelectionVM(), Item:Emote Wheel Item VM Object Reference=@n3) -> (then)
  ]}
 UpdateUseButtonVisibility {nodes [
   n1 = UpdateUseButtonVisibility!() -> (then, isVisible:bool=~) {is=title K2Node_FunctionEntry}
   n2 = GetCurrentInputType(self=Get CommonInputSubsystem())
   > SetBottomButtonVisibility(
    self=Get ItemUse_BottomBarDecoyButton() -> (ItemUse_BottomBarDecoyButton:WBP Bottom Bar Decoy Button Object Reference)
    ShouldButtonBeVisible=AND Boolean(A=@n1, B="Equal (Enum)"(A=@n2, B=Gamepad))
   )
  ]}
 UseItem {nodes [
   UseItem() {is=title K2Node_FunctionEntry}
   n1 = Get CommonInputSubsystem()
   > n2 = Branch(Condition="Not Equal (Enum)"(A=GetCurrentInputType(self=@n1), B=Gamepad))
   n3 = Get!(execute:Exec=[@n5.then, @n2.then], self:Emote Wheel Item Selection VM Object Reference=Get EmoteWheelItemSelectionVM()) -> (then:Exec, else:Exec, SelectedItem:Emote Wheel Item VM Object Reference)
   > Play(self:Emote Wheel Item VM Object Reference=@n3)
   > Call On Close Menu()
   n4 = Get WBP_ItemCollectionList()
   n5 = Branch(
    execute=@n2.else
    Condition=NOT Boolean(A=Get Is Search Hovered!(self:WBP Item Collection List Object Reference=@n4) -> (Is Search Hovered:bool))
   )
   FocusSearchField!(execute=@n5.else, self:WBP Item Collection List Object Reference=@n4) -> (then, bPassThrough:bool)
  ]}
]
variables [
 ChatEventWatch {type=AsyncAction_StartListeningToEvent, default=null}
 GamepadNavDown {type=Key, default=Gamepad_DPad_Down}
 HoveredFortItemVM {type=FortItemVM, default=null}
 IgnoreInputs {type=boolean}
 OnCloseMenu {}
 ShowClearButton {type=boolean}
 ShowUseButton {type=boolean}
]
widgets {
 info {inheritedWidgetCount=0, parentClass=&CommonActivatableWidget, rootWidgetClass=&SafeZone, widgetCount=10}
 widgets [
  SafeZone {widgetClassPath=&SafeZone, bIsVariable=true, widget=SafeZone, properties {padBottom=false, padTop=false}}
  SizeBox {
   widgetClassPath=&SizeBox
   bIsVariable=true
   parent=SafeZone
   widget=SizeBox
   properties {bOverride_WidthOverride=true, widthOverride=1920}
   slot {class=SafeZoneSlot, properties {hAlign=HAlign_Center}}
  }
  GridPanel_Collection {
   widgetClassPath=&GridPanel
   parent=SizeBox
   widget=GridPanel_Collection
   properties {columnFill [1,0], rowFill [1,1,1,1]}
   slot {class=SizeBoxSlot, properties {padding=<ltrb 80 64 80 32>}}
  }
  SizeBox_0 {
   widgetClassPath=&SizeBox
   parent=GridPanel_Collection
   widget=SizeBox_0
   properties {bOverride_MaxDesiredWidth=true, bOverride_MinDesiredWidth=true, maxDesiredWidth=866, minDesiredWidth=866}
   slot {class=GridSlot, properties {column=1, rowSpan=4, padding=<ltrb 0 -30 26 72>}}
  }
  WBP_ItemCollectionList {
   widgetClassPath=/Locker/MPL/WBP_ItemCollectionList
   bIsVariable=true
   parent=SizeBox_0
   widget=WBP_ItemCollectionList
   properties {bIsFocusable=true, isLoadedAsEmoteWheel=true, usePreviewWidget=false}
   slot {class=SizeBoxSlot}
  }
  WBP_ItemSelection_ItemDetails {
   widgetClassPath=/Locker/MPL/WBP_ItemSelection_ItemDetails
   bIsVariable=true
   parent=GridPanel_Collection
   widget=WBP_ItemSelection_ItemDetails
   properties {fortLockerEquipUtilityVM=&self:WidgetTree.WBP_ItemSelection_ItemDetails.FortLockerEquipUtilityVM_0}
   slot {class=GridSlot, properties {horizontalAlignment=HAlign_Left, row=2, rowSpan=2, verticalAlignment=VAlign_Bottom, padding=<ltrb 0 0 0 64>}}
  }
  WBP_CollectionNavigationWidget {
   widgetClassPath=&/WBP_CollectionNavigationWidget
   bIsVariable=true
   parent=GridPanel_Collection
   widget=WBP_CollectionNavigationWidget
   slot {class=GridSlot}
  }
  DecoyBottomButtons {
   widgetClassPath=&HorizontalBox
   parent=GridPanel_Collection
   widget=DecoyBottomButtons
   slot {class=GridSlot, properties {column=1, horizontalAlignment=HAlign_Right, row=3, verticalAlignment=VAlign_Bottom}}
  }
  ItemUse_BottomBarDecoyButton {
   widgetClassPath=/Game/UI/Frontend/BottomBar/WBP_BottomBarDecoyButton
   bIsVariable=true
   parent=DecoyBottomButtons
   widget=ItemUse_BottomBarDecoyButton
   properties {
    holdData=&CommonUI_HoldData_Base_C
    priority=10000
    visibility=Collapsed
    triggeringInputAction {dataTable=&InputActionDataTable, rowName=EWUse}
   }
   slot {class=HorizontalBoxSlot}
  }
  ClearSearch_BottomBarDecoyButton {
   widgetClassPath=&WBP_BottomBarDecoyButton
   bIsVariable=true
   parent=DecoyBottomButtons
   widget=ClearSearch_BottomBarDecoyButton
   properties {
    holdData=&CommonUI_HoldData_Base_C
    priority=500
    visibility=Collapsed
    triggeringInputAction {dataTable=&InputActionDataTable, rowName=EWClearResults}
   }
   slot {class=HorizontalBoxSlot}
  }
 ]
}
