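# Emote wheel slot button showing an item's icon and name, and broadcasting its view-model item on click and its index on selection.
# SetEmoteIndex picks the key binding and label layout from per-index arrays, rotating even slots 45 degrees while counter-rotating their content.
# Selection plays animation, rumble and sound, and switches title scrolling or truncation.
# Icons lazy-load into a dynamic material with rounded corners; compact UI adjusts opacity and size.
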WidgetBlueprint /EmoteWheel/WBP_Emote_RadialEntry
animations [EntryActive {
  binding_count=4
  length_seconds=0.15
  object=&self:EntryActive
  sequence {bindings [
    LazyImage_Item {id=#1, possessed_class=FortLazyImage, tracks []}
    Item_Text {
     id=#2
     possessed_class=CommonTextBlock
     tracks [{
       type=MovieSceneColorTrack
       display_name=Color and Opacity
       sections [{
         type=MovieSceneColorSection
         end_frame=3
         start_frame=0
         channels [
          Color.R {type=MovieSceneScriptingFloatChannel, default=1}
          Color.G {type=MovieSceneScriptingFloatChannel, default=1}
          Color.B {type=MovieSceneScriptingFloatChannel, default=1}
          Color.A {type=MovieSceneScriptingFloatChannel, keys [[1,0.7],[3,1]]}
         ]
        }]
      }]
    }
    GridSlot_3 {
     id=#3
     parent_id=#1
     possessed_class=GridSlot
     tracks [{
       type=MovieSceneMarginTrack
       display_name=Padding
       sections [{
         type=MovieSceneMarginSection
         end_frame=3
         start_frame=0
         channels [
          Left {type=MovieSceneScriptingFloatChannel, keys [[0,9],[3,7]]}
          Top {type=MovieSceneScriptingFloatChannel, keys [[0,9],[3,7]]}
          Right {type=MovieSceneScriptingFloatChannel, keys [[0,9],[3,7]]}
          Bottom {type=MovieSceneScriptingFloatChannel, keys [[0,9],[3,7]]}
         ]
        }]
      }]
    }
    SizeBox_Entry {
     id=#4
     possessed_class=SizeBox
     tracks [{
       type=MovieScene2DTransformTrack
       sections [{
         type=MovieScene2DTransformSection
         end_frame=3
         start_frame=0
         channels [
          Scale.X {type=MovieSceneScriptingFloatChannel, keys [[0,1],[3,1.02],[3.6,1]]}
          Scale.Y {type=MovieSceneScriptingFloatChannel, keys [[0,1],[3,1.02],[3.6,1]]}
         ]
        }]
      }]
    }
   ]}
 }]
graphs [
 EventGraph {nodes [
   n1 = "On Clicked (EntryNameInteractionButton)"() {is=title K2Node_ComponentBoundEvent}
   Call On Entry Clicked!(
    execute:Exec=[@n1.then, @n2.then]
    self:WBP Emote Radial Entry Object Reference
    Item:Emote Wheel Item VM Object Reference=Get EmoteWheelItemVM() -> (EmoteWheelItemVM:Emote Wheel Item VM Object Reference)
   ) -> (then:Exec) {is=title K2Node_CallDelegate}
   TriggerRumbleFeedback() {is=title K2Node_CustomEvent}
   > Client Play Force Feedback!(
    execute
    self:Player Controller Object Reference=GetOwningPlayer()
    ForceFeedbackEffect:Force Feedback Effect Object Reference=&/FF_EW_SlotSelected
    Tag:Name=EmoteWheel
    bLooping:bool=false
    bIgnoreTimeDilation:bool=false
    bPlayWhilePaused:bool=false
   ) -> (then)
   StopRumbleFeedback() {is=title K2Node_CustomEvent}
   > ClientStopForceFeedback!(
    execute
    self:Player Controller Object Reference=GetOwningPlayer()
    ForceFeedbackEffect:Force Feedback Effect Object Reference=&FF_EW_SlotSelected
    Tag:Name=EmoteWheel
   ) -> (then)
   n2 = Event On Clicked()
   Event PreConstruct()
   > SetMobileStyle() {is=title K2Node_CallFunction}
   Event On Selected()
   > OnEntryActive()
   > TriggerRumbleFeedback() {is=title K2Node_CallFunction}
   > UpdateScrollEnabled!(execute, self, isSelected:bool=true) -> (then)
   > UpdateOverflowPolicy!(execute, self, isSelected:bool=true) -> (then)
   > PlaySound2D(Sound=/Game/Sounds/Fort_UI_Sounds/FrontEnd/Tabs/MSS_UI_Button_Tab_Hover_Generic)
   Event On Deselected()
   > OnEntryNotActive()
   > StopRumbleFeedback() {is=title K2Node_CallFunction}
   > UpdateScrollEnabled(isSelected=false)
   > UpdateOverflowPolicy(isSelected=false)
   n3 = Get EmoteIndex() -> (EmoteIndex:int=0)
   "In Range (Integer)"!(Value:int=@n3, Min:int=2, Max:int=4, InclusiveMin:bool=true, InclusiveMax:bool=true) -> (ReturnValue:bool)
   "In Range (Integer)"(Value=@n3, Min=4, Max=6, InclusiveMin=true, InclusiveMax=true)
   OR Boolean(A="integer > integer"(A=@n3, B=5), B="Equal (Integer)"(A=@n3))
   "integer < integer"(A=@n3, B=3)
  ]}
 GetItemTextVisible {nodes [
   n1 = GetItemTextVisible!() -> (then, Display Name:Text) {is=title K2Node_FunctionEntry}
   n2 = Get Source String from Text!(Text:Text=@n1) -> (ReturnValue:String) {is=title K2Node_CallFunction}
   n3 = MakeLiteralString(Value=EMPTY)
   > Return Node!(
    execute
    InVisibility:ESlateVisibility Enum=Select(Option 0=SelfHitTestInvisible, Option 1=Collapsed, Index="Equal Exactly (String)"(A=@n2, B=@n3))
   )
  ]}
 OnEntryActive {nodes [
   OnEntryActive() {is=title K2Node_FunctionEntry}
   > PlayAnimationForward(InAnimation=Get EntryActive() -> (EntryActive:Widget Animation Object Reference))
   > Call On Entry Active Index Changed!(
    execute:Exec
    self:WBP Emote Radial Entry Object Reference
    ActiveEntryIndex:int=Get EmoteIndex() -> (EmoteIndex=0)
    Item:Emote Wheel Item VM Object Reference=Get EmoteWheelItemVM()
   ) -> (then:Exec) {is=title K2Node_CallDelegate}
  ]}
 OnEntryActiveIndexChanged {nodes [OnEntryActiveIndexChanged!() -> (then, ActiveEntryIndex:int, Item:Emote Wheel Item VM Object Reference) {is=title K2Node_FunctionEntry}]}
 OnEntryClicked {nodes [OnEntryClicked!() -> (then, Item:Emote Wheel Item VM Object Reference) {is=title K2Node_FunctionEntry}]}
 OnEntryNotActive {nodes [
   OnEntryNotActive() {is=title K2Node_FunctionEntry}
   > PlayAnimationReverse(InAnimation=Get EntryActive())
  ]}
 OnHandleInputMethodChanged {nodes [
   n1 = OnHandleInputMethodChanged!() -> (then, NewInputType:ECommonInputType Enum) {is=title K2Node_FunctionEntry}
   > SetIsSelectable!(
    execute
    self:Common Button Base Object Reference
    bInIsSelectable:bool=Select!(MouseAndKeyboard:bool=false, Gamepad:bool=false, Touch:bool=true, Count:bool=false, Index:ECommonInputType Enum=@n1) -> (ReturnValue:bool)
   ) -> (then)
  ]}
 SetAllowInteractionWithEntryText {nodes [
   n1 = SetAllowInteractionWithEntryText!() -> (then, bAllowInteraction:bool=~) {is=title K2Node_FunctionEntry}
   n2 = IsUsingInputType!(InputType:ECommonInputType Enum=Touch, PlayerController:Player Controller Object Reference=GetOwningPlayer()) -> (ReturnValue:bool)
   > SetVisibility(
    self=Get EntryNameInteractionButton() -> (EntryNameInteractionButton:Button Object Reference)
    InVisibility=Select!(Option 0:ESlateVisibility Enum=Collapsed, Option 1:ESlateVisibility Enum=Visible, Index:bool=AND Boolean(A=@n1, B=@n2)) -> (ReturnValue:ESlateVisibility Enum)
   )
  ]}
 SetEmoteIndex {nodes [
   n1 = SetEmoteIndex!() -> (then, ItemIndex:int) {is=title K2Node_FunctionEntry}
   > n2 = Set EmoteIndex(EmoteIndex:int=@n1) -> (Output_Get:int=0)
   > SetKBMEntryInput!(execute, self, Index:int=@n2) -> (then)
   n3 = Get EmoteIndex() -> (EmoteIndex=0)
   n4 = Get Name_Attachment() -> (Name_Attachment:Common Visual Attachment Object Reference)
   > Set ContentAnchor!(
    execute
    ContentAnchor:Vector 2D Structure="Get (a copy)"(Array:Array of Vector 2D Structures=Get ContentAnchor_Array() -> (ContentAnchor_Array:Array of Vector 2D Structures), Dimension 1=@n3) -> (Output:Vector 2D Structure)
    self:Common Visual Attachment Object Reference=@n4
   ) -> (then, Output_Get:Vector 2D Structure)
   > SetGridSlotAlignment!(
    execute
    Widget:Widget Object Reference=@n4
    Horizontal:EHorizontalAlignment Enum="Get (a copy)"(
     Array:Array of EHorizontalAlignment Enums=Get HorizontalAlignment_Array() -> (HorizontalAlignment_Array:Array of EHorizontalAlignment Enums=HAlign_Fill)
     Dimension 1=@n3
    ) -> (Output:EHorizontalAlignment Enum)
    Vertical:EVerticalAlignment Enum="Get (a copy)"(
     Array:Array of EVerticalAlignment Enums=Get Vertical Alignment_Array() -> (Vertical Alignment_Array:Array of EVerticalAlignment Enums=VAlign_Fill)
     Dimension 1=@n3
    ) -> (Output:EVerticalAlignment Enum)
   ) -> (then)
   > SetGridSlotNudge!(
    execute
    Widget:Widget Object Reference=@n4
    Nudge:Vector 2D Structure="Get (a copy)"(Array:Array of Vector 2D Structures=Get NudgeArray() -> (NudgeArray:Array of Vector 2D Structures), Dimension 1=@n3) -> (Output:Vector 2D Structure)
   ) -> (then)
   > Set Justification!(
    execute
    Justification:ETextJustify Enum="Get (a copy)"(
     Array:Array of ETextJustify Enums=Get TextJustification_Array() -> (TextJustification_Array:Array of ETextJustify Enums=Left)
     Dimension 1=@n3
    ) -> (Output:ETextJustify Enum)
    self:Text Layout Widget Object Reference=Get Item_Text() -> (Item_Text:Common Text Object Reference)
   ) -> (then, Output_Get:ETextJustify Enum=Left)
   n5 = "Equal (Integer)"(A="% (Integer)"(A=Get EmoteIndex() -> (EmoteIndex=0), B=2) -> (ReturnValue=0))
   > SetRenderTransformAngle!(
    execute
    self:Widget Object Reference=Self-Reference()
    Angle:float=Select(Option 0:float=0, Option 1:float=45, Index=@n5) -> (ReturnValue:float)
   ) -> (then)
   > SetRenderTransformAngle(self=Get SizeBox_Entry() -> (SizeBox_Entry:Size Box Object Reference), Angle=Select(Option 0=0, Option 1=-45, Index=@n5))
  ]}
 SetIconScale {nodes [
   n1 = SetIconScale!() -> (then, ScaleFactor:double=~) {is=title K2Node_FunctionEntry}
   > SetRenderScale!(
    execute
    self:Widget Object Reference=Get LazyImage_Item() -> (LazyImage_Item:Fort Lazy Image Object Reference)
    Scale_X:double=@n1
    Scale_Y:double=@n1
   ) -> (then)
  ]}
 SetItemImage {nodes [
   SetItemImage!() -> (then, Item Image:Soft Object Reference) {is=title K2Node_FunctionEntry}
   n1 = Get Item Image() -> (Item Image:Soft Object Reference)
   > n2 = Branch(Condition=NOT Boolean(A=IsAssetNull!(Asset:Soft Object Reference=@n1) -> (ReturnValue:bool)))
   n3 = Get LazyImage_Item()
   > SetVisibility(self=@n3, InVisibility=SelfHitTestInvisible)
   > SetEnableTextureReleasingOverride!(
    execute
    self:Fort Lazy Image Object Reference=@n3
    bInEnableTextureReleasingOverride:bool=IsTextureReleasingEnabled!(self:Emote Wheel Item VM Object Reference=Get EmoteWheelItemVM()) -> (ReturnValue:bool)
   ) -> (then)
   > SetMaterialTextureParamName!(execute, self:Common Lazy Image Object Reference=@n3, TextureParamName:Name=Item) -> (then)
   > SetBrushFromLazyDisplayAsset!(
    execute
    self:Common Lazy Image Object Reference=@n3
    LazyObject:"Soft Object Reference (by ref)"=Get Item Image()
    bMatchTextureSize:bool=false
   ) -> (then)
   > n4 = GetDynamicMaterial(self=@n3)
   > n5 = Set ItemDynamicMaterial(ItemDynamicMaterial:Material Instance Dynamic Object Reference=@n4) -> (Output_Get:Material Instance Dynamic Object Reference)
   > UpdateMaterialCorners!(execute, self, DynamicMaterialInstance:Material Instance Dynamic Object Reference=@n5) -> (then)
   SetVisibility(execute=@n2.else, self=@n3, InVisibility=Collapsed)
  ]}
 SetKBMEntryInput {nodes [
   n1 = SetKBMEntryInput!() -> (then, Index:int) {is=title K2Node_FunctionEntry}
   > SetTriggeringInputAction!(
    execute
    self:Common Button Base Object Reference
    InputActionRow:"Data Table Row Handle Structure (by ref)"="Get (a copy)"(
     Array:Array of Data Table Row Handle Structures=Get KBMEntryInput() -> (KBMEntryInput:Array of Data Table Row Handle Structures)
     Dimension 1=@n1
    ) -> (Output:Data Table Row Handle Structure)
   ) -> (then)
  ]}
 SetMobileStyle {nodes [
   SetMobileStyle() {is=title K2Node_FunctionEntry}
   > Branch(Condition=IsUsingCompactUI())
   > SetOpacity!(execute, self:Text Object Reference=Get Item_Text(), InOpacity:float=1) -> (then)
   > n1 = Branch(Condition=IsTabletAspectRatio!() -> (ReturnValue:bool))
   n2 = Get SizeBox_Entry()
   SetWidthOverride(execute=@n1.else, self=@n2, InWidthOverride=96)
   > SetHeightOverride(self=@n2, InHeightOverride=96)
   > SetSize!(execute, self:Spacer Object Reference=Get Spacer_Entry() -> (Spacer_Entry:Spacer Object Reference), InSize:Vector 2D Structure=(X=232,Y=232)) -> (then)
  ]}
 SetTruncationTitles {nodes [
   n1 = SetTruncationTitles!() -> (then, ShouldTruncateTitles:bool=~) {is=title K2Node_FunctionEntry}
   > Set Should Truncate Titles(Should Truncate Titles:bool=@n1) -> (Output_Get:bool)
  ]}
 UpdateMaterialCorners {nodes [
   n1 = UpdateMaterialCorners!() -> (then, DynamicMaterialInstance:Material Instance Dynamic Object Reference) {is=title K2Node_FunctionEntry}
   > SetScalarParameterValue(self=@n1, ParameterName=UseValueInPixels, Value="To Float (Boolean)"(InBool=Get UseRadiusInPixels() -> (UseRadiusInPixels:bool)))
   n2 = Get CornerRadius() -> (CornerRadius:float)
   > SetScalarParameterValue(self=@n1, ParameterName=TopLeftCornerRadius, Value=@n2)
   > SetScalarParameterValue(self=@n1, ParameterName=TopRightCornerRadius, Value=@n2)
   > SetScalarParameterValue(self=@n1, ParameterName=BottomLeftCornerRadius, Value=@n2)
   > SetScalarParameterValue(self=@n1, ParameterName=BottomRightCornerRadius, Value=@n2)
  ]}
 UpdateOverflowPolicy {nodes [
   n1 = UpdateOverflowPolicy!() -> (then, isSelected:bool=~) {is=title K2Node_FunctionEntry}
   > n2 = Branch(Condition=@n1)
   n3 = Get Item_Text()
   > SetTextOverflowPolicy!(
    execute
    self:Text Object Reference=@n3
    InOverflowPolicy:ETextOverflowPolicy Enum=Select(
     Option 0:ETextOverflowPolicy Enum=Clip
     Option 1:ETextOverflowPolicy Enum=MultilineEllipsis
     Index=Get Should Truncate Titles() -> (Should Truncate Titles:bool)
    ) -> (ReturnValue:ETextOverflowPolicy Enum)
   ) -> (then)
   SetTextOverflowPolicy(execute=@n2.else, self=@n3, InOverflowPolicy=MultilineEllipsis)
  ]}
 UpdateScrollEnabled {nodes [
   n1 = UpdateScrollEnabled!() -> (then, isSelected:bool=~) {is=title K2Node_FunctionEntry}
   > n2 = Branch(Condition=@n1)
   n3 = Get Item_Text()
   > SetScrollingEnabled!(execute, self:Common Text Object Reference=@n3, bInIsScrollingEnabled:bool=NOT Boolean(A=Get Should Truncate Titles())) -> (then)
   SetScrollingEnabled(execute=@n2.else, self=@n3, bInIsScrollingEnabled=false)
  ]}
 UpdateTextStyle {nodes [UpdateTextStyle!() -> (then, isSelected:bool=~) {is=title K2Node_FunctionEntry}]}
]
variables [
 CornerRadius {type=number, default=10}
 EmoteIndex {type=integer, default=2}
 ItemDynamicMaterial {type=MaterialInstanceDynamic}
 OnEntryActiveIndexChanged {}
 OnEntryClicked {}
 SelectedStyleTags {type="array<GameplayTag>", default [{tagName=StyleSheet.Style.Strong}]}
 Should Truncate Titles {type=boolean}
 UseRadiusInPixels {type=boolean, default=true}
]
widgets {
 info {inheritedWidgetCount=1, parentClass=&CommonButtonLegacy, rootWidgetClass=&Overlay, widgetCount=11}
 widgets [
  Overlay_Main {widgetClassPath=&Overlay, widget=Overlay_Main}
  Spacer_Entry {
   widgetClassPath=&Spacer
   bIsVariable=true
   parent=Overlay_Main
   widget=Spacer_Entry
   properties {visibility=Visible, size=<256 256>}
   slot {class=OverlaySlot, properties {horizontalAlignment=HAlign_Center, verticalAlignment=VAlign_Center}}
  }
  SizeBox_Entry {
   widgetClassPath=&SizeBox
   bIsVariable=true
   parent=Overlay_Main
   widget=SizeBox_Entry
   properties {bOverride_HeightOverride=true, bOverride_WidthOverride=true, heightOverride=114, widthOverride=114}
   slot {class=OverlaySlot, properties {horizontalAlignment=HAlign_Center, verticalAlignment=VAlign_Center}}
  }
  GridPanel_0 {
   widgetClassPath=&GridPanel
   parent=SizeBox_Entry
   widget=GridPanel_0
   properties {columnFill [1], rowFill [1]}
   slot {class=SizeBoxSlot}
  }
  LazyImage_Item {
   widgetClassPath=&FortLazyImage
   bIsVariable=true
   parent=GridPanel_0
   widget=LazyImage_Item
   properties {
    visibility=Collapsed
    brush {
     bIsDynamicallyLoaded=false
     drawAs=Image
     imageType=FullColor
     mirroring=NoMirror
     resourceName=None
     resourceObject=/FortUILibrary/Systems/Items/MaterialInstances/MI_UIKit_ItemLocker
     tiling=NoTile
     imageSize=<32 32>
     margin=<ltrb 0 0 0 0>
     outlineSettings {
      bUseBrushTransparency=false
      roundingType=HalfHeightRadius
      width=0
      color {colorUseRule=UseColor_Specified, specifiedColor=<rgba 0 0 0 0>}
      cornerRadii=<0 0 0 0>
     }
     tintColor {colorUseRule=UseColor_Specified, specifiedColor=<rgba 1 1 1 1>}
     uVRegion {bIsValid=false, max=<0 0>, min=<0 0>}
    }
    loadingBackgroundBrush {
     bIsDynamicallyLoaded=false
     drawAs=Image
     imageType=FullColor
     mirroring=NoMirror
     resourceName=None
     resourceObject=/FortUILibrary/Systems/Items/MaterialInstances/MI_UIKit_Item_Loading
     tiling=NoTile
     imageSize=<32 32>
     margin=<ltrb 0 0 0 0>
     outlineSettings {
      bUseBrushTransparency=false
      roundingType=HalfHeightRadius
      width=0
      color {colorUseRule=UseColor_Specified, specifiedColor=<rgba 0 0 0 0>}
      cornerRadii=<0 0 0 0>
     }
     tintColor {colorUseRule=UseColor_Specified, specifiedColor=<rgba 1 1 1 1>}
     uVRegion {bIsValid=false, max=<0 0>, min=<0 0>}
    }
    loadingThrobberBrush {
     bIsDynamicallyLoaded=false
     drawAs=Image
     imageType=FullColor
     mirroring=NoMirror
     resourceName=None
     resourceObject=/Game/Athena/UI/Frontend/Spatial/Art/T_Ui_Empty
     tiling=NoTile
     imageSize=<16 16>
     margin=<ltrb 0 0 0 0>
     outlineSettings {
      bUseBrushTransparency=false
      roundingType=HalfHeightRadius
      width=0
      color {colorUseRule=UseColor_Specified, specifiedColor=<rgba 0 0 0 0>}
      cornerRadii=<0 0 0 0>
     }
     tintColor {colorUseRule=UseColor_Specified, specifiedColor=<rgba 1 1 1 1>}
     uVRegion {bIsValid=false, max=<0 0>, min=<0 0>}
    }
   }
   slot {class=GridSlot, properties {padding=<ltrb 9 9 9 9>}}
  }
  Name_Attachment {
   widgetClassPath=&CommonVisualAttachment
   bIsVariable=true
   parent=GridPanel_0
   widget=Name_Attachment
   properties {contentAnchor=<0 1>}
   slot {class=GridSlot, properties {horizontalAlignment=HAlign_Center, verticalAlignment=VAlign_Top, padding=<ltrb -16 0 -16 0>}}
  }
  Overlay_1 {widgetClassPath=&Overlay, parent=Name_Attachment, widget=Overlay_1, slot {class=SizeBoxSlot}}
  EntryNameInteractionButton {
   widgetClassPath=&Button
   parent=Overlay_1
   widget=EntryNameInteractionButton
   properties {renderOpacity=0}
   slot {class=OverlaySlot, properties {horizontalAlignment=HAlign_Fill, verticalAlignment=VAlign_Fill, padding=<ltrb -50 -50 -50 -50>}}
  }
  ItemText_SizeBox {
   widgetClassPath=&SizeBox
   parent=Overlay_1
   widget=ItemText_SizeBox
   properties {
    bOverride_MaxDesiredHeight=true
    bOverride_MaxDesiredWidth=true
    bOverride_MinDesiredWidth=true
    maxDesiredHeight=50
    maxDesiredWidth=200
    minDesiredHeight=30
    minDesiredWidth=200
   }
   slot {class=OverlaySlot}
  }
  Item_Text {
   widgetClassPath=&CommonTextBlock
   bIsVariable=true
   parent=ItemText_SizeBox
   widget=Item_Text
   properties {
    autoWrapText=true
    bIsScrollingEnabled=false
    lineHeightPercentage=0.95
    scrollOrientation=Orient_Vertical
    scrollStyle=/FortUILibrary/Styles/Typography/Foundation/TextScrollStyle-NoFade.TextScrollStyle-NoFade_C
    text=Placeholder
    textOverflowPolicy=MultilineEllipsis
    colorAndOpacity {colorUseRule=UseColor_Specified, specifiedColor=<rgba 1 1 1 0.7>}
    font {
     bForceMonospaced=false
     bMaterialIsStencil=false
     fontMaterial=null
     fontObject=&HeadingNow
     letterSpacing=10
     monospacedWidth=1
     size=15.75
     skewAmount=0
     typefaceFontName=74 Regular
     outlineSettings {
      bApplyOutlineToDropShadows=false
      bMiteredCorners=false
      bSeparateFillAlpha=false
      outlineMaterial=null
      outlineSize=0
      outlineColor=<rgba 0 0 0 1>
     }
    }
    shadowOffset=<0 0>
   }
   slot {class=SizeBoxSlot}
  }
  InputActionWidget {widgetClassPath=&CommonActionWidget, bInherited=true, widget=null}
 ]
}
