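# Radial emote wheel widget, derived from FortRadialMenu, that shows a category's equipped emotes for selection.
# PopulateWheelEntries rebuilds the dynamic entry box from the category view model, applying per-category display settings and binding click events that set the selected item and rebroadcast.
# Selection converts indices through GetMirroredIndex, then updates the background material's segment, indicator, selection view model and highlight animation.
# Compact non-tablet layouts shrink the wheel on PreConstruct.
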WidgetBlueprint /EmoteWheel/WBP_Emote_RadialWheel
animations [UpdateSegment {
  binding_count=1
  length_seconds=0.8
  object=&self:UpdateSegment
  sequence {bindings [BackgroundWithSegments {
     id=#1
     possessed_class=Image
     tracks [{
       type=MovieSceneWidgetMaterialTrack
       sections [{
         type=MovieSceneParameterSection
         end_frame=16
         start_frame=0
         channels [
          SubSegmentOpacity {type=MovieSceneScriptingFloatChannel, keys [[0,0.2],[3,1],[16,1]]}
          SelectedColorCenterR {type=MovieSceneScriptingFloatChannel, default=1}
          SelectedColorCenterG {type=MovieSceneScriptingFloatChannel, default=1}
          SelectedColorCenterB {type=MovieSceneScriptingFloatChannel, default=1}
          SelectedColorCenterA {type=MovieSceneScriptingFloatChannel, keys [[1,0.1],[5,0.5],[11,0.4],[16,0.4]]}
          SelectedColorOuterR {type=MovieSceneScriptingFloatChannel, default=1}
          SelectedColorOuterG {type=MovieSceneScriptingFloatChannel, default=1}
          SelectedColorOuterB {type=MovieSceneScriptingFloatChannel, default=1}
          SelectedColorOuterA {type=MovieSceneScriptingFloatChannel, keys [[1,0],[7,0.25],[15,0.1]]}
         ]
        }]
      }]
    }]}
 }]
graphs [
 BP_GetDesiredFocusTarget {nodes [
   BP_GetDesiredFocusTarget() {is=title K2Node_FunctionEntry}
   n1 = GetAllEntries!(self:Dynamic Entry Box Base Object Reference=Get EntryBox_ContextRadial() -> (EntryBox_ContextRadial:Dynamic Entry Box Object Reference)) -> (ReturnValue:Array of User Widget Object References)
   > Branch(Condition=Is Not Empty!(TargetArray:Array of User Widget Object References=@n1) -> (ReturnValue:bool))
   > Return Node!(
    execute
    ReturnValue:Widget Object Reference="Get (a copy)"(Array:Array of User Widget Object References=@n1) -> (Output:User Widget Object Reference)
   )
  ]}
 EmoteWheelEntryClicked {nodes [
   n1 = EmoteWheelEntryClicked!() -> (then, Item:Emote Wheel Item VM Object Reference) {is=title K2Node_FunctionEntry}
   > SetSelectedItem!(
    execute
    self:Emote Wheel Item Selection VM Object Reference=Get EmoteWheelItemSelectionVM() -> (EmoteWheelItemSelectionVM:Emote Wheel Item Selection VM Object Reference)
    Item:Emote Wheel Item VM Object Reference=@n1
   ) -> (then)
   > Call On Entry Clicked!(execute:Exec, self:WBP Emote Radial Wheel Object Reference, Item:Emote Wheel Item VM Object Reference=@n1) -> (then:Exec) {is=title K2Node_CallDelegate}
  ]}
 EventGraph {nodes [
   Event Destruct()
   > UnbindEmoteEntries()
   Event PreConstruct()
   > SetMobileStyle() {is=title K2Node_CallFunction}
   n1 = Event SelectItemAtIndex!() -> (OutputDelegate, then, Index:int=0) {is=title K2Node_Event}
   n2 = GetMirroredIndex!(Index:int=@n1.Index) -> (MirroredIndex:int=0)
   > Parent: SelectItemAtIndex!(execute, Index:int=@n2) -> (then) {is=title K2Node_CallParentFunction}
   n3 = Get EmoteWheelItemSelectionVM()
   > Branch(Condition="Not Equal (Integer)"(A=@n2, B=Get SelectedSlotIndex!(self:Emote Wheel Item Selection VM Object Reference=@n3) -> (SelectedSlotIndex:int=0)))
   > UpdateSelectedSegment!(execute, self, ActiveEntryIndex:int=@n2, Item:Emote Wheel Item VM Object Reference) -> (then)
  ]}
 GetMirroredIndex {nodes [
   n1 = GetMirroredIndex!() -> (then, Index:int) {is=title K2Node_FunctionEntry}
   > Branch(Condition="integer >= integer"(A=@n1))
   n2 = GetNumEntries!(self:Dynamic Entry Box Base Object Reference=Get EntryBox_ContextRadial()) -> (ReturnValue:int=0)
   > Return Node!(execute, MirroredIndex:int="% (Integer)"(A=int - int(A=@n2, B=@n1), B=@n2) -> (ReturnValue=0))
  ]}
 OnEntryClicked {nodes [OnEntryClicked!() -> (then, Item:Emote Wheel Item VM Object Reference) {is=title K2Node_FunctionEntry}]}
 PopulateWheelEntries {nodes [
   n1 = PopulateWheelEntries!() -> (then, Category:Emote Wheel Category VM Object Reference) {is=title K2Node_FunctionEntry}
   > Is Valid(InputObject=@n1)
   > UnbindEmoteEntries()
   > Reset!(execute, self:Dynamic Entry Box Object Reference=Get EntryBox_ContextRadial(), bDeleteWidgets:bool=true) -> (then)
   > n2 = For Each Loop(Array:Array of Emote Wheel Item VM Object References=GetEquippedItems!(self:Emote Wheel Category VM Object Reference=Get Category() -> (Category:Emote Wheel Category VM Object Reference)) -> (ReturnValue:Array of Emote Wheel Item VM Object References)) -> (Array Element:Emote Wheel Item VM Object Reference)
   RefreshRadialData(execute=@n2.Completed, self:Fort Radial Menu Object Reference)
   n3 = Get EmoteWheelItemSelectionVM()
   > SelectItemAtIndex!(
    execute
    self:Fort Radial Menu Object Reference
    Index:int=GetMirroredIndex(Index=Get SelectedSlotIndex(self=@n3) -> (SelectedSlotIndex=0)) -> (MirroredIndex=0) {is=title K2Node_CallFunction}
   ) -> (then)
   n4 = Create Entry of Class!(
    execute=@n2.LoopBody
    self:Dynamic Entry Box Object Reference=Get EntryBox_ContextRadial()
    EntryClass:User Widget Class Reference=&/WBP_Emote_RadialEntry.WBP_Emote_RadialEntry_C
   ) -> (then, ReturnValue:WBP Emote Radial Entry Object Reference)
   > SetEmoteWheelItemVM!(execute, self:WBP Emote Radial Entry Object Reference=@n4, Viewmodel:Emote Wheel Item VM Object Reference=@n2) -> (then)
   > SetEmoteIndex!(execute, self:WBP Emote Radial Entry Object Reference=@n4, ItemIndex:int=@n2.Array Index) -> (then)
   n5 = Get Category()
   > SetIconScale!(
    execute
    self:WBP Emote Radial Entry Object Reference=@n4
    ScaleFactor:double=Get TextureScaleFactor!(self:Emote Wheel Category Config Object Reference=GetConfig!(self:Emote Wheel Category VM Object Reference=@n5) -> (ReturnValue:Emote Wheel Category Config Object Reference)) -> (TextureScaleFactor:float)
   ) -> (then)
   n6 = Get EmoteWheelItemSelectionVM()
   > SetTruncationTitles!(
    execute
    self:WBP Emote Radial Entry Object Reference=@n4
    ShouldTruncateTitles:bool=ShouldTruncateTitles!(self:Emote Wheel Category VM Object Reference=Get SelectedCategory!(self:Emote Wheel Item Selection VM Object Reference=@n6) -> (SelectedCategory:Emote Wheel Category VM Object Reference)) -> (ReturnValue:bool)
   ) -> (then)
   n7 = Get Category()
   > SetAllowInteractionWithEntryText!(
    execute
    self:WBP Emote Radial Entry Object Reference=@n4
    bAllowInteraction:bool=Get bAllowInteractionWithItemNames!(self:Emote Wheel Category Config Object Reference=GetConfig(self=@n7)) -> (bAllowInteractionWithItemNames:bool)
   ) -> (then)
   > Bind Event to On Entry Clicked(self:WBP Emote Radial Entry Object Reference=@n4, Delegate=Create Event()) {is=title K2Node_AddDelegate}
  ]}
 RadialItemSelectionChanged {nodes [RadialItemSelectionChanged() {is=title K2Node_FunctionEntry}]}
 SetMobileStyle {nodes [
   SetMobileStyle() {is=title K2Node_FunctionEntry}
   > Branch(Condition=IsUsingCompactUI())
   > n1 = Branch(Condition=IsTabletAspectRatio!() -> (ReturnValue:bool))
   n2 = Get SizeBox_Wheel() -> (SizeBox_Wheel:Size Box Object Reference)
   SetWidthOverride(execute=@n1.else, self=@n2, InWidthOverride=536)
   > SetHeightOverride(self=@n2, InHeightOverride=536)
   > SetPadding!(
    execute
    self:Grid Slot Object Reference=SlotAsGridSlot(Widget=Get EntryBox_ContextRadial())
    InPadding_Left:float=-55
    InPadding_Top:float=-55
    InPadding_Right:float=-55
    InPadding_Bottom:float=-55
   ) -> (then)
  ]}
 ToggleCenterBlur {nodes [
   ToggleCenterBlur!() -> (then, ShowBlur:bool=~) {is=title K2Node_FunctionEntry}
   > GetDynamicMaterial(self=Get BackgroundWithSegments() -> (BackgroundWithSegments:Image Object Reference))
  ]}
 UnbindEmoteEntries {nodes [
   UnbindEmoteEntries() {is=title K2Node_FunctionEntry}
   > n1 = For Each Loop(Array:Array of User Widget Object References=GetAllEntries(self=Get EntryBox_ContextRadial())) -> (Array Element:User Widget Object Reference)
   > n2 = Cast To WBP_Emote_RadialEntry!(execute, Object=@n1) -> (then, CastFailed, AsWBP Emote Radial Entry:WBP Emote Radial Entry Object Reference) {is=title K2Node_DynamicCast}
   > Unbind Event from On Entry Clicked(self:WBP Emote Radial Entry Object Reference=@n2, Delegate=Create Event()) {is=title K2Node_RemoveDelegate}
  ]}
 UpdateSelectedSegment {nodes [
   UpdateSelectedSegment!() -> (then, ActiveEntryIndex:int, Item:Emote Wheel Item VM Object Reference) {is=title K2Node_FunctionEntry}
   > n1 = GetDynamicMaterial(self=Get BackgroundWithSegments())
   n2 = Get EntryBox_ContextRadial()
   n3 = "Clamp (Integer)"(Value=Get ActiveEntryIndex() -> (ActiveEntryIndex:int=0), Min=0, Max=int - int(A=GetNumEntries(self=@n2) -> (ReturnValue=0))) -> (ReturnValue=0)
   > SetScalarParameterValue(self=@n1, ParameterName=SelectedIndex, Value="To Float (Integer)"(InInt=@n3))
   > UpdateSelectedSegment!(
    execute
    self:WBP Emote Indicator Object Reference=Get WBP_EmoteIndicator() -> (WBP_EmoteIndicator:WBP Emote Indicator Object Reference)
    Segment:int=@n3
   ) -> (then) {is=title K2Node_CallFunction}
   > SetSelectedSlotIndex!(execute, self:Emote Wheel Item Selection VM Object Reference=Get EmoteWheelItemSelectionVM(), InSlotIndex:int=@n3) -> (then)
   n4 = Get UpdateSegment() -> (UpdateSegment:Widget Animation Object Reference)
   > Check and Stop Animation!(execute:Exec, Animation:Widget Animation Object Reference=@n4) -> (then:Exec) {is=title K2Node_MacroInstance}
   > PlayAnimation(InAnimation=@n4)
  ]}
 UpdateSlotIndicatorVisibility {nodes [
   n1 = UpdateSlotIndicatorVisibility!() -> (then, ShouldShow:bool=~) {is=title K2Node_FunctionEntry}
   > SetVisibility(self=Get WBP_EmoteIndicator(), InVisibility=Select(Option 0=Hidden, Option 1=SelfHitTestInvisible, Index=@n1))
  ]}
]
variables [
 OnEntryClicked {}
 RadialItemSelectionChanged {}
]
widgets {
 info {inheritedWidgetCount=1, parentClass=/Script/FortniteUI.FortRadialMenu, rootWidgetClass=&Overlay, widgetCount=8}
 widgets [
  Overlay_0 {widgetClassPath=&Overlay, widget=Overlay_0}
  SizeBox_Wheel {
   widgetClassPath=&SizeBox
   bIsVariable=true
   parent=Overlay_0
   widget=SizeBox_Wheel
   properties {bOverride_HeightOverride=true, bOverride_WidthOverride=true, heightOverride=628, widthOverride=628}
   slot {class=OverlaySlot, properties {horizontalAlignment=HAlign_Center, verticalAlignment=VAlign_Center}}
  }
  GridPanel_0 {
   widgetClassPath=&GridPanel
   parent=SizeBox_Wheel
   widget=GridPanel_0
   properties {columnFill [1], rowFill [1]}
   slot {class=SizeBoxSlot, properties {horizontalAlignment=HAlign_Center, verticalAlignment=VAlign_Center}}
  }
  WBP_EmoteWheel_PostBufferUpdate {
   widgetClassPath=&/WBP_EmoteWheel_PostBufferUpdate
   bIsVariable=true
   parent=GridPanel_0
   widget=WBP_EmoteWheel_PostBufferUpdate
   slot {class=GridSlot, properties {layer=-25}}
  }
  BackgroundWithSegments {
   widgetClassPath=&Image
   bIsVariable=true
   parent=GridPanel_0
   widget=BackgroundWithSegments
   properties {
    visibility=SelfHitTestInvisible
    brush {
     bIsDynamicallyLoaded=false
     drawAs=Image
     imageType=FullColor
     mirroring=NoMirror
     resourceName=None
     resourceObject=/EmoteWheel/MaterialInstances/MI_UI_EmoteWheel_Slots
     tiling=NoTile
     imageSize=<648 648>
     margin=<ltrb 0 0 0 0>
     outlineSettings {
      bUseBrushTransparency=false
      roundingType=HalfHeightRadius
      width=0
      color {colorUseRule=UseColor_Specified, specifiedColor=<rgba 0 0 0 0>}
      cornerRadii=<0 0 0 0>
     }
     tintColor {colorUseRule=UseColor_Specified, specifiedColor=<rgba 1 1 1 1>}
     uVRegion {bIsValid=false, max=<0 0>, min=<0 0>}
    }
   }
   slot {class=GridSlot, properties {horizontalAlignment=HAlign_Center, verticalAlignment=VAlign_Center}}
  }
  EntryBox_ContextRadial {
   widgetClassPath=&DynamicEntryBox
   bInherited=true
   bIsVariable=true
   parent=GridPanel_0
   widget=EntryBox_ContextRadial
   properties {
    entryBoxType=Radial
    entryWidgetClass=&WBP_Emote_RadialEntry_C
    navigation=&self:WidgetTree.EntryBox_ContextRadial.WidgetNavigation_0
    numDesignerPreviewEntries=8
    radialBoxSettings {angleBetweenItems=45, bClockwiseOrder=true, bDistributeItemsEvenly=true, marginSize=0, sectorCentralAngle=360, startingAngle=90}
   }
   slot {class=GridSlot, properties {layer=30, padding=<ltrb -57 -57 -57 -57>}}
  }
  DebugActivationTxt {
   widgetClassPath=&CommonTextBlock
   bIsVariable=true
   parent=GridPanel_0
   widget=DebugActivationTxt
   properties {
    style=&TextStyle-Base_C
    text=is active?
    visibility=Collapsed
    font {
     bForceMonospaced=false
     bMaterialIsStencil=false
     fontMaterial=null
     fontObject=&NotoSans
     letterSpacing=0
     monospacedWidth=1
     size=24
     skewAmount=0
     typefaceFontName=NotoSans-Regular
     outlineSettings {
      bApplyOutlineToDropShadows=false
      bMiteredCorners=false
      bSeparateFillAlpha=false
      outlineMaterial=null
      outlineSize=0
      outlineColor=<rgba 0 0 0 1>
     }
    }
    shadowOffset=<0 0>
   }
   slot {class=GridSlot}
  }
  WBP_EmoteIndicator {
   widgetClassPath=/FortUILibrary/UIKit/EmoteWheel/WBP_EmoteIndicator
   bIsVariable=true
   parent=GridPanel_0
   widget=WBP_EmoteIndicator
   properties {show Background=true}
   slot {class=GridSlot}
  }
 ]
}
