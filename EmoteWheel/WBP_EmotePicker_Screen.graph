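# Emote wheel picker screen that pairs a radial wheel with category tabs and an expandable locker collection.
# Tab selection updates the selection view model, plays a directional wheel transition and async-creates optional category overlays.
# One animation, played forward or reversed, swaps wheel and collection, with the collection widget created on demand.
# Previous/next input defers to an injected sub-tab interface when valid, and bottom-bar prompts adapt to input type.
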WidgetBlueprint /EmoteWheel/WBP_EmotePicker_Screen
animations [
 Anim_Collection_ExpandContract {
  binding_count=14
  length_seconds=0.15
  object=&self:Anim_Collection_ExpandContract
  sequence {bindings [
    Button_Arrow {
     id=#1
     possessed_class=WBP_UIKit_ViewMore_ArrowButton_C
     tracks [
      {
       type=MovieSceneFloatTrack
       display_name=Render Opacity
       sections [{
         type=MovieSceneFloatSection
         end_frame=3
         start_frame=0
         channels [None {type=MovieSceneScriptingFloatChannel, keys [[0,1],[1,0],[2,0],[3,1]]}]
        }]
      }
      {
       type=MovieScene2DTransformTrack
       sections [
        {
         type=MovieScene2DTransformSection
         end_frame=3
         start_frame=0
         channels [
          Translation.X {type=MovieSceneScriptingFloatChannel, const [0,1]}
          Translation.Y {type=MovieSceneScriptingFloatChannel, keys [[0,0],[3,-30]]}
          Angle {type=MovieSceneScriptingFloatChannel, keys [[0,0],[1,0],[2,180],[3,180]]}
          Scale.X {type=MovieSceneScriptingFloatChannel, const [1,1]}
          Scale.Y {type=MovieSceneScriptingFloatChannel, const [1,1]}
          Shear.X {type=MovieSceneScriptingFloatChannel, const [0,1]}
          Shear.Y {type=MovieSceneScriptingFloatChannel, const [0,1]}
         ]
        }
        {
         type=MovieScene2DTransformSection
         end_frame=3
         start_frame=0
         channels [
          Translation.X {type=MovieSceneScriptingFloatChannel}
          Translation.Y {type=MovieSceneScriptingFloatChannel, keys [[0,-10],[3,-40]]}
          Angle {type=MovieSceneScriptingFloatChannel}
          Scale.X {type=MovieSceneScriptingFloatChannel}
          Scale.Y {type=MovieSceneScriptingFloatChannel}
          Shear.X {type=MovieSceneScriptingFloatChannel}
          Shear.Y {type=MovieSceneScriptingFloatChannel}
         ]
        }
       ]
      }
     ]
    }
    GridSlot_0 {
     id=#2
     parent_id=#3
     possessed_class=GridSlot
     tracks [{
       type=MovieSceneDoubleVectorTrack
       display_name=Nudge
       sections [{
         type=MovieSceneDoubleVectorSection
         end_frame=3
         start_frame=0
         channels [
          Vector.X {type=MovieSceneScriptingDoubleChannel, const [0,2]}
          Vector.Y {type=MovieSceneScriptingDoubleChannel, keys [[0,0],[3,-900]]}
         ]
        }]
      }]
    }
    WBP_Emote_RadialWheel {
     id=#4
     possessed_class=WBP_Emote_RadialWheel_C
     tracks [
      {
       type=MovieSceneFloatTrack
       display_name=Render Opacity
       sections [{type=MovieSceneFloatSection, end_frame=3, start_frame=0, channels [None {type=MovieSceneScriptingFloatChannel, keys [[0.3333,1],[3,0]]}]}]
      }
      {
       type=MovieSceneByteTrack
       display_name=Visibility
       sections [{type=MovieSceneByteSection, end_frame=3, start_frame=0, channels [None {type=MovieSceneScriptingByteChannel, keys [[0,4],[2.6667,4],[3,1]]}]}]
      }
     ]
    }
    Background_BlockClick_BlurOnlyHighEndDevice {
     id=#5
     possessed_class=Image
     tracks [{
       type=MovieSceneWidgetMaterialTrack
       sections [{
         type=MovieSceneParameterSection
         end_frame=3
         start_frame=0
         channels [Center Opacity {type=MovieSceneScriptingFloatChannel, keys [[0,0.5],[3,1]]}]
        }]
      }]
    }
    GridSlot_19 {
     id=#6
     parent_id=#4
     possessed_class=GridSlot
     tracks [{
       type=MovieSceneDoubleVectorTrack
       display_name=Nudge
       sections [{
         type=MovieSceneDoubleVectorSection
         end_frame=3
         start_frame=0
         channels [
          Vector.X {type=MovieSceneScriptingDoubleChannel, const [0,1]}
          Vector.Y {type=MovieSceneScriptingDoubleChannel, keys [[0,0],[3,-900]]}
         ]
        }]
      }]
    }
    OverlaySlot_2 {id=#7, parent_id=#1, possessed_class=OverlaySlot, tracks []}
    TabGroup_VBox {id=#3, possessed_class=VerticalBox, tracks []}
    WBP_EmotePicker_Screen_C_105 {
     id=#8
     possessed_class=WBP_EmotePicker_Screen_C
     tracks [{
       type=MovieSceneDoubleTrack
       display_name=Lerp Top to Bottom Of Screen Height
       sections [{type=MovieSceneDoubleSection, end_frame=3, start_frame=0, channels [None {type=MovieSceneScriptingDoubleChannel, keys [[0,0],[3,1]]}]}]
      }]
    }
    GridSlot_0 {id=#9, parent_id=#10, possessed_class=GridSlot, tracks []}
    CollectionButton {
     id=#11
     possessed_class=Overlay
     tracks [{
       type=MovieSceneFloatTrack
       display_name=Render Opacity
       sections [{
         type=MovieSceneFloatSection
         end_frame=3
         start_frame=0
         channels [None {type=MovieSceneScriptingFloatChannel, keys [[0,1],[1,0],[2,0],[3,1]]}]
        }]
      }]
    }
    Overlay_LockerCollection {
     id=#12
     possessed_class=Overlay
     tracks [
      {
       type=MovieScene2DTransformTrack
       sections [{
         type=MovieScene2DTransformSection
         end_frame=3
         start_frame=0
         channels [
          Translation.X {type=MovieSceneScriptingFloatChannel, const [0,1]}
          Translation.Y {type=MovieSceneScriptingFloatChannel, keys [[0,1080],[3,0]]}
          Angle {type=MovieSceneScriptingFloatChannel, const [0,1]}
          Scale.X {type=MovieSceneScriptingFloatChannel, const [1,1]}
          Scale.Y {type=MovieSceneScriptingFloatChannel, const [1,1]}
          Shear.X {type=MovieSceneScriptingFloatChannel, const [0,1]}
          Shear.Y {type=MovieSceneScriptingFloatChannel, const [0,1]}
         ]
        }]
      }
      {
       type=MovieSceneFloatTrack
       display_name=Render Opacity
       sections [{type=MovieSceneFloatSection, end_frame=3, start_frame=0, channels [None {type=MovieSceneScriptingFloatChannel, keys [[0,0],[3,1]]}]}]
      }
     ]
    }
    SizeBoxSlot_0 {
     id=#13
     parent_id=#11
     possessed_class=SizeBoxSlot
     tracks [{
       type=MovieSceneMarginTrack
       display_name=Padding
       sections [{
         type=MovieSceneMarginSection
         end_frame=3
         start_frame=0
         channels [
          Left {type=MovieSceneScriptingFloatChannel, const [0,1]}
          Top {type=MovieSceneScriptingFloatChannel, keys [[0,-130],[3,0]]}
          Right {type=MovieSceneScriptingFloatChannel, const [0,1]}
          Bottom {type=MovieSceneScriptingFloatChannel, const [0,1]}
         ]
        }]
      }]
    }
    Button_ViewCollection {
     id=#14
     possessed_class=WBP_UIKit_Button_Generic_C
     tracks [{
       type=MovieScene2DTransformTrack
       sections [{
         type=MovieScene2DTransformSection
         end_frame=3
         start_frame=0
         channels [
          Translation.X {type=MovieSceneScriptingFloatChannel, const [0,2]}
          Translation.Y {type=MovieSceneScriptingFloatChannel, keys [[0,0],[3,56]]}
          Angle {type=MovieSceneScriptingFloatChannel, const [0,1]}
          Scale.X {type=MovieSceneScriptingFloatChannel, const [1,1]}
          Scale.Y {type=MovieSceneScriptingFloatChannel, const [1,1]}
          Shear.X {type=MovieSceneScriptingFloatChannel, const [0,1]}
          Shear.Y {type=MovieSceneScriptingFloatChannel, const [0,1]}
         ]
        }]
      }]
    }
    TabGroup_EmoteCategories {
     id=#10
     possessed_class=WBP_UIKit_TabGroup_C
     tracks [{
       type=MovieSceneByteTrack
       display_name=Visibility
       sections [{type=MovieSceneByteSection, end_frame=3, start_frame=0, channels [None {type=MovieSceneScriptingByteChannel, keys [[0,4],[2,4],[3,1]]}]}]
      }]
    }
   ]}
 }
 Intro {binding_count=0, length_seconds=0.25, object=&self:Intro, sequence {bindings []}}
 Outro {
  binding_count=2
  length_seconds=0.05
  object=&self:Outro
  sequence {bindings [
    BG_Darkener {
     id=#15
     possessed_class=Image
     tracks [{
       type=MovieSceneFloatTrack
       display_name=Render Opacity
       sections [{type=MovieSceneFloatSection, end_frame=1, start_frame=0, channels [None {type=MovieSceneScriptingFloatChannel, keys [[0,1],[1,0]]}]}]
      }]
    }
    Background_BlockClick_BlurOnlyHighEndDevice {
     id=#16
     possessed_class=Image
     tracks [{
       type=MovieSceneFloatTrack
       display_name=Render Opacity
       sections [{type=MovieSceneFloatSection, end_frame=1, start_frame=0, channels [None {type=MovieSceneScriptingFloatChannel, keys [[0,1],[1,0]]}]}]
      }]
    }
   ]}
 }
 BlurFade {
  binding_count=2
  length_seconds=1.15
  object=&self:BlurFade
  sequence {bindings [
    BG_Darkener {
     id=#15
     possessed_class=Image
     tracks [{
       type=MovieSceneFloatTrack
       display_name=Render Opacity
       sections [{type=MovieSceneFloatSection, end_frame=23, start_frame=0, channels [None {type=MovieSceneScriptingFloatChannel, keys [[8,0],[23,1]]}]}]
      }]
    }
    Background_BlockClick_BlurOnlyHighEndDevice {
     id=#16
     possessed_class=Image
     tracks [{
       type=MovieSceneFloatTrack
       display_name=Render Opacity
       sections [{type=MovieSceneFloatSection, end_frame=23, start_frame=0, channels [None {type=MovieSceneScriptingFloatChannel, keys [[8,0],[23,1]]}]}]
      }]
    }
   ]}
 }
 TabNext {
  binding_count=1
  length_seconds=0.15
  object=&self:TabNext
  sequence {bindings [WBP_Emote_RadialWheel {
     id=#17
     possessed_class=WBP_Emote_RadialWheel_C
     tracks [
      {
       type=MovieScene2DTransformTrack
       sections [{
         type=MovieScene2DTransformSection
         end_frame=3
         start_frame=0
         channels [
          Translation.X {type=MovieSceneScriptingFloatChannel, keys [[0,0],[1,30],[3,0]]}
          Translation.Y {type=MovieSceneScriptingFloatChannel, default=0}
          Angle {type=MovieSceneScriptingFloatChannel}
          Scale.X {type=MovieSceneScriptingFloatChannel}
          Scale.Y {type=MovieSceneScriptingFloatChannel}
          Shear.X {type=MovieSceneScriptingFloatChannel}
          Shear.Y {type=MovieSceneScriptingFloatChannel}
         ]
        }]
      }
      {
       type=MovieSceneFloatTrack
       display_name=Render Opacity
       sections [{type=MovieSceneFloatSection, end_frame=3, start_frame=0, channels [None {type=MovieSceneScriptingFloatChannel, keys [[0,0],[1,0],[3,1]]}]}]
      }
     ]
    }]}
 }
 TabLast {
  binding_count=1
  length_seconds=0.15
  object=&self:TabLast
  sequence {bindings [WBP_Emote_RadialWheel {
     id=#18
     possessed_class=WBP_Emote_RadialWheel_C
     tracks [
      {
       type=MovieScene2DTransformTrack
       sections [{
         type=MovieScene2DTransformSection
         end_frame=3
         start_frame=0
         channels [
          Translation.X {type=MovieSceneScriptingFloatChannel, keys [[0,0],[1,-30],[3,0]]}
          Translation.Y {type=MovieSceneScriptingFloatChannel, default=0}
          Angle {type=MovieSceneScriptingFloatChannel}
          Scale.X {type=MovieSceneScriptingFloatChannel}
          Scale.Y {type=MovieSceneScriptingFloatChannel}
          Shear.X {type=MovieSceneScriptingFloatChannel}
          Shear.Y {type=MovieSceneScriptingFloatChannel}
         ]
        }]
      }
      {
       type=MovieSceneFloatTrack
       display_name=Render Opacity
       sections [{type=MovieSceneFloatSection, end_frame=3, start_frame=0, channels [None {type=MovieSceneScriptingFloatChannel, keys [[0,0],[1,0],[3,1]]}]}]
      }
     ]
    }]}
 }
]
graphs [
 BP_GetDesiredFocusTarget {nodes [
   Get Desired Focus Target()
   > Return Node!(execute, ReturnValue:Widget Object Reference=Get WBP_Emote_RadialWheel() -> (WBP_Emote_RadialWheel:WBP Emote Radial Wheel Object Reference))
  ]}
 BP_GetDesiredInputConfig {nodes [
   Get Desired Input Config() {is=title K2Node_FunctionEntry}
   > Return Node(ReturnValue:UIInput Config Structure=Make UIInput Config!(
     bIgnoreMoveInput:bool=False
     bIgnoreLookInput:bool=true
     InputMode:ECommonInputMode Enum=All
     MouseCaptureMode:EMouseCaptureMode Enum=CaptureDuringRightMouseDown
     MouseLockMode:EMouseLockMode Enum=DoNotLock
     bHideCursorDuringViewportCapture:bool=True
    ) -> (UIInputConfig:UIInput Config Structure) {is=title K2Node_MakeStruct})
  ]}
 CenterMouseToViewport {nodes [
   CenterMouseToViewport() {is=title K2Node_FunctionEntry}
   n1 = vector2d * vector2d!(A:Vector 2D Structure=GetViewportSize!() -> (ReturnValue:Vector 2D Structure), B:Vector 2D Structure=(X=0.5,Y=0.5)) -> (ReturnValue_X:double, ReturnValue_Y:double) {is=title K2Node_PromotableOperator}
   > Set Mouse Position!(execute, self:Player Controller Object Reference=GetOwningPlayer(), X:int=Truncate(A=@n1), Y:int=Truncate(A=@n1.ReturnValue_Y)) -> (then) {is=title K2Node_CallFunction}
  ]}
 EventGraph {nodes [
   n1 = "On Clicked (WBP_BottomBarDecoy_Close)"!() -> (OutputDelegate, then, Button:Common Button Base Object Reference) {is=title K2Node_ComponentBoundEvent}
   > n2 = TriggerFlowriderInteraction!(execute, self, Element:Widget Object Reference=@n1.Button, Value:String=Close Wheel) -> (then)
   n3 = Get!(execute:Exec=[@n2.then, @n63.then]) -> (then:Exec, else:Exec, LockerCollectionWidget:Common Activatable Widget Object Reference)
   > DeactivateWidget(self=@n3)
   > n4 = UnLoadCollection()
   DeactivateWidget(execute=[@n4.then, @n3.else])
   n5 = "On Clicked (Button_Arrow)"!() -> (OutputDelegate, then, Button:Common Button Base Object Reference) {is=title K2Node_ComponentBoundEvent}
   > n6 = Set ClickedCollectionButton(ClickedCollectionButton:Common Button Base Object Reference=@n5.Button) -> (Output_Get:Common Button Base Object Reference)
   n7 = Branch(execute=[@n20.then, @n6.then], Condition=Get IsCollectionExpanded() -> (IsCollectionExpanded:bool))
   > OnCollectionContract!(execute, self, Reset:bool=false) -> (then)
   n8 = Get EmoteWheelItemSelectionVM() -> (EmoteWheelItemSelectionVM:Emote Wheel Item Selection VM Object Reference)
   > n9 = SetIsCollectionShown!(execute, self:Emote Wheel Item Selection VM Object Reference=@n8, bInIsCollectionShown:bool=false) -> (then)
   TriggerFlowriderInteraction(
    execute=[@n10.then, @n9.then]
    Element=Get ClickedCollectionButton() -> (ClickedCollectionButton:Common Button Base Object Reference)
    Value=Select(Option 0:String=Open, Option 1:String=Close, Index=Get IsCollectionExpanded()) -> (ReturnValue:String)
   )
   OnCollectionExpand!(execute=@n7.else, self, SkipAnim:bool=false) -> (then)
   > n10 = SetIsCollectionShown(self=@n8, bInIsCollectionShown=true)
   n11 = "On Clicked (WBP_BottomBarDecoy_GamepadUse)"!() -> (OutputDelegate, then, Button:Common Button Base Object Reference) {is=title K2Node_ComponentBoundEvent}
   > n12 = Set ClickedUseButton(ClickedUseButton:Common Button Base Object Reference=@n11.Button) -> (Output_Get:Common Button Base Object Reference)
   UseItem(execute=[@n27.then, @n42.then, @n12.then, @n30.then])
   > TriggerFlowriderInteraction(Element=Get ClickedUseButton() -> (ClickedUseButton:Common Button Base Object Reference), Value=Use Item)
   n13 = "On Clicked (PreviousArrow_Button)"!() -> (OutputDelegate, then, Button:Common Button Base Object Reference) {is=title K2Node_ComponentBoundEvent}
   > GoToPrevious!(execute, self, InteractionWidget:Widget Object Reference=@n13.Button) -> (then)
   n14 = "On Clicked (NextArrow_Button)"!() -> (OutputDelegate, then, Button:Common Button Base Object Reference) {is=title K2Node_ComponentBoundEvent}
   > GoToNext!(execute, self, InteractionWidget:Widget Object Reference=@n14.Button) -> (then)
   n15 = "On Widget Injected (SubTab_WidgetInjectionSlot)"!() -> (OutputDelegate, then, Widget:Widget Object Reference) {is=title K2Node_ComponentBoundEvent}
   > n16 = Cast To BPI_EW_SubTabInterface!(execute, Object=@n15.Widget) -> (then, CastFailed, AsBPI EW Sub Tab Interface:BPI EW Sub Tab Interface Interface) {is=title K2Node_DynamicCast}
   > n17 = Set SubTabInterface(SubTabInterface:BPI EW Sub Tab Interface Interface=@n16) -> (Output_Get:BPI EW Sub Tab Interface Interface)
   > SetCategoryTopBarRef!(
    execute
    self:BPI EW Sub Tab Interface Interface=@n17
    TopBarRef:Object Reference=Get TabGroup_EmoteCategories() -> (TabGroup_EmoteCategories:WBP UIKit Tab Group Object Reference)
   ) -> (then)
   PrintString!(
    execute=@n16.CastFailed
    InString=WBP_EmotePicker_Screen BPI_EW_SubtabInterface Cast Failed
    bPrintToScreen
    bPrintToLog
    TextColor=(R=1,G=0.0177,B=0.0221,A=1)
    Duration=5
    Key
   ) -> (then)
   "On Swipe Left (TouchSwipePanel)"() {is=title K2Node_ComponentBoundEvent}
   > GoToPrevious(InteractionWidget=Get TouchSwipePanel() -> (TouchSwipePanel:Fort Swipe Panel Object Reference))
   "On Swipe Right (TouchSwipePanel)"() {is=title K2Node_ComponentBoundEvent}
   > GoToNext(InteractionWidget=Get TouchSwipePanel())
   n18 = "On Clicked (ViewCollectionButton)"() {is=title K2Node_ComponentBoundEvent}
   n19 = "On Clicked (Button_ViewCollection)"!() -> (OutputDelegate, then, Button:Common Button Base Object Reference) {is=title K2Node_ComponentBoundEvent}
   n20 = Set ClickedCollectionButton(execute=[@n18.then, @n19.then], ClickedCollectionButton=@n19.Button)
   "On Hovered (ViewCollectionButton)"() {is=title K2Node_ComponentBoundEvent}
   n21 = Get CommonInputSubsystem()
   > n22 = Branch(Condition="Not Equal (Enum)"!(A:ECommonInputType Enum=GetCurrentInputType(self=@n21), B:ECommonInputType Enum=Touch) -> (ReturnValue:bool))
   StopAllAnimations(
    execute=[@n28.then, @n22.then]
    self:User Widget Object Reference=Get Button_ViewCollection() -> (Button_ViewCollection:WBP UIKit Button Generic Object Reference)
   )
   n23 = Get Button_Arrow() -> (Button_Arrow:WBP UIKit View More Arrow Button Object Reference)
   > PlayPulseAnimation(self:WBP UIKit View More Arrow Button Object Reference=@n23)
   > PlayAnimationForward(
    self=@n23
    InAnimation=Get Hovered!(self:WBP UIKit View More Arrow Button Object Reference=@n23) -> (Hovered:Widget Animation Object Reference)
   )
   "On Unhovered (ViewCollectionButton)"() {is=title K2Node_ComponentBoundEvent}
   n24 = Get CommonInputSubsystem()
   > n25 = Branch(Condition="Not Equal (Enum)"(A=GetCurrentInputType(self=@n24), B=Touch))
   StopAllAnimations(execute=[@n29.then, @n25.then], self=Get Button_ViewCollection())
   > PlayAnimationForward(InAnimation=Get Unhovered!(self:WBP UIKit View More Arrow Button Object Reference=Get Button_Arrow()) -> (Unhovered:Widget Animation Object Reference))
   n26 = "On Clicked (WBP_BottomBarDecoy_UseRelease)"!() -> (OutputDelegate, then, Button:Common Button Base Object Reference) {is=title K2Node_ComponentBoundEvent}
   > n27 = Set ClickedUseButton(ClickedUseButton=@n26.Button)
   n28 = "On Hovered (Button_ViewCollection)"!() -> (OutputDelegate, then, Button:Common Button Base Object Reference) {is=title K2Node_ComponentBoundEvent}
   n29 = "On Unhovered (Button_ViewCollection)"!() -> (OutputDelegate, then, Button:Common Button Base Object Reference) {is=title K2Node_ComponentBoundEvent}
   n30 = "On Entry Clicked (WBP_Emote_RadialWheel)"!() -> (OutputDelegate, then, Item:Emote Wheel Item VM Object Reference) {is=title K2Node_ComponentBoundEvent}
   n31 = "On UIKit Tab Selected (TabGroup_EmoteCategories)"!() -> (OutputDelegate, then, TabID:Name=None, TabVM:MVVM Base Viewmodel Object Reference) {is=title K2Node_ComponentBoundEvent}
   > n32 = Set SelectedTabID(SelectedTabID:Name=@n31.TabID) -> (Output_Get:Name=None)
   > UpdateFlowriderSubView!(execute, self, SubViewString:String="To String (Name)"(InName=@n32)) -> (then)
   > HandleCategoryChanged!(execute, self, VM:MVVM Base Viewmodel Object Reference=@n31.TabVM) -> (then)
   > TriggerTabTransition!(execute, self, Index:int=GetScaledTabTransitionIndex!() -> (ReturnValue:int=0)) -> (then)
   n33 = Get SelectedCategory!(self:Emote Wheel Item Selection VM Object Reference=Get EmoteWheelItemSelectionVM()) -> (SelectedCategory:Emote Wheel Category VM Object Reference)
   > UpdateSlotIndicatorVisibility!(execute, self, CategoryDisplayName:Text=GetDisplayName!(self:Emote Wheel Category VM Object Reference=@n33) -> (ReturnValue:Text)) -> (then)
   n34 = GetConfig!(self:Emote Wheel Category VM Object Reference=@n33) -> (ReturnValue:Emote Wheel Category Config Object Reference)
   > UpdateCollectionButtonVisibility!(execute, self, ShouldShow:bool=Get bCanOpenCollection!(self:Emote Wheel Category Config Object Reference=@n34) -> (bCanOpenCollection:bool)) -> (then)
   > ClearChildren(self=Get OverlayPanel() -> (OverlayPanel:Overlay Object Reference))
   n35 = Get Overlay!(self:Emote Wheel Category Config Object Reference=@n34) -> (Overlay:User Widget Soft Class Reference)
   > n36 = Branch(Condition=IsValidSoftClassReference!(SoftClassReference:"Object Soft Class Reference (by ref)"=@n35) -> (ReturnValue:bool))
   n37 = GetOwningPlayer()
   > n38 = CreateWidgetAsync!(
    execute:Exec
    UserWidgetSoftClass:User Widget Soft Class Reference=@n35
    OwningPlayer:Player Controller Object Reference=@n37
    bSuspendInputUntilComplete:bool=true
    None:%=@n37
   ) -> (then:Exec, OnComplete:Exec, OnFailure:Exec, UserWidget:User Widget Object Reference) {is=title K2Node_AsyncAction}
   n39 = Set CurrentOverlayWidget(execute=@n38.OnComplete, CurrentOverlayWidget:User Widget Object Reference=@n38) -> (Output_Get:User Widget Object Reference)
   > AddChild(self=Get OverlayPanel(), Content=@n39)
   n40 = SlotAsOverlaySlot(Widget=@n39)
   > SetVerticalAlignment(self=@n40, InVerticalAlignment=VAlign_Fill)
   > SetHorizontalAlignment(self=@n40, InHorizontalAlignment=HAlign_Fill)
   > SetVisibility(self=Get OverlayPanel(), InVisibility=SelfHitTestInvisible)
   SetVisibility(execute=@n36.else, self=Get OverlayPanel(), InVisibility=Collapsed)
   n41 = "On Clicked (WBP_BottomBarDecoy_Use)"!() -> (OutputDelegate, then, Button:Common Button Base Object Reference) {is=title K2Node_ComponentBoundEvent}
   > n42 = Set ClickedUseButton(ClickedUseButton=@n41.Button)
   n43 = "On Clicked (WBP_BottomBarDecoy_CloseEsc)"!() -> (OutputDelegate, then, Button:Common Button Base Object Reference) {is=title K2Node_ComponentBoundEvent}
   > TriggerFlowriderInteraction(Element=@n43.Button, Value=Close Wheel)
   > CloseMenu()
   n44 = OnCollectionExpand!() -> (OutputDelegate, then, SkipAnim:bool=~) {is=title K2Node_CustomEvent}
   > n45 = Set IsCollectionExpanded(IsCollectionExpanded:bool=true) -> (Output_Get:bool)
   > SetCollectionButtonText!(execute, self, IsCollectionExpanded:bool=@n45) -> (then)
   > UpdateTouchNavButtons!(
    execute
    self
    Categories:Array of Emote Wheel Category VM Object References=Get TabCategoriesArray() -> (TabCategoriesArray:Array of Emote Wheel Category VM Object References)
   ) -> (then)
   > SetVisibility(
    self=Get SubTab_WidgetInjectionSlot() -> (SubTab_WidgetInjectionSlot:Fort Plugin Widget Injection Slot Object Reference)
    InVisibility=HitTestInvisible
   )
   > n46 = Branch(Condition=@n44.SkipAnim)
   n47 = Get Anim_Collection_ExpandContract() -> (Anim_Collection_ExpandContract:Widget Animation Object Reference)
   > PlayAnimationTimeRange!(
    execute
    self:User Widget Object Reference
    InAnimation:Widget Animation Object Reference=@n47
    StartAtTime:float=GetEndTime(self=@n47)
    EndAtTime:float=0
    NumLoopsToPlay:int=1
    PlayMode:EUMGSequencePlayMode Enum
    PlaybackSpeed:float=1
    bRestoreState:bool=false
   ) -> (then, ReturnValue:Widget Animation Handle Structure)
   n48 = Get Anim_Collection_ExpandContract()
   PlayAnimationForward(execute=@n46.else, InAnimation=@n48)
   n49 = OnCollectionContract!() -> (OutputDelegate, then, Reset:bool=~) {is=title K2Node_CustomEvent}
   > n50 = Set IsCollectionExpanded(IsCollectionExpanded=false)
   > SetCollectionButtonText(IsCollectionExpanded=@n50)
   > UpdateTouchNavButtons(Categories=Get TabCategoriesArray())
   > SetVisibility(self=Get SubTab_WidgetInjectionSlot(), InVisibility=Visible)
   > n51 = Branch(Condition=NOT Boolean(A=@n49.Reset))
   > PlayAnimationReverse(InAnimation=@n48)
   n52 = Get Anim_Collection_ExpandContract()
   PlayAnimation(execute=@n51.else, InAnimation=@n52, StartAtTime=GetEndTime(self=@n52), PlayMode=Reverse)
   OnLoadUnloadCollection() {is=title K2Node_CustomEvent}
   > n53 = Branch(Condition=Get IsCollectionExpanded())
   > AsyncLoadCollection()
   > ActivateWidget(self=Get LockerCollectionWidget() -> (LockerCollectionWidget:Common Activatable Widget Object Reference))
   UnLoadCollection(execute=@n53.else)
   > n54 = Cast To WBP_EW_Collection!(execute, Object=Get LockerCollectionWidget()) -> (then, CastFailed, AsWBP EW Collection:WBP EW Collection Object Reference) {is=title K2Node_DynamicCast}
   > Unbind Event from On Close Menu(self:WBP EW Collection Object Reference=@n54, Delegate=Create Event()) {is=title K2Node_RemoveDelegate}
   > DeactivateWidget(self=Get LockerCollectionWidget())
   > SetBottomButtonVisibility!(
    execute
    self:WBP Bottom Bar Decoy Button Object Reference=Get WBP_BottomBarDecoy_GamepadUse() -> (WBP_BottomBarDecoy_GamepadUse:WBP Bottom Bar Decoy Button Object Reference)
    ShouldButtonBeVisible:bool=true
   ) -> (then)
   AsyncLoadCollection() {is=title K2Node_CustomEvent}
   > Set IsMenuSticky(IsMenuSticky:bool=true) -> (Output_Get:bool)
   > HandleInputChanged!(execute, self, bNewInputType:ECommonInputType Enum=GetCurrentInputType(self=Get CommonInputSubsystem())) -> (then)
   > RemoveWidget!(execute, Widget:Widget Object Reference=Get LockerCollectionWidget()) -> (then)
   > n55 = Create Widget!(
    execute:Exec
    Class:User Widget Class Reference=Get CollectionClass() -> (CollectionClass:Common Activatable Widget Class Reference)
    OwningPlayer:Player Controller Object Reference=GetOwningPlayer()
   ) -> (then:Exec, ReturnValue:Common Activatable Widget Object Reference)
   > n56 = Cast To WBP_EW_Collection(Object=@n55)
   > SetEmoteWheelVM!(
    execute
    self:WBP EW Collection Object Reference=@n56
    Viewmodel:Emote Wheel VM Object Reference=Get EmoteWheelVM() -> (EmoteWheelVM:Emote Wheel VM Object Reference)
   ) -> (then)
   > Bind Event to On Close Menu(self:WBP EW Collection Object Reference=@n56, Delegate=Create Event()) {is=title K2Node_AddDelegate}
   > n57 = Set LockerCollectionWidget(LockerCollectionWidget:Common Activatable Widget Object Reference=@n56) -> (Output_Get:Common Activatable Widget Object Reference)
   > Branch(Condition=IsValid(Object=@n57))
   > AddChildToOverlay!(
    execute
    self:Overlay Object Reference=Get Overlay_LockerCollection() -> (Overlay_LockerCollection:Overlay Object Reference)
    Content:Widget Object Reference=@n57
   ) -> (then, ReturnValue:Overlay Slot Object Reference)
   n58 = SlotAsOverlaySlot(Widget=@n57)
   > SetVerticalAlignment(self=@n58, InVerticalAlignment=VAlign_Fill)
   > SetHorizontalAlignment(self=@n58, InHorizontalAlignment=HAlign_Fill)
   > SetVisibility(self=Get WBP_UIKit_Throbber() -> (WBP_UIKit_Throbber:WBP UIKit Throbber Object Reference), InVisibility=Collapsed)
   n59 = Get EmoteWheelItemSelectionVM()
   > UpdateCollectionDetailsVisibility!(
    execute
    self
    IsCollectionShown:bool=Get bIsCollectionShown!(self:Emote Wheel Item Selection VM Object Reference=@n59) -> (bIsCollectionShown:bool)
    CategoryDisplayName:Text=GetDisplayName(self=Get SelectedCategory(self=@n59))
   ) -> (then)
   > SetBottomButtonVisibility(self=Get WBP_BottomBarDecoy_GamepadUse(), ShouldButtonBeVisible=false)
   > SetVisibility(
    self=[
     Get BottomButtonLeft_Divider() -> (BottomButtonLeft_Divider:Image Object Reference)
     Get BottomBarLeft_UseRelease() -> (BottomBarLeft_UseRelease:WBP UIKit Button Generic Object Reference)
    ]
    InVisibility=Collapsed
   )
   n60 = TriggerTabTransition!() -> (OutputDelegate, then, Index:int) {is=title K2Node_CustomEvent}
   > n61 = Set CurrentTabIndex(CurrentTabIndex:int=@n60.Index) -> (Output_Get:int=0)
   n62 = Compare Int!(
    Exec:Exec=[@n61.then, @n78.then]
    Input:int=Get CurrentTabIndex() -> (CurrentTabIndex:int=0)
    CompareWith:int=Get LastTabIndex() -> (LastTabIndex:int=0)
   ) -> (">":Exec, "==":Exec, "<":Exec) {is=title K2Node_MacroInstance}
   Check and Stop Animation!(execute:Exec=@n62."<", Animation:Widget Animation Object Reference=Get TabNext() -> (TabNext:Widget Animation Object Reference)) -> (then:Exec) {is=title K2Node_MacroInstance}
   > Check and Stop Animation(Animation=Get TabLast() -> (TabLast:Widget Animation Object Reference))
   > PlayAnimationForward(InAnimation=Get TabLast())
   > Set LastTabIndex(LastTabIndex:int=Get CurrentTabIndex() -> (CurrentTabIndex=0)) -> (Output_Get:int=0)
   Check and Stop Animation(execute=@n62.">", Animation=Get TabNext())
   > Check and Stop Animation(Animation=Get TabLast())
   > PlayAnimationForward(InAnimation=Get TabNext())
   > Set LastTabIndex(LastTabIndex=Get CurrentTabIndex() -> (CurrentTabIndex=0)) -> (Output_Get=0)
   n63 = CloseMenu() {is=title K2Node_CustomEvent}
   n64 = Event PreConstruct()
   > n65 = Sequence()
   > Branch(Condition=@n64.IsDesignTime)
   > SetRenderOpacity(
    self=Get Background_BlockClick_BlurOnlyHighEndDevice() -> (Background_BlockClick_BlurOnlyHighEndDevice:Image Object Reference)
    InOpacity=0.3
   )
   SetCollectionButtonText(execute=@n65.then_1, IsCollectionExpanded=false)
   > SetIsInteractionEnabled!(
    execute
    self:Common Button Base Object Reference=[Get BottomBarLeft_UseRelease(), Get BottomBarLeft_QuickUse() -> (BottomBarLeft_QuickUse:WBP UIKit Button Generic Object Reference)]
    bInIsInteractionEnabled:bool=false
   ) -> (then)
   > SetMobileStyle() {is=title K2Node_CallFunction}
   Event Construct()
   > UpdateSubTabVisibility()
   n66 = GetAnalyticsEWActivationCount!(self:Fort Picker Context Object Reference=Get FortPickerContext() -> (ReturnValue:Fort Picker Context Object Reference) {is=title K2Node_GetSubsystem}) -> (ReturnValue:int=0)
   > Fire Event Main View Opened!(
    execute
    View:Widget Object Reference=Self-Reference()
    AdditionalInfo:Flow Analytics Additional Info Structure="Add String (Non-Localized)"!(
     Info:"Flow Analytics Additional Info Structure (by ref)"=Make Flow Analytics Additional Info!() -> (FlowAnalyticsAdditionalInfo:Flow Analytics Additional Info Structure) {is=title K2Node_MakeStruct}
     Key:String=emote_wheel_session_id
     Value:String=GetAnalyticsSessionID!(ActivationCount:int=@n66) -> (ReturnValue:String)
    ) -> (ReturnValue:Flow Analytics Additional Info Structure)
    bIsOverlay:bool=false
   ) -> (then)
   > BindToAnimationFinished(Animation=Get Anim_Collection_ExpandContract(), Delegate=Create Event())
   > Set IsMenuSticky(IsMenuSticky=false)
   > SetVisibility(self=[Get BottomBarLeft_UseRelease(), Get BottomButtonLeft_Divider()], InVisibility=SelfHitTestInvisible)
   > HandleInputChanged(bNewInputType=GetCurrentInputType(self=Get CommonInputSubsystem()))
   > SetAllInteractionsEnabled!(execute, self, IsInteractionEnabled:bool=false) -> (then)
   > LoadLastSelected!(
    execute
    self:Emote Wheel Item Selection VM Object Reference=Get EmoteWheelItemSelectionVM()
    Item:Emote Wheel VM Object Reference=Get EmoteWheelVM()
   ) -> (then)
   n67 = GetDisplayName(self=Get SelectedCategory(self=Get EmoteWheelItemSelectionVM()))
   > SelectTabByID!(
    execute
    self:Common Tab List Widget Base Object Reference=Get TabGroup_EmoteCategories()
    TabNameID:Name=String To Name(InString=Get Source String from Text!(Text:Text=@n67) -> (ReturnValue:String) {is=title K2Node_CallFunction})
    bSuppressClickFeedback:bool=true
   ) -> (then, ReturnValue:bool)
   > n68 = Branch(Condition=Get bIsCollectionShown(self=Get EmoteWheelItemSelectionVM()))
   > OnCollectionExpand(SkipAnim=true) {is=title K2Node_CallFunction}
   > n69 = Sequence()
   PlayAnimationForward(execute=[@n69.then_0, @n70.then], InAnimation=Get BlurFade() -> (BlurFade:Widget Animation Object Reference))
   SetAllInteractionsEnabled(execute=[@n69.then_1, @n70.Finished], IsInteractionEnabled=true)
   > UpdateCategoryInputActions()
   n70 = Play Animation with Finished event!(
    execute:Exec=@n68.else
    Widget:User Widget Object Reference=Self-Reference()
    InAnimation:Widget Animation Object Reference=Get Intro()
    StartAtTime:float=0
    NumLoopsToPlay:int=1
    PlayMode:EUMGSequencePlayMode Enum
    PlaybackSpeed:float=1
   ) -> (then:Exec, Result:Widget Animation Handle Structure, Finished:Exec) {is=title K2Node_PlayAnimation2}
   Event Destruct()
   > OnViewDestruct(self:Emote Wheel VM Object Reference=Get EmoteWheelVM())
   > UnbindFromAnimationFinished!(
    execute
    self:User Widget Object Reference
    Animation:Widget Animation Object Reference=Get Anim_Collection_ExpandContract()
    Delegate:Delegate=Create Event()
   ) -> (then)
   > SetAllInteractionsEnabled(IsInteractionEnabled=false)
   > Check and Stop Animation(Animation=Get Intro())
   > Check and Stop Animation(Animation=Get BlurFade())
   > n71 = Play Animation with Finished event(
    Widget=Self-Reference()
    InAnimation=Get Outro() -> (Outro:Widget Animation Object Reference)
    StartAtTime=0
    NumLoopsToPlay=1
    PlaybackSpeed=1
   )
   SetAllInteractionsEnabled(execute=@n71.Finished, IsInteractionEnabled=true)
   Event On Activated()
   > ToggleHUDVisibility!(execute, self, bHideElements:bool=true) -> (then)
   > SetBottomButtonVisibility(self=Get WBP_BottomBarDecoy_GamepadUse(), ShouldButtonBeVisible=true)
   n72 = Create Event()
   n73 = Get CommonInputSubsystem()
   > Bind Event to On Input Method Changed(self=@n73, Delegate=@n72)
   > CenterMouseToViewport() {is=title K2Node_CallFunction}
   > n74 = Cast To FortPlayerController(Object=GetOwningPlayer())
   > SetPickerOpened!(execute, self:Fort Player Controller Object Reference=@n74, bIsOpened:bool=true) -> (then)
   > PlaySound2D(Sound=/Game/Sounds/Fort_UI_Sounds/FrontEnd/New_FrontEnd/Fortnite_UI_MainMenu_Generic_02_Cue)
   > Set LastTabIndex(LastTabIndex=GetSelectedTabIndex!(self:UIKit Tab Group Object Reference=Get TabGroup_EmoteCategories()) -> (ReturnValue:int=0)) -> (Output_Get=0)
   > n75 = IsPluginMounted!(execute, PluginName:String=FN_Photomode) -> (then, ReturnValue:bool)
   > SetBottomButtonVisibility(self=Get WBP_BottomBarDecoy_PM() -> (WBP_BottomBarDecoy_PM:WBP Bottom Bar Decoy Button Object Reference), ShouldButtonBeVisible=@n75)
   > Branch(Condition=@n75)
   > Broadcast Event!(
    execute
    self:Gameplay Event Router Component Object Reference=GetGlobalEventRouter()
    EventData:Input Event Structure=Make Input Event!() -> (InputEvent:Input Event Structure) {is=title K2Node_MakeStruct}
    EventContext:Object Reference=GetOwningPlayer()
   ) -> (then, ReturnValue:bool)
   Event On Deactivated()
   > IncrementAnalyticsEWActivationCount(self:Fort Picker Context Object Reference=Get FortPickerContext())
   > ToggleHUDVisibility(bHideElements=false)
   > OnCollectionContract(Reset=true) {is=title K2Node_CallFunction}
   > Unbind Event from On Input Method Changed(self=@n73, Delegate=@n72)
   > FlushInputKeys(self:Emote Wheel VM Object Reference=Get EmoteWheelVM())
   > n76 = Cast To FortPlayerController(Object=GetOwningPlayer())
   > SetPickerOpened(self=@n76, bIsOpened=false)
   > PlaySound2D(Sound=/Game/Sounds/Fort_UI_Sounds/FrontEnd/New_FrontEnd/Fortnite_UI_MainMenu_Generic_03_Cue)
   n77 = Event TriggerChangeTabsAnimation!() -> (OutputDelegate, then, Index:int=0) {is=title K2Node_Event}
   > n78 = Set CurrentTabIndex(CurrentTabIndex=@n77.Index) -> (Output_Get=0)
  ]}
 GetPrimaryContentSetup {nodes [
   GetPrimaryContentSetup() {is=title K2Node_FunctionEntry}
   > Return Node(ReturnValue:Primary Content Setup Structure=Make Primary Content Setup!(
     bShowBottomBar:bool=false
     bShowChatWidget:bool=false
     bEnableNewTextChat:bool=False
     bShowTopBar:bool=false
     bTabSupportsInterface:bool=False
    ) -> (PrimaryContentSetup:Primary Content Setup Structure) {is=title K2Node_MakeStruct})
  ]}
 GetScaledTabTransitionIndex {nodes [
   GetScaledTabTransitionIndex() {is=title K2Node_FunctionEntry}
   n1 = GetSelectedTabIndex(self=Get TabGroup_EmoteCategories()) -> (ReturnValue=0)
   n2 = GetTabIndexByID!(self:UIKit Tab Group Object Reference=Get TabGroup_EmoteCategories(), TabNameID:Name=Get SelectedTabID() -> (SelectedTabID:Name=None)) -> (ReturnValue:int=0)
   n3 = "Not Equal (Enum)"(A=GetCurrentInputType(self=Get CommonInputSubsystem()), B=Gamepad)
   > Return Node(ReturnValue:int=int * int(A=Select(Option 0:int=@n2, Option 1:int=@n1, Index=@n3) -> (ReturnValue:int), B=10000))
  ]}
 GoToNext {nodes [
   GoToNext!() -> (then, InteractionWidget:Widget Object Reference) {is=title K2Node_FunctionEntry}
   > n1 = Branch(Condition=IsValidInterface!(Interface:"Interface Interface (by ref)"=Get SubTabInterface() -> (SubTabInterface:BPI EW Sub Tab Interface Interface)) -> (ReturnValue:bool))
   > GoToNextSubTab!(
    execute
    self:BPI EW Sub Tab Interface Interface=Get SubTabInterface()
    InteractionWidget:Widget Object Reference=Get InteractionWidget() -> (InteractionWidget:Widget Object Reference)
   ) -> (then)
   SelectNextTab!(execute=@n1.else, self:UIKit Tab Group Object Reference=Get TabGroup_EmoteCategories(), bSuppressClickFeedback:bool=true) -> (then)
   > TriggerFlowriderInteraction(Element=Get InteractionWidget(), Value=Next Tab)
  ]}
 GoToPrevious {nodes [
   GoToPrevious!() -> (then, InteractionWidget:Widget Object Reference) {is=title K2Node_FunctionEntry}
   > n1 = Branch(Condition=IsValidInterface(Interface=Get SubTabInterface()))
   > GoToPreviousSubTab!(execute, self:BPI EW Sub Tab Interface Interface=Get SubTabInterface(), InteractionWidget:Widget Object Reference=Get InteractionWidget()) -> (then)
   SelectPreviousTab!(execute=@n1.else, self:UIKit Tab Group Object Reference=Get TabGroup_EmoteCategories(), bSuppressClickFeedback:bool=true) -> (then)
   > TriggerFlowriderInteraction(Element=Get InteractionWidget(), Value=Previous Tab)
  ]}
 HandleCategoryChanged {nodes [
   n1 = HandleCategoryChanged!() -> (then, VM:MVVM Base Viewmodel Object Reference) {is=title K2Node_FunctionEntry}
   > n2 = Cast To EmoteWheelCategoryVM!(execute, Object=@n1) -> (then, CastFailed, AsEmote Wheel Category VM:Emote Wheel Category VM Object Reference) {is=title K2Node_DynamicCast}
   > SetSelectedCategory!(
    execute
    self:Emote Wheel Item Selection VM Object Reference=Get EmoteWheelItemSelectionVM()
    Category:Emote Wheel Category VM Object Reference=@n2
   ) -> (then)
   GetInjectedWidget!(self:Fort Plugin Widget Injection Slot Object Reference=Get SubTab_WidgetInjectionSlot()) -> (ReturnValue:Widget Object Reference)
  ]}
 HandleInputChanged {nodes [
   n1 = HandleInputChanged!() -> (then, bNewInputType:ECommonInputType Enum) {is=title K2Node_FunctionEntry}
   n2 = Get IsMenuSticky() -> (IsMenuSticky:bool)
   n3 = "Equal (Enum)"(A=@n1, B=MouseAndKeyboard)
   n4 = NOT Boolean(A=@n3)
   > SetBottomButtonVisibility(
    self=Get GamepadReleaseButtonToUse() -> (GamepadReleaseButtonToUse:WBP Bottom Bar Decoy Button Object Reference)
    ShouldButtonBeVisible=AND Boolean(A=NOT Boolean(A=@n2), B=@n4)
   )
   > SetBottomButtonVisibility(
    self=Get GamepadReleaseButtonToHide() -> (GamepadReleaseButtonToHide:WBP Bottom Bar Decoy Button Object Reference)
    ShouldButtonBeVisible=false
   )
   n5 = NOT Boolean(A=Get IsCollectionExpanded())
   > SetBottomButtonVisibility(
    self=Get WBP_BottomBarDecoy_Use() -> (WBP_BottomBarDecoy_Use:WBP Bottom Bar Decoy Button Object Reference)
    ShouldButtonBeVisible=AND Boolean!(A=@n4, B=@n2, C:bool=@n5) -> (ReturnValue)
   )
   > SetVisibility(
    self=Get BottomBar_Left() -> (BottomBar_Left:Horizontal Box Object Reference)
    InVisibility=Select(
     Option 0:ESlateVisibility Enum=Collapsed
     Option 1:ESlateVisibility Enum=SelfHitTestInvisible
     Index=AND Boolean!(A=@n5, B=@n3) -> (ReturnValue)
    ) -> (ReturnValue:ESlateVisibility Enum)
   )
   > HideActionButtonsTouchInput!(execute, self, InputType:ECommonInputType Enum=@n1) -> (then)
   > UpdateTouchNavButtons(Categories=Get TabCategoriesArray())
  ]}
 HideActionButtonsTouchInput {nodes [
   n1 = HideActionButtonsTouchInput!() -> (then, InputType:ECommonInputType Enum) {is=title K2Node_FunctionEntry}
   > Branch(Condition="Equal (Enum)"(A=@n1, B=Touch))
   > SetBottomButtonVisibility(
    self=Get WBP_BottomBarDecoy_UseRelease() -> (WBP_BottomBarDecoy_UseRelease:WBP Bottom Bar Decoy Button Object Reference)
    ShouldButtonBeVisible=false
   )
   > SetBottomButtonVisibility(self=Get WBP_BottomBarDecoy_Use(), ShouldButtonBeVisible=false)
   > SetBottomButtonVisibility(
    self=Get WBP_BottomBarDecoy_UseCommsRelease() -> (WBP_BottomBarDecoy_UseCommsRelease:WBP Bottom Bar Decoy Button Object Reference)
    ShouldButtonBeVisible=false
   )
  ]}
 OnKeyUp {nodes [
   n1 = OnKeyUp!() -> (then, MyGeometry:Geometry Structure, InKeyEvent:Key Event Structure) {is=title K2Node_FunctionEntry}
   > n2 = Branch(Condition=Contains Item(
     TargetArray:Array of Key Structures=Get ReleaseInputBindings() -> (ReleaseInputBindings:Array of Key Structures)
     ItemToFind:"Key Structure (by ref)"=GetKey!(Input:"Key Event Structure (by ref)"=@n1.InKeyEvent) -> (ReturnValue:Key Structure)
    ))
   > n3 = Branch(Condition=NOT Boolean(A=Get IsMenuSticky()))
   > UseItem()
   > TriggerFlowriderInteraction(Element=Get GamepadReleaseButtonToUse(), Value=Use Item Release)
   Return Node(
    execute=[@n3.else, @n2.else]
    ReturnValue:Event Reply Structure=Make Event Reply!() -> (EventReply:Event Reply Structure) {is=title K2Node_MakeStruct}
   )
  ]}
 OnMouseButtonDown {nodes [
   n1 = OnMouseButtonDown!() -> (then, MyGeometry:Geometry Structure, MouseEvent:"Pointer Event Structure (by ref)") {is=title K2Node_FunctionEntry}
   n2 = Get Effecting Button!(Input:"Pointer Event Structure (by ref)"=@n1.MouseEvent) -> (ReturnValue:Key Structure) {is=title K2Node_CallFunction}
   > n3 = Branch(Condition=NOT Boolean(A="Equal (Key)"(A:Key Structure=@n2, B:Key Structure=RightMouseButton) -> (ReturnValue:bool=~) {is=title K2Node_PromotableOperator}))
   > Return Node(ReturnValue=Make Event Reply())
   CloseMenu(execute=@n3.else) {is=title K2Node_CallFunction}
   > TriggerFlowriderInteraction(Element=Get WBP_BottomBarDecoy_Close() -> (WBP_BottomBarDecoy_Close:WBP Bottom Bar Decoy Button Object Reference), Value=RightMouseClose)
  ]}
 OnMouseButtonUp {nodes [
   n1 = OnMouseButtonUp!() -> (then, MyGeometry:Geometry Structure, MouseEvent:"Pointer Event Structure (by ref)") {is=title K2Node_FunctionEntry}
   > n2 = Branch(Condition=Contains Item(TargetArray=Get ReleaseInputBindings(), ItemToFind=Get Effecting Button(Input=@n1.MouseEvent)))
   > n3 = Branch(Condition=NOT Boolean(A=Get IsMenuSticky()))
   > UseItem()
   > TriggerFlowriderInteraction(Element=Get GamepadReleaseButtonToUse(), Value=Use Item Release)
   Return Node(execute=[@n3.else, @n2.else], ReturnValue=Make Event Reply())
  ]}
 OnTouchEnded {nodes [
   n1 = OnTouchEnded!() -> (then, MyGeometry:Geometry Structure, InTouchEvent:"Pointer Event Structure (by ref)") {is=title K2Node_FunctionEntry}
   > n2 = GetConsoleVariableBoolValue!(execute, VariableName:String=EmoteWheelPickerScreen.EnableTapAnywhereToCloseOnTouch) -> (then, ReturnValue:bool)
   n3 = IsUnderLocation!(
    Geometry:"Geometry Structure (by ref)"=GetCachedGeometry!(self:Widget Object Reference=Get WBP_Emote_RadialWheel()) -> (ReturnValue:"Geometry Structure (by ref)")
    AbsoluteCoordinate:"Vector 2D Structure (by ref)"=Get Screen Space Position!(Input:"Pointer Event Structure (by ref)"=@n1.InTouchEvent) -> (ReturnValue:Vector 2D Structure) {is=title K2Node_CallFunction}
   ) -> (ReturnValue:bool)
   n4 = Get IsCollectionExpanded()
   > n5 = Branch(Condition=AND Boolean!(A=@n2, B=NOT Boolean(A=@n3), C:bool=NOT Boolean(A=@n4)) -> (ReturnValue))
   > TriggerFlowriderInteraction(Element=Self-Reference(), Value=Close Wheel)
   > CloseMenu()
   > Return Node(ReturnValue=Unhandled())
   Return Node(execute=@n5.else, ReturnValue=Unhandled())
  ]}
 PopulateTabs {nodes [
   n1 = PopulateTabs!() -> (then, Categories:Array of Emote Wheel Category VM Object References) {is=title K2Node_FunctionEntry}
   > Set TabCategoriesArray(TabCategoriesArray:Array of Emote Wheel Category VM Object References=@n1) -> (Output_Get:Array of Emote Wheel Category VM Object References)
   > RemoveAllTabs(self:Common Tab List Widget Base Object Reference=Get TabGroup_EmoteCategories())
   > n2 = For Each Loop(Array:Array of Emote Wheel Category VM Object References=Get Categories() -> (Categories:Array of Emote Wheel Category VM Object References)) -> (Array Element:Emote Wheel Category VM Object Reference)
   UpdateTouchNavButtons(execute=@n2.Completed, Categories=Get TabCategoriesArray())
   n3 = GetDisplayName(self=@n2)
   AddTab!(
    execute=@n2.LoopBody
    self:WBP UIKit Tab Group Base Object Reference=Get TabGroup_EmoteCategories()
    Tab VM:MVVM Base Viewmodel Object Reference=@n2
    Tab ID:Name=String To Name(InString=Get Source String from Text(Text=@n3))
    Index:int=@n2.Array Index
   ) -> (then)
  ]}
 SetAllInteractionsEnabled {nodes [
   n1 = SetAllInteractionsEnabled!() -> (then, IsInteractionEnabled:bool=~) {is=title K2Node_FunctionEntry}
   > SetIsInteractionEnabled(self=[Get WBP_BottomBarDecoy_Use(), Get GamepadReleaseButtonToUse(), Get WBP_BottomBarDecoy_Close()], bInIsInteractionEnabled=@n1)
  ]}
 SetCollectionButtonText {nodes [
   n1 = SetCollectionButtonText!() -> (then, IsCollectionExpanded:bool=~) {is=title K2Node_FunctionEntry}
   > n2 = Set Is Collection Expanded(Is Collection Expanded:bool=@n1) -> (Output_Get:bool)
   > n3 = Branch(Condition=@n2)
   n4 = Get Block_Text!(self:WBP UIKit Button Generic Base Object Reference=Get Button_ViewCollection()) -> (Block_Text:WBP UIKit Block Text Object Reference)
   > SetText!(execute, self:WBP UIKit Block Text Object Reference=@n4, Text:Text="LOCTEXT(\"View Emote Wheel\")") -> (then)
   SetText(
    execute=@n3.else
    self=@n4
    Text=Format Text!(Format="LOCTEXT(\"View Collection ({count})\")", count:int=Get CollectionCount() -> (CollectionCount:int=0)) -> (Result)
   )
  ]}
 SetLerpTopToBottomOfScreenHeight {nodes [
   n1 = SetLerpTopToBottomOfScreenHeight!() -> (then, NewValue:double=~) {is=title K2Node_FunctionEntry}
   n2 = Lerp(A=1, B=0, Alpha=@n1)
   n3 = Get AnimationGrid_ForDynamicScreenHeight() -> (AnimationGrid_ForDynamicScreenHeight:Grid Panel Object Reference)
   > SetRowFill!(execute, self:Grid Panel Object Reference=@n3, RowIndex:int=0, Coefficient:float=@n2) -> (then)
   > SetRowFill(self=@n3, RowIndex=1, Coefficient=float - float(A=1, B=@n2))
  ]}
 SetMobileStyle {nodes [
   SetMobileStyle() {is=title K2Node_FunctionEntry}
   > Branch(Condition=IsUsingCompactUI())
   n1 = SlotAsGridSlot(Widget=Get TabGroup_VBox() -> (TabGroup_VBox:Vertical Box Object Reference))
   n2 = Get!(self:Grid Slot Object Reference=@n1) -> (Padding_Left:float=~, Padding_Top:float=~, Padding_Right:float=~, Padding_Bottom:float=~)
   > SetPadding!(
    execute
    self:Grid Slot Object Reference=@n1
    InPadding_Left:float=@n2
    InPadding_Top:float=14
    InPadding_Right:float=@n2.Padding_Right
    InPadding_Bottom:float=@n2.Padding_Bottom
   ) -> (then)
   > Branch(Condition=IsUsingInputType!(
     InputType:ECommonInputType Enum=Touch
     PlayerController:Player Controller Object Reference=GetPlayerController!(PlayerIndex:int=0) -> (ReturnValue:Player Controller Object Reference)
    ) -> (ReturnValue:bool))
   > Set UseMinButtonSize!(execute, UseMinButtonSize:bool=false, self:WBP UIKit Tab Group Base Object Reference=Get TabGroup_EmoteCategories()) -> (then, Output_Get:bool)
  ]}
 ToggleHUDVisibility {nodes [
   ToggleHUDVisibility!() -> (then, bHideElements:bool=~) {is=title K2Node_FunctionEntry}
   > Branch(Condition=IsInZone!(self:Fort Global UIContext Object Reference=Get FortGlobalUIContext() -> (ReturnValue:Fort Global UIContext Object Reference) {is=title K2Node_GetSubsystem}) -> (ReturnValue:bool))
   > n1 = GetFortHUDContextByPlayer!(
    execute
    LocalPlayer:Local Player Object Reference=GetOwningLocalPlayer!(self:Widget Object Reference) -> (ReturnValue:Local Player Object Reference)
   ) -> (then, ReturnValue:Fort HUDContext Object Reference)
   > SetHUDElementVisibility!(
    execute
    self:Fort HUDContext Object Reference=@n1
    HUDElementTags:"Gameplay Tag Container Structure (by ref)"=MakeGameplayTagContainerFromTag!(SingleTag:Gameplay Tag Structure=(TagName="HUD.Athena.HideforRadial")) -> (ReturnValue:Gameplay Tag Container Structure)
    bHideElements:bool=Get bHideElements() -> (bHideElements:bool)
    ReasonName:"Name (by ref)"=None
   ) -> (then)
  ]}
 TriggerFlowriderInteraction {nodes [
   n1 = TriggerFlowriderInteraction!() -> (then, Element:Widget Object Reference, Value:String) {is=title K2Node_FunctionEntry}
   n2 = GetAnalyticsEWActivationCount(self=Get FortPickerContext()) -> (ReturnValue=0)
   n3 = Make Flow Analytics Additional Info()
   > Fire Event Element Interacted(
    Element=@n1
    AdditionalInfo="Add String (Non-Localized)"(
     Info="Add String (Non-Localized)"(Info=@n3, Key=Button Name, Value=@n1.Value)
     Key=emote_wheel_session_id
     Value=GetAnalyticsSessionID(ActivationCount=@n2)
    )
   )
  ]}
 UnLoadCollection {nodes [
   UnLoadCollection() {is=title K2Node_FunctionEntry}
   n1 = GetAnalyticsEWActivationCount(self=Get FortPickerContext()) -> (ReturnValue=0)
   > Fire Event Main View Opened(
    View=Self-Reference()
    AdditionalInfo="Add String (Non-Localized)"(Info=Make Flow Analytics Additional Info(), Key=emote_wheel_session_id, Value=GetAnalyticsSessionID(ActivationCount=@n1))
    bIsOverlay=false
   )
   > UpdateFlowriderSubView(SubViewString="To String (Name)"(InName=Get SelectedTabID() -> (SelectedTabID=None)))
   > HandleInputChanged(bNewInputType=GetCurrentInputType(self=Get CommonInputSubsystem())) {is=title K2Node_CallFunction}
   > n2 = Get!(execute:Exec) -> (then:Exec, else:Exec, LockerCollectionWidget:Common Activatable Widget Object Reference)
   > RemoveFromParent(self=@n2)
   > Set LockerCollectionWidget()
   > SetVisibility(self=Get WBP_UIKit_Throbber(), InVisibility=HitTestInvisible)
  ]}
 UpdateCategoryInputActions {nodes [
   UpdateCategoryInputActions() {is=title K2Node_FunctionEntry}
   > SetInputActionsType!(
    execute
    self:WBP UIKit Tab Group Base Object Reference=Get TabGroup_EmoteCategories()
    InputActions:E_UI_TabsInputActionType Enum=NewEnumerator0
    InputActionsWithMouseScroll:bool=false
   ) -> (then)
  ]}
 UpdateCollectionButtonVisibility {nodes [
   n1 = UpdateCollectionButtonVisibility!() -> (then, ShouldShow:bool=~) {is=title K2Node_FunctionEntry}
   n2 = "Not Equal (Boolean)"(A=HasMultipleLocalPlayers!() -> (ReturnValue:bool), B=true) -> (ReturnValue=~)
   > SetVisibility(
    self=Get CollectionButton() -> (CollectionButton:Overlay Object Reference)
    InVisibility=Select(Option 0=Collapsed, Option 1=SelfHitTestInvisible, Index=AND Boolean!(A=@n1, B=@n2) -> (ReturnValue))
   )
  ]}
 UpdateCollectionCount {nodes [
   n1 = UpdateCollectionCount!() -> (then, count:int) {is=title K2Node_FunctionEntry}
   > Set CollectionCount(CollectionCount:int=@n1) -> (Output_Get:int=0)
   > n2 = Branch(Condition=Get Is Collection Expanded() -> (Is Collection Expanded:bool))
   SetText(
    execute=@n2.else
    self=Get Block_Text(self=Get Button_ViewCollection())
    Text=Format Text(Format="LOCTEXT(\"View Collection ({count})\")", count=Get CollectionCount() -> (CollectionCount=0))
   )
  ]}
 UpdateCollectionDetailsVisibility {nodes [
   UpdateCollectionDetailsVisibility!() -> (then, IsCollectionShown:bool=~, CategoryDisplayName:Text) {is=title K2Node_FunctionEntry}
   > n1 = Get()
   > n2 = Cast To WBP_EW_Collection(Object=@n1)
   n3 = "Equal, Case Insensitive (String)"(
    A="Get (a copy)"(Array=Get SlotIndicatorHideStringArray() -> (SlotIndicatorHideStringArray:Array of Strings))
    B=Get Source String from Text(Text=Get CategoryDisplayName() -> (CategoryDisplayName:Text))
   )
   n4 = Get IsCollectionShown() -> (IsCollectionShown:bool)
   > SetVisibility(
    self=Get WBP_ItemSelection_ItemDetails!(self:WBP EW Collection Object Reference=@n2) -> (WBP_ItemSelection_ItemDetails:WBP Item Selection Item Details Object Reference)
    InVisibility=Select(Option 0=SelfHitTestInvisible, Option 1=Collapsed, Index=AND Boolean(A=@n4, B=@n3))
   )
  ]}
 UpdateFlowriderSubView {nodes [
   n1 = UpdateFlowriderSubView!() -> (then, SubViewString:String) {is=title K2Node_FunctionEntry}
   n2 = GetAnalyticsEWActivationCount(self=Get FortPickerContext()) -> (ReturnValue=0)
   > "Fire Event Sub View Opened (Non-Localized)"!(
    execute
    FortPC:Fort Player Controller Object Reference=GetFortPlayerControllerFromActor!(Actor:Actor Object Reference=GetOwningPlayer()) -> (ReturnValue:Fort Player Controller Object Reference)
    ViewName:String=@n1
    AdditionalInfo:Flow Analytics Additional Info Structure="Add String (Non-Localized)"(Info=Make Flow Analytics Additional Info(), Key=emote_wheel_session_id, Value=GetAnalyticsSessionID(ActivationCount=@n2))
    bIsOverlay:bool=false
   ) -> (then)
   Set PromptNudge!(execute, PromptNudge:Vector 2D Structure, self:WBP UIKit Button Generic Base Object Reference=Get Button_ViewCollection()) -> (then, Output_Get:Vector 2D Structure)
  ]}
 UpdateReleasePrompts {nodes [
   UpdateReleasePrompts!() -> (then, ReleaseInputBindings:Array of Key Structures) {is=title K2Node_FunctionEntry}
   > Set MatchFound(MatchFound:bool=false) -> (Output_Get:bool)
   > n1 = For Each Loop with Break!(Exec:Exec, Array:Array of Key Structures=Get ReleaseInputBindings(), Break:Exec=@n3.then_0) -> (LoopBody:Exec, Array Element:Key Structure, Array Index:int, Completed:Exec)
   > n2 = Cast To FortPlayerController(Object=GetOwningPlayer())
   > Branch(Condition=IsInputKeyDown!(self:Player Controller Object Reference=@n2, Key:Key Structure=@n1) -> (ReturnValue:bool))
   > Set MatchFound(MatchFound=true)
   > n3 = Sequence!(execute) -> (then_0, then_1, then_2:Exec)
   n4 = Get GamepadPossibleInputs() -> (GamepadPossibleInputs:Map of Key Structures to Data Table Row Handle Structures)
   n5 = Branch(
    execute=@n3.then_1
    Condition="Equal (Key)"(
     A=Get Key Value By Index!(TargetMap:Map of Key Structures to Data Table Row Handle Structures=@n4, Index:int=0) -> (Key:Key Structure, Value:Data Table Row Handle Structure) {is=title K2Node_CallFunction}
     B=@n1
    ) -> (ReturnValue=~)
   )
   Set GamepadReleaseButtonToUse(execute=[@n5.then, @n10.else], GamepadReleaseButtonToUse:WBP Bottom Bar Decoy Button Object Reference=Get WBP_BottomBarDecoy_UseRelease()) -> (Output_Get:WBP Bottom Bar Decoy Button Object Reference)
   > n6 = Set GamepadReleaseButtonToHide(GamepadReleaseButtonToHide:WBP Bottom Bar Decoy Button Object Reference=Get WBP_BottomBarDecoy_UseCommsRelease()) -> (Output_Get:WBP Bottom Bar Decoy Button Object Reference)
   HandleInputChanged(execute=[@n7.then, @n6.then], bNewInputType=GetCurrentInputType(self=Get CommonInputSubsystem()))
   Set GamepadReleaseButtonToUse(execute=@n5.else, GamepadReleaseButtonToUse=Get WBP_BottomBarDecoy_UseCommsRelease())
   > n7 = Set GamepadReleaseButtonToHide(GamepadReleaseButtonToHide=Get WBP_BottomBarDecoy_UseRelease())
   n8 = Find(
    TargetMap:Map of Key Structures to Data Table Row Handle Structures=Get KBMPossibleInputs() -> (KBMPossibleInputs:Map of Key Structures to Data Table Row Handle Structures)
    Key:"Key Structure (by ref)"=@n1
   ) -> (Value:Data Table Row Handle Structure)
   Branch(execute=@n3.then_2, Condition=@n8.ReturnValue)
   n9 = Get BottomBarLeft_UseRelease()
   > SetTriggeringInputAction!(execute, self:Common Button Base Object Reference=@n9, InputActionRow:"Data Table Row Handle Structure (by ref)"=@n8) -> (then)
   > UpdatePrompt!(
    execute
    self:WBP UIKit Button Generic Base Object Reference=@n9
    InputType:ECommonInputType Enum=GetCurrentInputType(self=Get CommonInputSubsystem())
   ) -> (then)
   n10 = Branch(execute=@n1.Completed, Condition=Get MatchFound() -> (MatchFound:bool))
   SetTriggeringInputAction(self=Get WBP_BottomBarDecoy_UseRelease())
  ]}
 UpdateSlotIndicatorVisibility {nodes [
   n1 = UpdateSlotIndicatorVisibility!() -> (then, CategoryDisplayName:Text) {is=title K2Node_FunctionEntry}
   n2 = Get Source String from Text(Text=@n1)
   n3 = Get SlotIndicatorHideStringArray()
   > UpdateSlotIndicatorVisibility!(
    execute
    self:WBP Emote Radial Wheel Object Reference=Get WBP_Emote_RadialWheel()
    ShouldShow:bool="Equal (Integer)"(A=Find Item!(TargetArray=@n3, ItemToFind:"String (by ref)"=@n2) -> (ReturnValue), B=-1)
   ) -> (then) {is=title K2Node_CallFunction}
  ]}
 UpdateSubTabVisibility {nodes [
   UpdateSubTabVisibility() {is=title K2Node_FunctionEntry}
   > SetVisibility(self=Get SubTab_WidgetInjectionSlot(), InVisibility=SelfHitTestInvisible)
  ]}
 UpdateTouchNavButtons {nodes [
   UpdateTouchNavButtons!() -> (then, Categories:Array of Emote Wheel Category VM Object References) {is=title K2Node_FunctionEntry}
   n1 = NOT Boolean(A=Get IsCollectionExpanded())
   n2 = "integer > integer"(A=Length(TargetArray:Array of Emote Wheel Category VM Object References=Get Categories()), B=1)
   > SetVisibility(
    self=[
     Get PreviousArrow_Button() -> (PreviousArrow_Button:WBP UIKit Button Tab Nav Arrow Object Reference)
     Get NextArrow_Button() -> (NextArrow_Button:WBP UIKit Button Tab Nav Arrow Object Reference)
    ]
    InVisibility=Select(Option 0=Collapsed, Option 1=SelfHitTestInvisible, Index=AND Boolean(A=@n1, B=@n2))
   )
  ]}
 UseItem {nodes [
   UseItem() {is=title K2Node_FunctionEntry}
   > n1 = Get!(execute, self:Emote Wheel Item Selection VM Object Reference=Get EmoteWheelItemSelectionVM()) -> (then, else, SelectedItem:Emote Wheel Item VM Object Reference)
   > n2 = Play(self:Emote Wheel Item VM Object Reference=@n1)
   CloseMenu(execute=[@n2.then, @n1.else])
  ]}
]
variables [
 ClickedCollectionButton {type=CommonButtonBase, default=null}
 ClickedUseButton {type=CommonButtonBase, default=null}
 CollectionClass {type=Class, default=&/WBP_EW_Collection.WBP_EW_Collection_C}
 CollectionCount {type=integer}
 CurrentOverlayWidget {type=UserWidget, default=null}
 CurrentTabIndex {type=integer}
 GamepadReleaseButtonToHide {type=WBP_BottomBarDecoyButton_C, default=null}
 GamepadReleaseButtonToUse {type=WBP_BottomBarDecoyButton_C, default=null}
 InputAction_Back {type=DataTableRowHandle, default {dataTable=&InputActionDataTable, rowName=Back}}
 InputAction_Close {type=DataTableRowHandle, default {dataTable=&InputActionDataTable, rowName=Close}}
 Is Collection Expanded {type=boolean}
 IsCollectionExpanded {type=boolean}
 IsMenuSticky {type=boolean}
 LastTabIndex {type=integer}
 LerpTopToBottomOfScreenHeight {type=number}
 LockerCollectionWidget {type=CommonActivatableWidget, default=null}
 ReleaseInputBindings {type="array<Key>", default []}
 SelectedTabID {type=string, default=None}
 SlotIndicatorHideStringArray {type="array<string>", default [Jam Tracks, Comms, Follower]}
 SoftCollectionClass {type=Class, default=&WBP_EW_Collection_C}
 SubTabInterface {default=None}
 TabCategoriesArray {type="array<EmoteWheelCategoryVM>", default []}
 TempBuiltInID {type=string, default=None}
 TempEmoteID {type=string, default=None}
 TempJamLoopID {type=string, default=None}
]
widgets {
 info {inheritedWidgetCount=0, parentClass=/Script/EmoteWheelRuntime.EmoteWheelPickerScreen, rootWidgetClass=&GridPanel, widgetCount=43}
 widgets [
  GridPanel_Main {widgetClassPath=&GridPanel, widget=GridPanel_Main, properties {columnFill [1], rowFill [1]}}
  BG_Darkener {
   widgetClassPath=&Image
   bIsVariable=true
   parent=GridPanel_Main
   widget=BG_Darkener
   properties {colorAndOpacity=<rgba 0.04 0.04 0.04 0.4>}
   slot {class=GridSlot, properties {layer=-101}}
  }
  WBP_EmotePicker_PostBufferUpdate {
   widgetClassPath=&/WBP_EmotePicker_PostBufferUpdate
   bIsVariable=true
   parent=GridPanel_Main
   widget=WBP_EmotePicker_PostBufferUpdate
   slot {class=GridSlot, properties {layer=-102}}
  }
  Background_BlockClick_BlurOnlyHighEndDevice {
   widgetClassPath=&Image
   bIsVariable=true
   parent=GridPanel_Main
   widget=Background_BlockClick_BlurOnlyHighEndDevice
   properties {brush {
     bIsDynamicallyLoaded=false
     drawAs=Image
     imageType=FullColor
     mirroring=NoMirror
     resourceName=None
     resourceObject=/EmoteWheel/MaterialInstances/MI_UI_EmoteScreen_Background
     tiling=NoTile
     imageSize=<32 32>
     margin=<ltrb 0 0 0 0>
     outlineSettings {
      bUseBrushTransparency=false
      roundingType=HalfHeightRadius
      width=0
      color {colorUseRule=UseColor_Specified, specifiedColor=<rgba 0 0 0 0>}
      cornerRadii=<0 0 0 0>
     }
     tintColor {colorUseRule=UseColor_Specified, specifiedColor=<rgba 1 1 1 1>}
     uVRegion {bIsValid=false, max=<0 0>, min=<0 0>}
    }}
   slot {class=GridSlot, properties {layer=-50}}
  }
  WBP_AthenaBottomBarWidget {
   widgetClassPath=/Game/UI/Frontend/BottomBar/WBP_AthenaBottomBarWidget
   bIsVariable=true
   parent=GridPanel_Main
   widget=WBP_AthenaBottomBarWidget
   slot {class=GridSlot, properties {horizontalAlignment=HAlign_Right, layer=100, verticalAlignment=VAlign_Bottom}}
  }
  SafeZone {widgetClassPath=&SafeZone, bIsVariable=true, parent=GridPanel_Main, widget=SafeZone, slot {class=GridSlot}}
  GridPanel_18 {
   widgetClassPath=&GridPanel
   parent=SafeZone
   widget=GridPanel_18
   properties {columnFill [1], rowFill [1]}
   slot {class=SafeZoneSlot}
  }
  TabGroup_VBox {
   widgetClassPath=&VerticalBox
   bIsVariable=true
   parent=GridPanel_18
   widget=TabGroup_VBox
   slot {class=GridSlot, properties {horizontalAlignment=HAlign_Center, verticalAlignment=VAlign_Top, padding=<ltrb 0 30 0 0>}}
  }
  TabGroup_EmoteCategories {
   widgetClassPath=/FortUILibrary/UIKit/Tabs/WBP_UIKit_TabGroup
   bIsVariable=true
   parent=TabGroup_VBox
   widget=TabGroup_EmoteCategories
   properties {
    backplateBrightness=Dark
    backplateClass=/FortUILibrary/UIKit/Backplates/WBP_UIKit_Backplate_Fallback.WBP_UIKit_Backplate_Fallback_C
    buttonClass=&/WBP_EmotePickerTab.WBP_EmotePickerTab_C
    hideMouseAndKeyboardPrompts=true
    inputActionsWithMouseScroll=true
    promptBackplateBrightness=Dark
    promptsBackplateClass=&WBP_UIKit_Backplate_Fallback_C
    usePromptBackplate=false
    initialTabsData []
   }
   slot {class=VerticalBoxSlot, properties {horizontalAlignment=HAlign_Center, verticalAlignment=VAlign_Top}}
  }
  SubTab_WidgetInjectionSlot {
   widgetClassPath=/Game/UI/Plugin/FortPluginWidgetInjectionSlot
   bIsVariable=true
   parent=TabGroup_VBox
   widget=SubTab_WidgetInjectionSlot
   properties {pluginId=EmoteWheelContent, pluginWidgetId=EmoteSubTab, visibility=Collapsed}
   slot {class=VerticalBoxSlot, properties {horizontalAlignment=HAlign_Center, verticalAlignment=VAlign_Top}}
  }
  RadialWheel_ScaleBox {
   widgetClassPath=&ScaleBox
   parent=GridPanel_18
   widget=RadialWheel_ScaleBox
   properties {stretchDirection=DownOnly}
   slot {class=GridSlot, properties {horizontalAlignment=HAlign_Center, verticalAlignment=VAlign_Center, padding=<ltrb 0 200 0 200>}}
  }
  WBP_Emote_RadialWheel {
   widgetClassPath=&/WBP_Emote_RadialWheel
   bIsVariable=true
   parent=RadialWheel_ScaleBox
   widget=WBP_Emote_RadialWheel
   properties {bAllowEmptySelection=false, navigation=null, desiredFocusWidget {widgetName=None}}
   slot {class=ScaleBoxSlot, properties {horizontalAlignment=HAlign_Fill, verticalAlignment=VAlign_Fill}}
  }
  Overlay_LockerCollection {
   widgetClassPath=&Overlay
   bIsVariable=true
   parent=GridPanel_18
   widget=Overlay_LockerCollection
   properties {renderTransform {angle=0, scale=<1 1>, shear=<0 0>, translation=<0 1080>}}
   slot {class=GridSlot, properties {layer=20}}
  }
  WBP_UIKit_Throbber {
   widgetClassPath=/FortUILibrary/UIKit/Throbber/WBP_UIKit_Throbber
   bIsVariable=true
   parent=Overlay_LockerCollection
   widget=WBP_UIKit_Throbber
   properties {throbberSize=Large}
   slot {class=OverlaySlot, properties {horizontalAlignment=HAlign_Center, verticalAlignment=VAlign_Center}}
  }
  AnimationGrid_ForDynamicScreenHeight {
   widgetClassPath=&GridPanel
   bIsVariable=true
   parent=GridPanel_18
   widget=AnimationGrid_ForDynamicScreenHeight
   properties {bOverride_Cursor=true, columnFill [1], rowFill [1,0]}
   slot {class=GridSlot, properties {layer=30}}
  }
  Spacer_Row0 {
   widgetClassPath=&Spacer
   parent=AnimationGrid_ForDynamicScreenHeight
   widget=Spacer_Row0
   slot {class=GridSlot, properties {rowSpan=1}}
  }
  VisualAttachment_Row1 {
   widgetClassPath=&CommonVisualAttachment
   parent=AnimationGrid_ForDynamicScreenHeight
   widget=VisualAttachment_Row1
   properties {contentAnchor=<0.5 0>}
   slot {class=GridSlot, properties {horizontalAlignment=HAlign_Center, row=1, rowSpan=1}}
  }
  CollectionButton {
   widgetClassPath=&Overlay
   bIsVariable=true
   parent=VisualAttachment_Row1
   widget=CollectionButton
   properties {bOverride_Cursor=true}
   slot {class=SizeBoxSlot, properties {padding=<ltrb 0 -130 0 0>}}
  }
  ViewCollectionButton {
   widgetClassPath=&Button
   parent=CollectionButton
   widget=ViewCollectionButton
   properties {isFocusable=false, backgroundColor=<rgba 1 1 1 0>, renderTransform {angle=0, scale=<1.5 1>, shear=<0 0>, translation=<0 0>}}
   slot {class=OverlaySlot, properties {horizontalAlignment=HAlign_Fill, verticalAlignment=VAlign_Fill}}
  }
  Button_ViewCollection {
   widgetClassPath=&WBP_UIKit_Button_Generic
   bIsVariable=true
   parent=CollectionButton
   widget=Button_ViewCollection
   properties {
    bIsFocusable=false
    shouldPromptImpactLayout=true
    text="View Collection {x}"
    useBackgroundBlock=false
    buttonPadding=<ltrb 18 0 0 0>
    font {
     bForceMonospaced=false
     bMaterialIsStencil=false
     fontMaterial=/FortUILibrary/UIKit/Buttons/MaterialInstances/CTAButtons/MI_UIKit_TextQuiet_Prompt
     fontObject=&HeadingNow
     letterSpacing=10
     monospacedWidth=1
     size=17.25
     skewAmount=0
     typefaceFontName=76 Bold
     outlineSettings {
      bApplyOutlineToDropShadows=false
      bMiteredCorners=false
      bSeparateFillAlpha=false
      outlineMaterial=null
      outlineSize=0
      outlineColor=<rgba 0 0 0 1>
     }
    }
    promptNudge=<0 0>
    triggeringInputAction {dataTable=/FortUILibrary/Input/LockerInputActionsDataTable, rowName=ViewEmoteCollection}
   }
   slot {class=OverlaySlot, properties {horizontalAlignment=HAlign_Center}}
  }
  Button_Arrow {
   widgetClassPath=/FortUILibrary/UIKit/Buttons/ArrowButtons/WBP_UIKit_ViewMore_ArrowButton
   bIsVariable=true
   parent=CollectionButton
   widget=Button_Arrow
   properties {bEnableMaterialPulse=true, holdData=&CommonUI_HoldData_Base_C}
   slot {class=OverlaySlot, properties {horizontalAlignment=HAlign_Center, padding=<ltrb 0 35 0 0>}}
  }
  BottomBar_Left {
   widgetClassPath=&HorizontalBox
   bIsVariable=true
   parent=GridPanel_18
   widget=BottomBar_Left
   slot {class=GridSlot, properties {horizontalAlignment=HAlign_Left, verticalAlignment=VAlign_Bottom, padding=<ltrb 80 0 0 35>}}
  }
  BottomBarLeft_UseRelease {
   widgetClassPath=&WBP_UIKit_Button_Generic
   bIsVariable=true
   parent=BottomBar_Left
   widget=BottomBarLeft_UseRelease
   properties {
    bDisplayInputActionWhenNotInteractable=true
    bIsFocusable=false
    bLocked=true
    text="Use It (Release)"
    useBackgroundBlock=false
    useIconBlock=false
    font {
     bForceMonospaced=false
     bMaterialIsStencil=false
     fontMaterial=&MI_UIKit_TextQuiet_Prompt
     fontObject=&HeadingNow
     letterSpacing=0
     monospacedWidth=1
     size=15
     skewAmount=0
     typefaceFontName=75 Medium
     outlineSettings {
      bApplyOutlineToDropShadows=false
      bMiteredCorners=false
      bSeparateFillAlpha=false
      outlineMaterial=null
      outlineSize=0
      outlineColor=<rgba 0 0 0 1>
     }
    }
    promptNudge=<-12 0>
    triggeringInputAction {dataTable=&InputActionDataTable, rowName=EWUseRelease}
   }
   slot {class=HorizontalBoxSlot, properties {horizontalAlignment=HAlign_Left, verticalAlignment=VAlign_Center}}
  }
  BottomButtonLeft_Divider {
   widgetClassPath=&Image
   bIsVariable=true
   parent=BottomBar_Left
   widget=BottomButtonLeft_Divider
   properties {
    renderOpacity=0.2
    brush {
     bIsDynamicallyLoaded=false
     drawAs=Image
     imageType=NoImage
     mirroring=NoMirror
     resourceName=None
     resourceObject=null
     tiling=NoTile
     imageSize=<2 28>
     margin=<ltrb 0 0 0 0>
     outlineSettings {
      bUseBrushTransparency=false
      roundingType=HalfHeightRadius
      width=0
      color {colorUseRule=UseColor_Specified, specifiedColor=<rgba 0 0 0 0>}
      cornerRadii=<0 0 0 0>
     }
     tintColor {colorUseRule=UseColor_Specified, specifiedColor=<rgba 1 1 1 1>}
     uVRegion {bIsValid=false, max=<0 0>, min=<0 0>}
    }
   }
   slot {class=HorizontalBoxSlot, properties {horizontalAlignment=HAlign_Left, padding=<ltrb 0 10 24 10>}}
  }
  QuickUseInput1 {
   widgetClassPath=&CommonActionWidget
   bIsVariable=true
   parent=BottomBar_Left
   widget=QuickUseInput1
   properties {
    progressMaterialParam=Percentage
    visibility=HitTestInvisible
    inputActions [{dataTable=&InputActionDataTable, rowName=EW_Slot_1}]
    progressMaterialBrush {
     bIsDynamicallyLoaded=false
     drawAs=Image
     imageType=NoImage
     mirroring=NoMirror
     resourceName=None
     resourceObject=/FortUILibrary/UIKit/Buttons/MaterialInstances/CTAButtons/MI_UIKit_EmptyProgress
     tiling=NoTile
     imageSize=<32 32>
     margin=<ltrb 0 0 0 0>
     outlineSettings {
      bUseBrushTransparency=false
      roundingType=HalfHeightRadius
      width=0
      color {colorUseRule=UseColor_Specified, specifiedColor=<rgba 0 0 0 0>}
      cornerRadii=<0 0 0 0>
     }
     tintColor {colorUseRule=UseColor_Specified, specifiedColor=<rgba 1 1 1 1>}
     uVRegion {bIsValid=false, max=<0 0>, min=<0 0>}
    }
   }
   slot {class=HorizontalBoxSlot, properties {horizontalAlignment=HAlign_Center, verticalAlignment=VAlign_Center, padding=<ltrb 0 0 2 0>}}
  }
  ActionPromptDivider {
   widgetClassPath=&Image
   bIsVariable=true
   parent=BottomBar_Left
   widget=ActionPromptDivider
   properties {
    renderOpacity=0.2
    brush {
     bIsDynamicallyLoaded=false
     drawAs=Image
     imageType=NoImage
     mirroring=NoMirror
     resourceName=None
     resourceObject=null
     tiling=NoTile
     imageSize=<2 6>
     margin=<ltrb 0 0 0 0>
     outlineSettings {
      bUseBrushTransparency=false
      roundingType=HalfHeightRadius
      width=0
      color {colorUseRule=UseColor_Specified, specifiedColor=<rgba 0 0 0 0>}
      cornerRadii=<0 0 0 0>
     }
     tintColor {colorUseRule=UseColor_Specified, specifiedColor=<rgba 1 1 1 1>}
     uVRegion {bIsValid=false, max=<0 0>, min=<0 0>}
    }
   }
   slot {class=HorizontalBoxSlot, properties {horizontalAlignment=HAlign_Left, verticalAlignment=VAlign_Center, padding=<ltrb 0 0 14 0>}}
  }
  BottomBarLeft_QuickUse {
   widgetClassPath=&WBP_UIKit_Button_Generic
   bIsVariable=true
   parent=BottomBar_Left
   widget=BottomBarLeft_QuickUse
   properties {
    bDisplayInputActionWhenNotInteractable=true
    bIsFocusable=false
    bLocked=true
    text=Quick Use
    useBackgroundBlock=false
    useIconBlock=false
    font {
     bForceMonospaced=false
     bMaterialIsStencil=false
     fontMaterial=&MI_UIKit_TextQuiet_Prompt
     fontObject=&HeadingNow
     letterSpacing=0
     monospacedWidth=1
     size=15
     skewAmount=0
     typefaceFontName=75 Medium
     outlineSettings {
      bApplyOutlineToDropShadows=false
      bMiteredCorners=false
      bSeparateFillAlpha=false
      outlineMaterial=null
      outlineSize=0
      outlineColor=<rgba 0 0 0 1>
     }
    }
    promptNudge=<-12 0>
    triggeringInputAction {dataTable=&InputActionDataTable, rowName=EW_Slot_8}
   }
   slot {class=HorizontalBoxSlot, properties {horizontalAlignment=HAlign_Left, verticalAlignment=VAlign_Center}}
  }
  DecoyBottomButtons {
   widgetClassPath=&HorizontalBox
   parent=GridPanel_18
   widget=DecoyBottomButtons
   slot {class=GridSlot, properties {horizontalAlignment=HAlign_Right, verticalAlignment=VAlign_Bottom}}
  }
  WBP_BottomBarDecoy_PM {
   widgetClassPath=/Game/UI/Frontend/BottomBar/WBP_BottomBarDecoyButton
   bIsVariable=true
   parent=DecoyBottomButtons
   widget=WBP_BottomBarDecoy_PM
   properties {visibility=Collapsed, triggeringInputAction {dataTable=/FortUILibrary/Input/ReplayInputActionDataTable, rowName=EnterPhotoMode}}
   slot {class=HorizontalBoxSlot}
  }
  WBP_BottomBarDecoy_Use {
   widgetClassPath=&WBP_BottomBarDecoyButton
   bIsVariable=true
   parent=DecoyBottomButtons
   widget=WBP_BottomBarDecoy_Use
   properties {triggeringInputAction {dataTable=&InputActionDataTable, rowName=EWUse}}
   slot {class=HorizontalBoxSlot, properties {horizontalAlignment=HAlign_Right, verticalAlignment=VAlign_Bottom}}
  }
  WBP_BottomBarDecoy_UseRelease {
   widgetClassPath=&WBP_BottomBarDecoyButton
   bIsVariable=true
   parent=DecoyBottomButtons
   widget=WBP_BottomBarDecoy_UseRelease
   properties {pressMethod=ButtonRelease, triggeringInputAction {dataTable=&InputActionDataTable, rowName=EWUseRelease}}
   slot {class=HorizontalBoxSlot, properties {horizontalAlignment=HAlign_Right, verticalAlignment=VAlign_Bottom}}
  }
  WBP_BottomBarDecoy_UseCommsRelease {
   widgetClassPath=&WBP_BottomBarDecoyButton
   bIsVariable=true
   parent=DecoyBottomButtons
   widget=WBP_BottomBarDecoy_UseCommsRelease
   properties {pressMethod=ButtonRelease, triggeringInputAction {dataTable=&InputActionDataTable, rowName=EWUseCommsRelease}}
   slot {class=HorizontalBoxSlot, properties {horizontalAlignment=HAlign_Right, verticalAlignment=VAlign_Bottom}}
  }
  WBP_BottomBarDecoy_Close {
   widgetClassPath=&WBP_BottomBarDecoyButton
   bIsVariable=true
   parent=DecoyBottomButtons
   widget=WBP_BottomBarDecoy_Close
   properties {triggeringInputAction {dataTable=&InputActionDataTable, rowName=EW_Close}}
   slot {class=HorizontalBoxSlot}
  }
  WBP_BottomBarDecoy_CloseEsc {
   widgetClassPath=&WBP_BottomBarDecoyButton
   bIsVariable=true
   parent=DecoyBottomButtons
   widget=WBP_BottomBarDecoy_CloseEsc
   properties {bDisplayInActionBar=false, triggeringInputAction {dataTable=&InputActionDataTable, rowName=Back}}
   slot {class=HorizontalBoxSlot}
  }
  WBP_BottomBarDecoy_GamepadUse {
   widgetClassPath=&WBP_BottomBarDecoyButton
   bIsVariable=true
   parent=DecoyBottomButtons
   widget=WBP_BottomBarDecoy_GamepadUse
   properties {bDisplayInActionBar=false, triggeringInputAction {dataTable=&InputActionDataTable, rowName=EWUse}}
   slot {class=HorizontalBoxSlot}
  }
  OverlayPanel {
   widgetClassPath=&Overlay
   bIsVariable=true
   parent=GridPanel_18
   widget=OverlayPanel
   slot {class=GridSlot, properties {layer=40}}
  }
  TouchSwipe_Border {
   widgetClassPath=&CommonHardwareVisibilityBorder
   parent=GridPanel_18
   widget=TouchSwipe_Border
   properties {visibilityQuery {
     autoDescription=" ANY( Input.Touch )"
     tokenStreamVersion=0
     userDescription=""
     queryTokenStream [0,1,1,1,0]
     tagDictionary [{tagName=Input.Touch}]
    }}
   slot {class=GridSlot, properties {layer=-1}}
  }
  Overlay_0 {
   widgetClassPath=&Overlay
   parent=TouchSwipe_Border
   widget=Overlay_0
   slot {class=BorderSlot, properties {padding=<ltrb 0 0 0 0>}}
  }
  TouchSwipePanel {
   widgetClassPath=/Script/FortniteUI.FortSwipePanel
   bIsVariable=true
   parent=Overlay_0
   widget=TouchSwipePanel
   properties {bAllowSwipeFromChildTouchMove=false, bBeginSwipeOnPointerEnter=false, bConsumePointerInput=false, swipeThreshold=<20 0>}
   slot {class=OverlaySlot, properties {horizontalAlignment=HAlign_Fill, verticalAlignment=VAlign_Fill}}
  }
  HardwareVisibility_PreviousButton {
   widgetClassPath=&CommonHardwareVisibilityBorder
   parent=GridPanel_18
   widget=HardwareVisibility_PreviousButton
   properties {
    horizontalAlignment=HAlign_Center
    verticalAlignment=VAlign_Center
    background {
     bIsDynamicallyLoaded=false
     drawAs=NoDrawType
     imageType=NoImage
     mirroring=NoMirror
     resourceName=None
     resourceObject=null
     tiling=NoTile
     imageSize=<56 56>
     margin=<ltrb 0 0 0 0>
     outlineSettings {
      bUseBrushTransparency=false
      roundingType=HalfHeightRadius
      width=0
      color {colorUseRule=UseColor_Specified, specifiedColor=<rgba 0 0 0 0>}
      cornerRadii=<0 0 0 0>
     }
     tintColor {colorUseRule=UseColor_Specified, specifiedColor=<rgba 1 1 1 1>}
     uVRegion {bIsValid=false, max=<0 0>, min=<0 0>}
    }
    visibilityQuery {
     autoDescription=" ANY( Input.Touch )"
     tokenStreamVersion=0
     userDescription=""
     queryTokenStream [0,1,1,1,0]
     tagDictionary [{tagName=Input.Touch}]
    }
   }
   slot {class=GridSlot, properties {horizontalAlignment=HAlign_Left, verticalAlignment=VAlign_Center, padding=<ltrb 120 0 0 50>}}
  }
  PreviousArrow_Button {
   widgetClassPath=/FortUILibrary/UIKit/Buttons/TabButtons/WBP_UIKit_Button_TabNavArrow
   bIsVariable=true
   parent=HardwareVisibility_PreviousButton
   widget=PreviousArrow_Button
   properties {
    iconMaterial=/FortUILibrary/Foundation/Buttons/Materials/Arrows/MI_UIKit_Icon_Button_Next
    minHeight=80
    minHeightOverride=56
    minWidth=80
    minWidthOverride=56
    mobileMinHeight=120
    mobileMinWidth=120
    iconSize=<55.5 55.5>
    promptNudge=<0 0>
    renderTransform {angle=0, scale=<-1 1>, shear=<0 0>, translation=<0 0>}
   }
   slot {class=BorderSlot, properties {horizontalAlignment=HAlign_Center, verticalAlignment=VAlign_Center, padding=<ltrb 0 0 0 0>}}
  }
  HardwareVisibility_NextButton {
   widgetClassPath=&CommonHardwareVisibilityBorder
   parent=GridPanel_18
   widget=HardwareVisibility_NextButton
   properties {
    horizontalAlignment=HAlign_Right
    verticalAlignment=VAlign_Center
    background {
     bIsDynamicallyLoaded=false
     drawAs=NoDrawType
     imageType=NoImage
     mirroring=NoMirror
     resourceName=None
     resourceObject=null
     tiling=NoTile
     imageSize=<56 56>
     margin=<ltrb 0 0 0 0>
     outlineSettings {
      bUseBrushTransparency=false
      roundingType=HalfHeightRadius
      width=0
      color {colorUseRule=UseColor_Specified, specifiedColor=<rgba 0 0 0 0>}
      cornerRadii=<0 0 0 0>
     }
     tintColor {colorUseRule=UseColor_Specified, specifiedColor=<rgba 1 1 1 1>}
     uVRegion {bIsValid=false, max=<0 0>, min=<0 0>}
    }
    visibilityQuery {
     autoDescription=" ANY( Input.Touch )"
     tokenStreamVersion=0
     userDescription=""
     queryTokenStream [0,1,1,1,0]
     tagDictionary [{tagName=Input.Touch}]
    }
   }
   slot {class=GridSlot, properties {horizontalAlignment=HAlign_Right, verticalAlignment=VAlign_Center, padding=<ltrb 0 0 120 50>}}
  }
  NextArrow_Button {
   widgetClassPath=&WBP_UIKit_Button_TabNavArrow
   bIsVariable=true
   parent=HardwareVisibility_NextButton
   widget=NextArrow_Button
   properties {
    iconMaterial=&MI_UIKit_Icon_Button_Next
    minHeightOverride=56
    minWidthOverride=56
    mobileMinHeight=120
    mobileMinWidth=120
    iconSize=<55.5 55.5>
   }
   slot {class=BorderSlot, properties {horizontalAlignment=HAlign_Right, verticalAlignment=VAlign_Center, padding=<ltrb 0 0 0 0>}}
  }
 ]
}
