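# Translucent UI material for the rarity label backing in the emote wheel, shaped as a rounded box with a CornerRadius parameter.
# At default quality it grades the Slate post buffer through tint, power, S-curve, saturation boost and hue shift, then darkens it heavily.
# Low, Medium and High quality swap in flat dark gray and lower box alpha.
# Color and ColorMix drive a color blend; Anim_Color blends two identical inputs, so has no effect.
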Material /EmoteWheel/Materials/M_UI_RarityLabel_Backing_GizmoWheel
graph {
 properties {bAutomaticallySetUsageInEditor=false, blendMode=BLEND_Translucent, materialDomain=MD_UI, usedShadingModels=Default Lit}
 nodes [
  n1 = Multiply(A=NamedRerouteUsage(), B=OneMinus(None=Constant4Vector() {constant=<rgba 0.0015 0.003 0.002 1>}.RGB))
  n2 = NamedRerouteUsage()
  n3 = Power(
   Base=Multiply(A=Add(A=@n1, B=@n2), B=Constant4Vector() {constant=<rgba 2 2 2.5 1>}.RGB)
   Exp=Divide(B=Constant4Vector() {constant=<rgba 1.55 1.45 1.666 1>}.RGB)
  )
  n4 = Constant() {r=0.9}
  n5 = Constant()
  n6 = Constant() {r=0.75}
  n7 = MaterialFunctionCall(
   Hue Shift Percentage=Constant()
   Texture=Desaturation(
    None=MaterialFunctionCall(In=@n3, Power=@n4) {materialFunction=&SCurve}.Result
    Fraction=Subtract(A=@n5, B=@n6)
   )
  ) {materialFunction=&HueShift}
  n8 = Constant() {r=0.07}
  n9 = Divide(A=Constant() {r=100}) {constB=100}
  n10 = Constant4Vector() {constant=<rgba 0.013 0.013 0.013 1>}
  n11 = QualitySwitch(
   Default=LinearInterpolate(A=NamedRerouteUsage(), B=Multiply(A=@n7.Result, B=@n8), Alpha=Saturate(None=@n9))
   Low=@n10.RGB
   High=@n10.RGB
   Medium=@n10.RGB
  )
  LinearInterpolate(A=@n11, B=@n11, Alpha=ScalarParameter() {parameterName=Anim_Color})
  LinearInterpolate(B=VectorParameter() {parameterName=Color, defaultValue=<rgba 0.0781 0.0781 0.0781 1>}.RGB, Alpha=ScalarParameter() {parameterName=ColorMix})
  n12 = Constant() {r=0.464}
  Multiply(
   A=MaterialFunctionCall(
    Size=Constant() {r=1}
    CornerRoundness=ScalarParameter() {defaultValue=0.25, parameterName=CornerRadius}
    KeepAspectRatio=StaticBool() {value=true}
   ) {materialFunction=&MF_UI_Box}.Fill
   B=QualitySwitch(Default=Constant() {r=0.75}, Low=@n12, High=@n12, Medium=@n12)
  ) {constB=0.75}
  NamedRerouteDeclaration(None=MaterialFunctionCall!("UVs (Optional)") {materialFunction=/Engine/Functions/UserInterface/GetSlatePost0}.LinearRGB) {name=Buffer, nodeColor=<rgba 0 1 0.9297 1>}
 ]
}
