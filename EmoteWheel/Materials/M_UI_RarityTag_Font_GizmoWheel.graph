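# M_UI_RarityTag_Font_GizmoWheel is a translucent UI material for the emote wheel, apparently coloring rarity tag text with a two-color gradient.
# A linear gradient function builds a mask from texture coordinates, with parameters for start, stop and rotation angle.
# The saturated mask passes through an EaseInOutSine curve for a smooth falloff, then blends between the Color_1 and Color_2 parameters.
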Material /EmoteWheel/Materials/M_UI_RarityTag_Font_GizmoWheel
graph {
 properties {bAutomaticallySetUsageInEditor=false, blendMode=BLEND_Translucent, materialDomain=MD_UI, usedShadingModels=Default Lit}
 nodes [
  n1 = Saturate(None=MaterialFunctionCall(
    UVs=TextureCoordinate()
    GradientStart=ScalarParameter() {defaultValue=-0.25, parameterName=Start}
    GradientEnd=ScalarParameter() {defaultValue=1.333, parameterName=Stop}
    RotationAngle=ScalarParameter() {defaultValue=-0.125, parameterName=RotationAngle}
   ) {materialFunction=&MF_UI_LinearGradient}.Result)
  LinearInterpolate(
   A=VectorParameter() {parameterName=Color_1, defaultValue=<rgba 1 1 1 1>}.RGB
   B=VectorParameter() {parameterName=Color_2, defaultValue=<rgba 0.5 0.5 0.5 1>}.RGB
   Alpha=Saturate(None=MaterialFunctionCall("Value (0-1)"=@n1) {materialFunction=&MF_UI_EaseCurves}.EaseInOutSine)
  )
 ]
}
