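# Gizmo variant of the emote radial wheel widget, showing a category's equipped gizmo items as radial slots.
# PreConstruct creates a dynamic material for the segmented background.
# PopulateWheelEntries rebuilds entries from the category, configures each from its view model and config, and restores selection through a mirrored index.
# SetWheelSlotColors drives per-slot material parameters by name array, coloring by rarity or series and highlighting Uncommon through Transcendent items, except slot zero.
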WidgetBlueprint /EmoteWheel/WBP_Gizmo_RadialWheel
graphs [
 EventGraph {nodes [
   n1 = Event PreConstruct()
   > Parent: PreConstruct(IsDesignTime=@n1.IsDesignTime)
   n2 = Get BackgroundWithSegments() -> (BackgroundWithSegments:Image Object Reference)
   > SetBrushFromMaterial(self=@n2, Material=Get MI_GizmoWheel() -> (MI_GizmoWheel:Material Interface Object Reference))
   > n3 = GetDynamicMaterial(self=@n2)
   > Set MID_BackgroundWithSegments(MID_BackgroundWithSegments:Material Instance Dynamic Object Reference=@n3) -> (Output_Get:Material Instance Dynamic Object Reference)
   Event Construct()
   > Parent: Construct()
   Event Tick()
  ]}
 PopulateWheelEntries {nodes [
   n1 = PopulateWheelEntries!() -> (then, Category:Emote Wheel Category VM Object Reference) {is=title K2Node_FunctionEntry}
   > Is Valid(InputObject=@n1)
   > UnbindEmoteEntries(self:WBP Emote Radial Wheel Object Reference)
   > Reset!(
    execute
    self:Dynamic Entry Box Object Reference=Get EntryBox_ContextRadial() -> (EntryBox_ContextRadial:Dynamic Entry Box Object Reference)
    bDeleteWidgets:bool=true
   ) -> (then)
   > n2 = For Each Loop(Array:Array of Emote Wheel Item VM Object References=GetEquippedItems!(self:Emote Wheel Category VM Object Reference=Get Category() -> (Category:Emote Wheel Category VM Object Reference)) -> (ReturnValue:Array of Emote Wheel Item VM Object References)) -> (Array Element:Emote Wheel Item VM Object Reference)
   RefreshRadialData(execute=@n2.Completed, self:Fort Radial Menu Object Reference)
   n3 = Get EmoteWheelItemSelectionVM() -> (EmoteWheelItemSelectionVM:Emote Wheel Item Selection VM Object Reference)
   > SelectItemAtIndex!(
    execute
    self:Fort Radial Menu Object Reference=Self-Reference()
    Index:int=GetMirroredIndex!(Index:int=Get SelectedSlotIndex!(self:Emote Wheel Item Selection VM Object Reference=@n3) -> (SelectedSlotIndex:int=0)) -> (MirroredIndex:int=0)
   ) -> (then)
   n4 = Create Entry of Class!(
    execute=@n2.LoopBody
    self:Dynamic Entry Box Object Reference=Get EntryBox_ContextRadial()
    EntryClass:User Widget Class Reference=&/WBP_Gizmo_RadialEntry.WBP_Gizmo_RadialEntry_C
   ) -> (then, ReturnValue:WBP Gizmo Radial Entry Object Reference)
   > n5 = Cast To GizmosWheelCategoryVM!(
    execute
    Object=Get SelectedCategory!(self:Emote Wheel Item Selection VM Object Reference=Get EmoteWheelItemSelectionVM()) -> (SelectedCategory:Emote Wheel Category VM Object Reference)
   ) -> (then, CastFailed, AsGizmos Wheel Category VM:Gizmos Wheel Category VM Object Reference) {is=title K2Node_DynamicCast}
   n6 = GetEquippedItems(self=@n5)
   > Set SlotCount!(
    execute
    SlotCount:int=Length(TargetArray:Array of Emote Wheel Item VM Object References=@n6)
    self:WBP Gizmo Radial Entry Object Reference=@n4
   ) -> (then, Output_Get:int=0)
   > n7 = Cast To GizmosWheelItemVM!(execute, Object=@n2) -> (then, CastFailed, AsGizmos Wheel Item VM:Gizmos Wheel Item VM Object Reference) {is=title K2Node_DynamicCast}
   > SetGizmosWheelItemVM!(execute, self:WBP Gizmo Radial Entry Object Reference=@n4, Viewmodel:Gizmos Wheel Item VM Object Reference=@n7) -> (then)
   > SetWheelSlotColors!(
    execute
    self
    ItemDefinition:Item Definition Base Object Reference=Get ItemDefinition!(self:Gizmos Wheel Item VM Object Reference=@n7) -> (ItemDefinition:Fort World Item Definition Object Reference)
    Index:int=@n2.Array Index
   ) -> (then)
   > SetEmoteIndex!(execute, self:WBP Gizmo Radial Entry Object Reference=@n4, ItemIndex:int=@n2.Array Index) -> (then)
   n8 = Get Category()
   > SetIconScale!(
    execute
    self:WBP Gizmo Radial Entry Object Reference=@n4
    ScaleFactor:double=Get TextureScaleFactor!(self:Emote Wheel Category Config Object Reference=GetConfig!(self:Emote Wheel Category VM Object Reference=@n8) -> (ReturnValue:Emote Wheel Category Config Object Reference)) -> (TextureScaleFactor:float)
   ) -> (then)
   n9 = Get EmoteWheelItemSelectionVM()
   > SetTruncationTitles!(
    execute
    self:WBP Gizmo Radial Entry Object Reference=@n4
    ShouldTruncateTitles:bool=ShouldTruncateTitles!(self:Emote Wheel Category VM Object Reference=Get SelectedCategory(self=@n9)) -> (ReturnValue:bool)
   ) -> (then)
   n10 = Get Category()
   > SetAllowInteractionWithEntryText!(
    execute
    self:WBP Gizmo Radial Entry Object Reference=@n4
    bAllowInteraction:bool=Get bAllowInteractionWithItemNames!(self:Emote Wheel Category Config Object Reference=GetConfig(self=@n10)) -> (bAllowInteractionWithItemNames:bool)
   ) -> (then)
   > Bind Event to On Entry Clicked(self:WBP Gizmo Radial Entry Object Reference=@n4, Delegate=Create Event()) {is=title K2Node_AddDelegate}
   "Get (a copy)"(Array:Array of Emote Wheel Item VM Object References=@n6, Dimension 1=@n2.Array Index) -> (Output:Emote Wheel Item VM Object Reference)
  ]}
 SetWheelSlotColors {nodes [
   n1 = SetWheelSlotColors!() -> (then, ItemDefinition:Item Definition Base Object Reference, Index:int) {is=title K2Node_FunctionEntry}
   n2 = IsValid(Object=@n1)
   n3 = "Not Equal (Integer)"(A=Get Index() -> (Index:int=0))
   n4 = Select!(
    Common:bool=false
    Uncommon:bool=true
    Rare:bool=true
    Epic:bool=true
    Legendary:bool=true
    Mythic:bool=true
    Transcendent:bool=true
    Unattainable:bool=false
    Index:EFortRarity Enum=GetRarity!(ItemDefinition:Item Definition Base Object Reference=Get ItemDefinition!() -> (ItemDefinition:Item Definition Base Object Reference)) -> (ReturnValue:EFortRarity Enum=Common)
   ) -> (ReturnValue:bool)
   > SetScalarParameterValue(
    self=Get MID_BackgroundWithSegments() -> (MID_BackgroundWithSegments:Material Instance Dynamic Object Reference)
    ParameterName="Get (a ref)"(
     Array:Array of Names=Get SlotHighlightEnableParameterNames() -> (SlotHighlightEnableParameterNames:Array of Names)
     Dimension 1=Get Index() -> (Index=0)
    ) -> (Output:"Name (by ref)")
    Value="To Float (Boolean)"(InBool=AND Boolean!(A=@n2, B=@n4, C:bool=@n3) -> (ReturnValue))
   )
   > SetVectorParameterValue(
    self=Get MID_BackgroundWithSegments()
    ParameterName="Get (a ref)"(Array=Get SlotColorParameterNames() -> (SlotColorParameterNames:Array of Names), Dimension 1=Get Index() -> (Index=0))
    Value=GetRarityOrSeriesColors!(ItemDefinition:Item Definition Base Object Reference=Get ItemDefinition()) -> (
     ReturnValue_Color1:Linear Color Structure
     ReturnValue_Color2:Linear Color Structure
     ReturnValue_Color3:Linear Color Structure
     ReturnValue_Color4:Linear Color Structure
     ReturnValue_Color5:Linear Color Structure
    ).ReturnValue_Color2
   )
   > Return Node()
  ]}
]
variables [
 MI_GizmoWheel {type=MaterialInterface, default=/EmoteWheel/MaterialInstances/MI_UI_GizmoWheel_Slots}
 MID_BackgroundWithSegments {type=MaterialInstanceDynamic}
 SlotColorParameterNames {type="array<string>", default [SlotColor0,SlotColor1,SlotColor2,SlotColor3,SlotColor4,SlotColor5,SlotColor6,SlotColor7]}
 SlotHighlightEnableParameterNames {
  type="array<string>"
  default [
   IsHighlightSlot0
   IsHighlightSlot1
   IsHighlightSlot2
   IsHighlightSlot3
   IsHighlightSlot4
   IsHighlightSlot5
   IsHighlightSlot6
   IsHighlightSlot7
  ]
 }
]
widgets {
 info {inheritedWidgetCount=0, parentClass=&/WBP_Emote_RadialWheel.WBP_Emote_RadialWheel_C, rootWidgetClass=null, widgetCount=0}
 widgets []
}
