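# Translucent UI material for a circle in the Activity Browser frontend's clock art, named as a white circle.
# A RadialGradientExponential function, driven by Radius and Density scalar parameters, generates the circular shape, with a high default density that tightens the falloff.
# An RGBA vector parameter provides an adjustable color, defaulting to pinkish red rather than white.
# The graph shows no links from either the gradient or the color to the material outputs.
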Material /Game/Athena/UI/Frontend/ActivityBrowser/Art/M_UI_Clock_WhiteCirlce
graph {
 properties {
  bUsedWithStaticMesh=false
  blendMode=BLEND_Translucent
  materialDomain=MD_UI
  pixelDepthOffsetMode=PDOM_Legacy
  usedShadingModels=Default Lit
 }
 nodes [
  MaterialFunctionCall(Radius=ScalarParameter() {defaultValue=0.5, parameterName=Radius}, Density=ScalarParameter() {defaultValue=32, parameterName=Density}) {materialFunction=&RadialGradientExponential}
  VectorParameter() {parameterName=RGBA, defaultValue=<rgba 1 0 0.1022 1>}
 ]
}
