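# Activatable screen for the Gizmos wheel, a radial menu that selects and uses gizmo items, showing name, rarity, type and expiry.
# View models drive item details, while dynamic material parameters set segment count, selection and rarity color.
# Releasing a bound input uses the selection unless the menu is sticky, as when the on-demand collection widget opens.
# Activation sets stick control and hides HUD elements; button prompts follow input type, and interactions send analytics.
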WidgetBlueprint /EmoteWheel/WBP_GizmosWheel_Screen
animations [
 Anim_Collection_ExpandContract {
  binding_count=4
  length_seconds=0.15
  object=&self:Anim_Collection_ExpandContract
  sequence {bindings [
    WBP_RadialWheel {
     id=#1
     possessed_class=WBP_Emote_RadialWheel_C
     tracks [
      {
       type=MovieSceneFloatTrack
       display_name=Render Opacity
       sections [{type=MovieSceneFloatSection, end_frame=3, start_frame=0, channels [None {type=MovieSceneScriptingFloatChannel, keys [[0.3333,1],[3,0]]}]}]
      }
      {
       type=MovieSceneByteTrack
       display_name=Visibility
       sections [{type=MovieSceneByteSection, end_frame=3, start_frame=0, channels [None {type=MovieSceneScriptingByteChannel, keys [[0,4],[2.6667,4],[3,1]]}]}]
      }
     ]
    }
    Background_BlockClick_BlurOnlyHighEndDevice {
     id=#2
     possessed_class=Image
     tracks [{
       type=MovieSceneWidgetMaterialTrack
       sections [{
         type=MovieSceneParameterSection
         end_frame=3
         start_frame=0
         channels [Center Opacity {type=MovieSceneScriptingFloatChannel, keys [[0,0.5],[3,1]]}]
        }]
      }]
    }
    GridSlot_19 {
     id=#3
     parent_id=#1
     possessed_class=GridSlot
     tracks [{
       type=MovieSceneDoubleVectorTrack
       display_name=Nudge
       sections [{
         type=MovieSceneDoubleVectorSection
         end_frame=3
         start_frame=0
         channels [
          Vector.X {type=MovieSceneScriptingDoubleChannel, const [0,1]}
          Vector.Y {type=MovieSceneScriptingDoubleChannel, keys [[0,0],[3,-900]]}
         ]
        }]
      }]
    }
    WBP_EmotePicker_Screen_C_105 {
     id=#4
     possessed_class=WBP_EmotePicker_Screen_C
     tracks [{
       type=MovieSceneDoubleTrack
       display_name=Lerp Top to Bottom Of Screen Height
       sections [{type=MovieSceneDoubleSection, end_frame=3, start_frame=0, channels [None {type=MovieSceneScriptingDoubleChannel, keys [[0,0],[3,1]]}]}]
      }]
    }
   ]}
 }
 Intro {binding_count=0, length_seconds=0.25, object=&self:Intro, sequence {bindings []}}
 Outro {
  binding_count=2
  length_seconds=0.05
  object=&self:Outro
  sequence {bindings [
    BG_Darkener {
     id=#5
     possessed_class=Image
     tracks [{
       type=MovieSceneFloatTrack
       display_name=Render Opacity
       sections [{type=MovieSceneFloatSection, end_frame=1, start_frame=0, channels [None {type=MovieSceneScriptingFloatChannel, keys [[0,1],[1,0]]}]}]
      }]
    }
    Background_BlockClick_BlurOnlyHighEndDevice {
     id=#6
     possessed_class=Image
     tracks [{
       type=MovieSceneFloatTrack
       display_name=Render Opacity
       sections [{type=MovieSceneFloatSection, end_frame=1, start_frame=0, channels [None {type=MovieSceneScriptingFloatChannel, keys [[0,1],[1,0]]}]}]
      }]
    }
   ]}
 }
 BlurFade {
  binding_count=2
  length_seconds=1.15
  object=&self:BlurFade
  sequence {bindings [
    BG_Darkener {
     id=#5
     possessed_class=Image
     tracks [{
       type=MovieSceneFloatTrack
       display_name=Render Opacity
       sections [{type=MovieSceneFloatSection, end_frame=23, start_frame=0, channels [None {type=MovieSceneScriptingFloatChannel, keys [[8,0],[23,1]]}]}]
      }]
    }
    Background_BlockClick_BlurOnlyHighEndDevice {
     id=#6
     possessed_class=Image
     tracks [{
       type=MovieSceneFloatTrack
       display_name=Render Opacity
       sections [{type=MovieSceneFloatSection, end_frame=23, start_frame=0, channels [None {type=MovieSceneScriptingFloatChannel, keys [[8,0],[23,1]]}]}]
      }]
    }
   ]}
 }
]
graphs [
 BP_GetDesiredFocusTarget {nodes [
   Get Desired Focus Target()
   > Return Node!(execute, ReturnValue:Widget Object Reference=Get WBP_RadialWheel() -> (WBP_RadialWheel:WBP Gizmo Radial Wheel Object Reference))
  ]}
 BP_GetDesiredInputConfig {nodes [
   Get Desired Input Config() {is=title K2Node_FunctionEntry}
   > Return Node(ReturnValue:UIInput Config Structure=Make UIInput Config!(
     bIgnoreMoveInput:bool=False
     bIgnoreLookInput:bool=true
     InputMode:ECommonInputMode Enum=All
     MouseCaptureMode:EMouseCaptureMode Enum=CaptureDuringRightMouseDown
     MouseLockMode:EMouseLockMode Enum=DoNotLock
     bHideCursorDuringViewportCapture:bool=True
    ) -> (UIInputConfig:UIInput Config Structure) {is=title K2Node_MakeStruct})
  ]}
 CenterMouseToViewport {nodes [
   CenterMouseToViewport() {is=title K2Node_FunctionEntry}
   n1 = vector2d * vector2d!(A:Vector 2D Structure=GetViewportSize!() -> (ReturnValue:Vector 2D Structure), B:Vector 2D Structure=(X=0.5,Y=0.5)) -> (ReturnValue_X:double, ReturnValue_Y:double) {is=title K2Node_PromotableOperator}
   > Set Mouse Position!(execute, self:Player Controller Object Reference=GetOwningPlayer(), X:int=Truncate(A=@n1), Y:int=Truncate(A=@n1.ReturnValue_Y)) -> (then) {is=title K2Node_CallFunction}
  ]}
 EventGraph {nodes [
   n1 = "On Clicked (WBP_BottomBarDecoy_Close)"!() -> (OutputDelegate, then, Button:Common Button Base Object Reference) {is=title K2Node_ComponentBoundEvent}
   > n2 = TriggerFlowriderInteraction!(execute, self, Element:Widget Object Reference=@n1.Button, Value:String=Close Wheel) -> (then)
   n3 = Get!(execute:Exec=[@n2.then, @n19.then]) -> (then:Exec, else:Exec, LockerCollectionWidget:Common Activatable Widget Object Reference)
   > DeactivateWidget()
   > n4 = UnLoadCollection()
   DeactivateWidget(execute=[@n4.then, @n3.else])
   n5 = "On Entry Clicked (WBP_RadialWheel)"!() -> (OutputDelegate, then, Item:Emote Wheel Item VM Object Reference) {is=title K2Node_ComponentBoundEvent}
   UseItem(execute=[@n9.then, @n11.then, @n7.then, @n5.then])
   > TriggerFlowriderInteraction(Element=Get ClickedUseButton() -> (ClickedUseButton:Common Button Base Object Reference), Value=Use Item)
   n6 = "On Clicked (WBP_BottomBarDecoy_GamepadUse)"!() -> (OutputDelegate, then, Button:Common Button Base Object Reference) {is=title K2Node_ComponentBoundEvent}
   > n7 = Set ClickedUseButton(ClickedUseButton:Common Button Base Object Reference=@n6.Button) -> (Output_Get:Common Button Base Object Reference)
   n8 = "On Clicked (WBP_BottomBarDecoy_UseRelease)"!() -> (OutputDelegate, then, Button:Common Button Base Object Reference) {is=title K2Node_ComponentBoundEvent}
   > n9 = Set ClickedUseButton(ClickedUseButton=@n8.Button)
   n10 = "On Clicked (WBP_BottomBarDecoy_Use)"!() -> (OutputDelegate, then, Button:Common Button Base Object Reference) {is=title K2Node_ComponentBoundEvent}
   > n11 = Set ClickedUseButton(ClickedUseButton=@n10.Button)
   n12 = "On Clicked (WBP_BottomBarDecoy_CloseEsc)"!() -> (OutputDelegate, then, Button:Common Button Base Object Reference) {is=title K2Node_ComponentBoundEvent}
   > TriggerFlowriderInteraction(Element=@n12.Button, Value=Close Wheel)
   > CloseMenu()
   OnLoadUnloadCollection() {is=title K2Node_CustomEvent}
   > n13 = Branch(Condition=Get IsCollectionExpanded() -> (IsCollectionExpanded:bool))
   > AsyncLoadCollection()
   > ActivateWidget(self=Get LockerCollectionWidget() -> (LockerCollectionWidget:Common Activatable Widget Object Reference))
   UnLoadCollection(execute=@n13.else)
   > n14 = Cast To WBP_EW_Collection!(execute, Object=Get LockerCollectionWidget()) -> (then, CastFailed, AsWBP EW Collection:WBP EW Collection Object Reference) {is=title K2Node_DynamicCast}
   > Unbind Event from On Close Menu(self:WBP EW Collection Object Reference=@n14, Delegate=Create Event()) {is=title K2Node_RemoveDelegate}
   > DeactivateWidget(self=Get LockerCollectionWidget())
   > SetBottomButtonVisibility!(
    execute
    self:WBP Bottom Bar Decoy Button Object Reference=Get WBP_BottomBarDecoy_GamepadUse() -> (WBP_BottomBarDecoy_GamepadUse:WBP Bottom Bar Decoy Button Object Reference)
    ShouldButtonBeVisible:bool=true
   ) -> (then)
   AsyncLoadCollection() {is=title K2Node_CustomEvent}
   > Set IsMenuSticky(IsMenuSticky:bool=true) -> (Output_Get:bool)
   > HandleInputChanged!(execute, self, bNewInputType:ECommonInputType Enum=GetCurrentInputType(self=Get CommonInputSubsystem())) -> (then)
   > RemoveWidget!(execute, Widget:Widget Object Reference=Get LockerCollectionWidget()) -> (then)
   > n15 = Create Widget!(
    execute:Exec
    Class:User Widget Class Reference=Get CollectionClass() -> (CollectionClass:Common Activatable Widget Class Reference)
    OwningPlayer:Player Controller Object Reference=GetOwningPlayer()
   ) -> (then:Exec, ReturnValue:Common Activatable Widget Object Reference)
   > n16 = Cast To WBP_EW_Collection(Object=@n15)
   > SetEmoteWheelVM!(
    execute
    self:WBP EW Collection Object Reference=@n16
    Viewmodel:Emote Wheel VM Object Reference=Get GizmosWheelVM() -> (GizmosWheelVM:Gizmos Wheel VM Object Reference)
   ) -> (then)
   > Bind Event to On Close Menu(self:WBP EW Collection Object Reference=@n16, Delegate=Create Event()) {is=title K2Node_AddDelegate}
   > n17 = Set LockerCollectionWidget(LockerCollectionWidget:Common Activatable Widget Object Reference=@n16) -> (Output_Get:Common Activatable Widget Object Reference)
   > Branch(Condition=IsValid(Object=@n17))
   n18 = SlotAsOverlaySlot(Widget=@n17)
   > SetVerticalAlignment(self=@n18, InVerticalAlignment=VAlign_Fill)
   > SetHorizontalAlignment(self=@n18, InHorizontalAlignment=HAlign_Fill)
   > SetVisibility(InVisibility=Collapsed)
   > SetBottomButtonVisibility(self=Get WBP_BottomBarDecoy_GamepadUse(), ShouldButtonBeVisible=false)
   > SetVisibility(
    self=[
     Get BottomButtonLeft_Divider() -> (BottomButtonLeft_Divider:Image Object Reference)
     Get BottomBarLeft_UseRelease() -> (BottomBarLeft_UseRelease:WBP UIKit Button Generic Object Reference)
    ]
    InVisibility=Collapsed
   )
   n19 = CloseMenu() {is=title K2Node_CustomEvent}
   n20 = Event PreConstruct()
   > n21 = Sequence()
   > Branch(Condition=@n20.IsDesignTime)
   > SetRenderOpacity(
    self=Get Background_BlockClick_BlurOnlyHighEndDevice() -> (Background_BlockClick_BlurOnlyHighEndDevice:Image Object Reference)
    InOpacity=0.3
   )
   SetIsInteractionEnabled!(
    execute=@n21.then_1
    self:Common Button Base Object Reference=[Get BottomBarLeft_QuickUse() -> (BottomBarLeft_QuickUse:WBP UIKit Button Generic Object Reference), Get BottomBarLeft_UseRelease()]
    bInIsInteractionEnabled:bool=false
   ) -> (then)
   > SetMobileStyle() {is=title K2Node_CallFunction}
   Event Construct()
   n22 = GetAnalyticsEWActivationCount!(self:Fort Picker Context Object Reference=Get FortPickerContext() -> (ReturnValue:Fort Picker Context Object Reference) {is=title K2Node_GetSubsystem}) -> (ReturnValue:int=0)
   > Fire Event Main View Opened!(
    execute
    View:Widget Object Reference=Self-Reference()
    AdditionalInfo:Flow Analytics Additional Info Structure="Add String (Non-Localized)"!(
     Info:"Flow Analytics Additional Info Structure (by ref)"=Make Flow Analytics Additional Info!() -> (FlowAnalyticsAdditionalInfo:Flow Analytics Additional Info Structure) {is=title K2Node_MakeStruct}
     Key:String=emote_wheel_session_id
     Value:String=GetAnalyticsSessionID!(ActivationCount:int=@n22) -> (ReturnValue:String)
    ) -> (ReturnValue:Flow Analytics Additional Info Structure)
    bIsOverlay:bool=false
   ) -> (then)
   > BindToAnimationFinished(
    Animation=Get Anim_Collection_ExpandContract() -> (Anim_Collection_ExpandContract:Widget Animation Object Reference)
    Delegate=Create Event()
   )
   > Set IsMenuSticky(IsMenuSticky=false)
   > SetVisibility(self=[Get BottomBarLeft_UseRelease(), Get BottomButtonLeft_Divider()], InVisibility=SelfHitTestInvisible)
   > HandleInputChanged(bNewInputType=GetCurrentInputType(self=Get CommonInputSubsystem()))
   > SetAllInteractionsEnabled!(execute, self, IsInteractionEnabled:bool=false) -> (then)
   > n23 = GetDynamicMaterial(self=Get Image_Timer() -> (Image_Timer:Image Object Reference))
   > SetScalarParameterValue(self=@n23, ParameterName=TimerSpeedMult, Value=0)
   > PlayAnimationForward(InAnimation=Get BlurFade() -> (BlurFade:Widget Animation Object Reference))
   n24 = Get EmoteWheelItemSelectionVM() -> (EmoteWheelItemSelectionVM:Emote Wheel Item Selection VM Object Reference)
   > Is Valid(InputObject=@n24)
   n25 = Get EmoteWheelItemSelectionVM()
   > SetScalarParameterValue(
    self=Get MID_BackgroundWithSegments!(self:WBP Gizmo Radial Wheel Object Reference=Get WBP_RadialWheel()) -> (MID_BackgroundWithSegments:Material Instance Dynamic Object Reference)
    ParameterName=SelectedIndex
    Value="To Float (Integer)"(InInt=Get SelectedSlotIndex!(self:Emote Wheel Item Selection VM Object Reference=@n25) -> (SelectedSlotIndex:int=0))
   )
   > UpdateSelectedItem!(
    execute
    self
    Item:Emote Wheel Item VM Object Reference=Get SelectedItem!(self:Emote Wheel Item Selection VM Object Reference=@n24) -> (SelectedItem:Emote Wheel Item VM Object Reference)
   ) -> (then)
   > SetAllInteractionsEnabled(IsInteractionEnabled=true)
   Event Destruct()
   > OnViewDestruct(self:Emote Wheel VM Object Reference=Get GizmosWheelVM())
   > UnbindFromAnimationFinished!(
    execute
    self:User Widget Object Reference
    Animation:Widget Animation Object Reference=Get Anim_Collection_ExpandContract()
    Delegate:Delegate=Create Event()
   ) -> (then)
   > SetAllInteractionsEnabled(IsInteractionEnabled=false)
   > Check and Stop Animation!(execute:Exec, Animation:Widget Animation Object Reference=Get Intro()) -> (then:Exec) {is=title K2Node_MacroInstance}
   > Check and Stop Animation(Animation=Get BlurFade())
   > n26 = Play Animation with Finished event!(
    execute:Exec
    Widget:User Widget Object Reference=Self-Reference()
    InAnimation:Widget Animation Object Reference=Get Outro() -> (Outro:Widget Animation Object Reference)
    StartAtTime:float=0
    NumLoopsToPlay:int=1
    PlayMode:EUMGSequencePlayMode Enum
    PlaybackSpeed:float=1
   ) -> (then:Exec, Result:Widget Animation Handle Structure, Finished:Exec) {is=title K2Node_PlayAnimation2}
   SetAllInteractionsEnabled(execute=@n26.Finished, IsInteractionEnabled=true)
   Event On Activated()
   > n27 = Switch on Int!(execute:Exec, Selection:int=GetGamepadInputScheme!() -> (ReturnValue:int=0)) -> (Default:Exec, 0:Exec, 1:Exec)
   n28 = Get WBP_RadialWheel()
   SetControllingStick!(execute=@n27.0, self:Fort Radial Menu Object Reference=@n28, NewControllingStick:EFortRadialControllingStick Enum=Right) -> (then)
   > SetForceCloseOnReleaseEnabled!(execute, self:Emote Wheel Picker Screen Object Reference, bEnabled:bool=false) -> (then)
   > n29 = Set IsUsingRightStick(IsUsingRightStick:bool=true) -> (Output_Get:bool)
   LoadLastSelected!(
    execute=[@n29.then, @n36.then, @n27.Default]
    self:Emote Wheel Item Selection VM Object Reference=Get EmoteWheelItemSelectionVM()
    Item:Emote Wheel VM Object Reference=Get GizmosWheelVM()
   ) -> (then)
   > ToggleHUDVisibility!(execute, self, bHideElements:bool=true) -> (then)
   > SetBottomButtonVisibility(self=Get WBP_BottomBarDecoy_GamepadUse(), ShouldButtonBeVisible=true)
   n30 = Create Event()
   n31 = Get CommonInputSubsystem()
   > Bind Event to On Input Method Changed(self=@n31, Delegate=@n30)
   > CenterMouseToViewport() {is=title K2Node_CallFunction}
   > n32 = Cast To FortPlayerController(Object=GetOwningPlayer())
   > SetPickerOpened!(execute, self:Fort Player Controller Object Reference=@n32, bIsOpened:bool=true) -> (then)
   > PlaySound2D(Sound=/Game/Sounds/Fort_UI_Sounds/FrontEnd/New_FrontEnd/Fortnite_UI_MainMenu_Generic_02_Cue)
   > n33 = IsPluginMounted!(execute, PluginName:String=FN_Photomode) -> (then, ReturnValue:bool)
   > SetBottomButtonVisibility(self=Get WBP_BottomBarDecoy_PM() -> (WBP_BottomBarDecoy_PM:WBP Bottom Bar Decoy Button Object Reference), ShouldButtonBeVisible=@n33)
   > Branch(Condition=@n33)
   > Broadcast Event!(
    execute
    self:Gameplay Event Router Component Object Reference=GetGlobalEventRouter()
    EventData:Input Event Structure=Make Input Event!() -> (InputEvent:Input Event Structure) {is=title K2Node_MakeStruct}
    EventContext:Object Reference=GetOwningPlayer()
   ) -> (then, ReturnValue:bool)
   SetControllingStick(execute=@n27.1, self=@n28, NewControllingStick=Left)
   > SetForceCloseOnReleaseEnabled(bEnabled=true)
   n34 = Get CommonInputSubsystem()
   > n35 = Branch(Condition="Not Equal (Enum)"!(A:ECommonInputType Enum=GetCurrentInputType(self=@n34), B:ECommonInputType Enum=Gamepad) -> (ReturnValue:bool))
   n36 = Set IsUsingRightStick(execute=[@n37.then, @n35.then], IsUsingRightStick=false)
   n37 = SetIgnoreMoveInput!(execute=@n35.else, self:Controller Object Reference=GetOwningPlayer(), bNewMoveInput:bool=true) -> (then)
   Event On Deactivated()
   > n38 = Branch(Condition=Get IsUsingRightStick() -> (IsUsingRightStick:bool))
   IncrementAnalyticsEWActivationCount(execute=[@n40.then, @n38.then], self:Fort Picker Context Object Reference=Get FortPickerContext())
   > ToggleHUDVisibility(bHideElements=false)
   > Unbind Event from On Input Method Changed(self=@n31, Delegate=@n30)
   > FlushInputKeys(self:Emote Wheel VM Object Reference=Get GizmosWheelVM())
   > n39 = Cast To FortPlayerController(Object=GetOwningPlayer())
   > SetPickerOpened(self=@n39, bIsOpened=false)
   > PlaySound2D(Sound=/Game/Sounds/Fort_UI_Sounds/FrontEnd/New_FrontEnd/Fortnite_UI_MainMenu_Generic_03_Cue)
   n40 = SetIgnoreMoveInput(execute=@n38.else, self=GetOwningPlayer(), bNewMoveInput=false)
  ]}
 GetFormattedTimespan {nodes [
   n1 = GetFormattedTimespan!() -> (then, Timespan:Timespan Structure) {is=title K2Node_FunctionEntry}
   n2 = Get Hours!(A:Timespan Structure=@n1) -> (ReturnValue:int=0) {is=title K2Node_CallFunction}
   n3 = Format Text!(Format="LOCTEXT(\"{HOURS}H\")", HOURS:int=@n2) -> (Result)
   n4 = Format Text!(Format="LOCTEXT(\"{MINUTES}M\")", MINUTES:int=Get Minutes!(A:Timespan Structure=@n1) -> (ReturnValue:int=0) {is=title K2Node_CallFunction}) -> (Result)
   n5 = "integer > integer"(A=@n2)
   n6 = Get Days!(A:Timespan Structure=@n1) -> (ReturnValue:int=0) {is=title K2Node_CallFunction}
   > Return Node!(
    execute
    Text:Text=SelectText(
     A=Format Text!(Format="LOCTEXT(\"{DAYS}D\")", DAYS:int=@n6) -> (Result)
     B=SelectText(A=@n3, B=@n4, bPickA=@n5)
     bPickA="integer > integer"(A=@n6)
    )
   )
  ]}
 GetPrimaryContentSetup {nodes [
   GetPrimaryContentSetup() {is=title K2Node_FunctionEntry}
   > Return Node!(
    execute
    ReturnValue:Primary Content Setup Structure=Make Primary Content Setup!(
     bShowBottomBar:bool=false
     bShowChatWidget:bool=false
     bEnableNewTextChat:bool=False
     bShowTopBar:bool=false
     bTabSupportsInterface:bool=False
    ) -> (PrimaryContentSetup:Primary Content Setup Structure) {is=title K2Node_MakeStruct}
   )
  ]}
 HandleInputChanged {nodes [
   n1 = HandleInputChanged!() -> (then, bNewInputType:ECommonInputType Enum) {is=title K2Node_FunctionEntry}
   n2 = Get IsMenuSticky() -> (IsMenuSticky:bool)
   n3 = "Equal (Enum)"(A=@n1, B=MouseAndKeyboard)
   n4 = NOT Boolean(A=@n3)
   > SetBottomButtonVisibility(
    self=Get GamepadReleaseButtonToUse() -> (GamepadReleaseButtonToUse:WBP Bottom Bar Decoy Button Object Reference)
    ShouldButtonBeVisible=AND Boolean(A=NOT Boolean(A=@n2), B=@n4)
   )
   > SetBottomButtonVisibility(
    self=Get GamepadReleaseButtonToHide() -> (GamepadReleaseButtonToHide:WBP Bottom Bar Decoy Button Object Reference)
    ShouldButtonBeVisible=false
   )
   n5 = NOT Boolean(A=Get IsCollectionExpanded())
   > SetBottomButtonVisibility(
    self=Get WBP_BottomBarDecoy_Use() -> (WBP_BottomBarDecoy_Use:WBP Bottom Bar Decoy Button Object Reference)
    ShouldButtonBeVisible=AND Boolean!(A=@n4, B=@n2, C:bool=@n5) -> (ReturnValue)
   )
   > SetVisibility(
    self=Get BottomBar_Left() -> (BottomBar_Left:Horizontal Box Object Reference)
    InVisibility=Select(Option 0=Collapsed, Option 1=SelfHitTestInvisible, Index=AND Boolean!(A=@n5, B=@n3) -> (ReturnValue))
   )
   > HideActionButtonsTouchInput!(execute, self, InputType:ECommonInputType Enum=@n1) -> (then)
  ]}
 HideActionButtonsTouchInput {nodes [
   n1 = HideActionButtonsTouchInput!() -> (then, InputType:ECommonInputType Enum) {is=title K2Node_FunctionEntry}
   > Branch(Condition="Equal (Enum)"(A=@n1, B=Touch))
   > SetBottomButtonVisibility(
    self=Get WBP_BottomBarDecoy_UseRelease() -> (WBP_BottomBarDecoy_UseRelease:WBP Bottom Bar Decoy Button Object Reference)
    ShouldButtonBeVisible=false
   )
   > SetBottomButtonVisibility(self=Get WBP_BottomBarDecoy_Use(), ShouldButtonBeVisible=false)
   > SetBottomButtonVisibility(
    self=Get WBP_BottomBarDecoy_UseCommsRelease() -> (WBP_BottomBarDecoy_UseCommsRelease:WBP Bottom Bar Decoy Button Object Reference)
    ShouldButtonBeVisible=false
   )
  ]}
 OnKeyUp {nodes [
   n1 = OnKeyUp!() -> (then, MyGeometry:Geometry Structure, InKeyEvent:Key Event Structure) {is=title K2Node_FunctionEntry}
   > n2 = Branch(Condition=Contains Item(
     TargetArray:Array of Key Structures=Get ReleaseInputBindings() -> (ReleaseInputBindings:Array of Key Structures)
     ItemToFind:"Key Structure (by ref)"=GetKey!(Input:"Key Event Structure (by ref)"=@n1.InKeyEvent) -> (ReturnValue:Key Structure)
    ))
   n3 = Get IsMenuSticky()
   n4 = Get IsUsingRightStick()
   > n5 = Branch(Condition=AND Boolean(A=NOT Boolean(A=@n3), B=NOT Boolean(A=@n4)))
   > UseItem()
   > TriggerFlowriderInteraction(Element=Get GamepadReleaseButtonToUse(), Value=Use Item Release)
   Return Node(
    execute=[@n5.else, @n2.else]
    ReturnValue:Event Reply Structure=Make Event Reply!() -> (EventReply:Event Reply Structure) {is=title K2Node_MakeStruct}
   )
  ]}
 OnMouseButtonDown {nodes [
   n1 = OnMouseButtonDown!() -> (then, MyGeometry:Geometry Structure, MouseEvent:"Pointer Event Structure (by ref)") {is=title K2Node_FunctionEntry}
   n2 = Get Effecting Button!(Input:"Pointer Event Structure (by ref)"=@n1.MouseEvent) -> (ReturnValue:Key Structure) {is=title K2Node_CallFunction}
   > n3 = Branch(Condition=NOT Boolean(A="Equal (Key)"(A:Key Structure=@n2, B:Key Structure=RightMouseButton) -> (ReturnValue:bool=~) {is=title K2Node_PromotableOperator}))
   > Return Node(ReturnValue=Make Event Reply())
   CloseMenu(execute=@n3.else) {is=title K2Node_CallFunction}
   > TriggerFlowriderInteraction(Element=Get WBP_BottomBarDecoy_Close() -> (WBP_BottomBarDecoy_Close:WBP Bottom Bar Decoy Button Object Reference), Value=RightMouseClose)
  ]}
 OnMouseButtonUp {nodes [
   n1 = OnMouseButtonUp!() -> (then, MyGeometry:Geometry Structure, MouseEvent:"Pointer Event Structure (by ref)") {is=title K2Node_FunctionEntry}
   > n2 = Branch(Condition=Contains Item(TargetArray=Get ReleaseInputBindings(), ItemToFind=Get Effecting Button(Input=@n1.MouseEvent)))
   > n3 = Branch(Condition=NOT Boolean(A=Get IsMenuSticky()))
   > UseItem()
   > TriggerFlowriderInteraction(Element=Get GamepadReleaseButtonToUse(), Value=Use Item Release)
   Return Node(execute=[@n3.else, @n2.else], ReturnValue=Make Event Reply())
  ]}
 OnTouchEnded {nodes [
   n1 = OnTouchEnded!() -> (then, MyGeometry:Geometry Structure, InTouchEvent:"Pointer Event Structure (by ref)") {is=title K2Node_FunctionEntry}
   > n2 = GetConsoleVariableBoolValue!(execute, VariableName:String=EmoteWheelPickerScreen.EnableTapAnywhereToCloseOnTouch) -> (then, ReturnValue:bool)
   n3 = IsUnderLocation!(
    Geometry:"Geometry Structure (by ref)"=GetCachedGeometry!(self:Widget Object Reference=Get WBP_RadialWheel()) -> (ReturnValue:"Geometry Structure (by ref)")
    AbsoluteCoordinate:"Vector 2D Structure (by ref)"=Get Screen Space Position!(Input:"Pointer Event Structure (by ref)"=@n1.InTouchEvent) -> (ReturnValue:Vector 2D Structure) {is=title K2Node_CallFunction}
   ) -> (ReturnValue:bool)
   n4 = Get IsCollectionExpanded()
   > n5 = Branch(Condition=AND Boolean!(A=@n2, B=NOT Boolean(A=@n3), C:bool=NOT Boolean(A=@n4)) -> (ReturnValue))
   > TriggerFlowriderInteraction(Element=Self-Reference(), Value=Close Wheel)
   > CloseMenu()
   > Return Node(ReturnValue=Unhandled())
   Return Node(execute=@n5.else, ReturnValue=Unhandled())
  ]}
 SetAllInteractionsEnabled {nodes [
   n1 = SetAllInteractionsEnabled!() -> (then, IsInteractionEnabled:bool=~) {is=title K2Node_FunctionEntry}
   > SetIsInteractionEnabled(self=[Get WBP_BottomBarDecoy_Use(), Get GamepadReleaseButtonToUse(), Get WBP_BottomBarDecoy_Close()], bInIsInteractionEnabled=@n1)
  ]}
 SetLerpTopToBottomOfScreenHeight {nodes [
   n1 = SetLerpTopToBottomOfScreenHeight!() -> (then, NewValue:double=~) {is=title K2Node_FunctionEntry}
   n2 = Lerp(A=1, B=0, Alpha=@n1)
   n3 = Get AnimationGrid_ForDynamicScreenHeight() -> (AnimationGrid_ForDynamicScreenHeight:Grid Panel Object Reference)
   > SetRowFill!(execute, self:Grid Panel Object Reference=@n3, RowIndex:int=0, Coefficient:float=@n2) -> (then)
   > SetRowFill(self=@n3, RowIndex=1, Coefficient=float - float(A=1, B=@n2))
  ]}
 SetMobileStyle {nodes [
   SetMobileStyle() {is=title K2Node_FunctionEntry}
   > Branch(Condition=IsUsingCompactUI())
  ]}
 ToggleHUDVisibility {nodes [
   ToggleHUDVisibility!() -> (then, bHideElements:bool=~) {is=title K2Node_FunctionEntry}
   > Branch(Condition=IsInZone!(self:Fort Global UIContext Object Reference=Get FortGlobalUIContext() -> (ReturnValue:Fort Global UIContext Object Reference) {is=title K2Node_GetSubsystem}) -> (ReturnValue:bool))
   > n1 = GetFortHUDContextByPlayer!(
    execute
    LocalPlayer:Local Player Object Reference=GetOwningLocalPlayer!(self:Widget Object Reference) -> (ReturnValue:Local Player Object Reference)
   ) -> (then, ReturnValue:Fort HUDContext Object Reference)
   > SetHUDElementVisibility!(
    execute
    self:Fort HUDContext Object Reference=@n1
    HUDElementTags:"Gameplay Tag Container Structure (by ref)"=MakeGameplayTagContainerFromTag!(SingleTag:Gameplay Tag Structure=(TagName="HUD.Athena.HideforRadial")) -> (ReturnValue:Gameplay Tag Container Structure)
    bHideElements:bool=Get bHideElements() -> (bHideElements:bool)
    ReasonName:"Name (by ref)"=None
   ) -> (then)
  ]}
 TriggerFlowriderInteraction {nodes [
   n1 = TriggerFlowriderInteraction!() -> (then, Element:Widget Object Reference, Value:String) {is=title K2Node_FunctionEntry}
   n2 = GetAnalyticsEWActivationCount(self=Get FortPickerContext()) -> (ReturnValue=0)
   n3 = Make Flow Analytics Additional Info()
   > Fire Event Element Interacted(
    Element=@n1
    AdditionalInfo="Add String (Non-Localized)"(
     Info="Add String (Non-Localized)"(Info=@n3, Key=Button Name, Value=@n1.Value)
     Key=emote_wheel_session_id
     Value=GetAnalyticsSessionID(ActivationCount=@n2)
    )
   )
  ]}
 UnLoadCollection {nodes [
   UnLoadCollection() {is=title K2Node_FunctionEntry}
   n1 = GetAnalyticsEWActivationCount(self=Get FortPickerContext()) -> (ReturnValue=0)
   > Fire Event Main View Opened(
    View=Self-Reference()
    AdditionalInfo="Add String (Non-Localized)"(Info=Make Flow Analytics Additional Info(), Key=emote_wheel_session_id, Value=GetAnalyticsSessionID(ActivationCount=@n1))
    bIsOverlay=false
   )
   > HandleInputChanged(bNewInputType=GetCurrentInputType(self=Get CommonInputSubsystem())) {is=title K2Node_CallFunction}
   > n2 = Get()
   > RemoveFromParent(self=@n2)
   > Set LockerCollectionWidget()
   > SetVisibility(InVisibility=HitTestInvisible)
  ]}
 UpdateFlowriderSubView {nodes [
   n1 = UpdateFlowriderSubView!() -> (then, SubViewString:String) {is=title K2Node_FunctionEntry}
   n2 = GetAnalyticsEWActivationCount(self=Get FortPickerContext()) -> (ReturnValue=0)
   > "Fire Event Sub View Opened (Non-Localized)"!(
    execute
    FortPC:Fort Player Controller Object Reference=GetFortPlayerControllerFromActor!(Actor:Actor Object Reference=GetOwningPlayer()) -> (ReturnValue:Fort Player Controller Object Reference)
    ViewName:String=@n1
    AdditionalInfo:Flow Analytics Additional Info Structure="Add String (Non-Localized)"(Info=Make Flow Analytics Additional Info(), Key=emote_wheel_session_id, Value=GetAnalyticsSessionID(ActivationCount=@n2))
    bIsOverlay:bool=false
   ) -> (then)
   Set PromptNudge!(execute, PromptNudge:Vector 2D Structure, self:WBP UIKit Button Generic Base Object Reference) -> (then, Output_Get:Vector 2D Structure)
  ]}
 UpdateItemExpirationView {nodes [
   n1 = UpdateItemExpirationView!() -> (then, Has Expiration Date:bool=~, Time Until Item Expires:Timespan Structure) {is=title K2Node_FunctionEntry}
   > SetVisibility(
    self=Get Overlay_TimeLeft() -> (Overlay_TimeLeft:Overlay Object Reference)
    InVisibility=Select(Option 0=Hidden, Option 1=Visible, Index=@n1)
   )
   > Set Text(
    Text=GetFormattedTimespan!(self, Timespan:Timespan Structure=@n1.Time Until Item Expires) -> (Text:Text) {is=title K2Node_CallFunction}
    self=Get Text_ItemTimeLeft() -> (Text_ItemTimeLeft:Common Text Object Reference)
   )
  ]}
 UpdateReleasePrompts {nodes [
   UpdateReleasePrompts!() -> (then, ReleaseInputBindings:Array of Key Structures) {is=title K2Node_FunctionEntry}
   > Set MatchFound(MatchFound:bool=false) -> (Output_Get:bool)
   > n1 = For Each Loop with Break!(Exec:Exec, Array:Array of Key Structures=Get ReleaseInputBindings(), Break:Exec=@n3.then_0) -> (LoopBody:Exec, Array Element:Key Structure, Array Index:int, Completed:Exec)
   > n2 = Cast To FortPlayerController(Object=GetOwningPlayer())
   > Branch(Condition=IsInputKeyDown!(self:Player Controller Object Reference=@n2, Key:Key Structure=@n1) -> (ReturnValue:bool))
   > Set MatchFound(MatchFound=true)
   > n3 = Sequence!(execute) -> (then_0, then_1, then_2:Exec)
   n4 = Get GamepadPossibleInputs() -> (GamepadPossibleInputs:Map of Key Structures to Data Table Row Handle Structures)
   n5 = Branch(
    execute=@n3.then_1
    Condition="Equal (Key)"(
     A=Get Key Value By Index!(TargetMap:Map of Key Structures to Data Table Row Handle Structures=@n4, Index:int=0) -> (Key:Key Structure, Value:Data Table Row Handle Structure) {is=title K2Node_CallFunction}
     B=@n1
    ) -> (ReturnValue=~)
   )
   Set GamepadReleaseButtonToUse(execute=[@n5.then, @n10.else], GamepadReleaseButtonToUse:WBP Bottom Bar Decoy Button Object Reference=Get WBP_BottomBarDecoy_UseRelease()) -> (Output_Get:WBP Bottom Bar Decoy Button Object Reference)
   > n6 = Set GamepadReleaseButtonToHide(GamepadReleaseButtonToHide:WBP Bottom Bar Decoy Button Object Reference=Get WBP_BottomBarDecoy_UseCommsRelease()) -> (Output_Get:WBP Bottom Bar Decoy Button Object Reference)
   HandleInputChanged(execute=[@n7.then, @n6.then], bNewInputType=GetCurrentInputType(self=Get CommonInputSubsystem()))
   Set GamepadReleaseButtonToUse(execute=@n5.else, GamepadReleaseButtonToUse=Get WBP_BottomBarDecoy_UseCommsRelease())
   > n7 = Set GamepadReleaseButtonToHide(GamepadReleaseButtonToHide=Get WBP_BottomBarDecoy_UseRelease())
   n8 = Find(
    TargetMap:Map of Key Structures to Data Table Row Handle Structures=Get KBMPossibleInputs() -> (KBMPossibleInputs:Map of Key Structures to Data Table Row Handle Structures)
    Key:"Key Structure (by ref)"=@n1
   ) -> (Value:Data Table Row Handle Structure)
   Branch(execute=@n3.then_2, Condition=@n8.ReturnValue)
   n9 = Get BottomBarLeft_UseRelease()
   > SetTriggeringInputAction!(execute, self:Common Button Base Object Reference=@n9, InputActionRow:"Data Table Row Handle Structure (by ref)"=@n8) -> (then)
   > UpdatePrompt!(
    execute
    self:WBP UIKit Button Generic Base Object Reference=@n9
    InputType:ECommonInputType Enum=GetCurrentInputType(self=Get CommonInputSubsystem())
   ) -> (then)
   n10 = Branch(execute=@n1.Completed, Condition=Get MatchFound() -> (MatchFound:bool))
   SetTriggeringInputAction(self=Get WBP_BottomBarDecoy_UseRelease())
  ]}
 UpdateSelectedCategory {nodes [
   n1 = UpdateSelectedCategory!() -> (then, Category:Emote Wheel Category VM Object Reference) {is=title K2Node_FunctionEntry}
   > n2 = Cast To GizmosWheelCategoryVM!(execute, Object=@n1) -> (then, CastFailed, AsGizmos Wheel Category VM:Gizmos Wheel Category VM Object Reference) {is=title K2Node_DynamicCast}
   > UpdateWheelSlotCount!(
    execute
    self
    SlotCount:int=Length(TargetArray:Array of Emote Wheel Item VM Object References=GetEquippedItems!(self:Emote Wheel Category VM Object Reference=@n2) -> (ReturnValue:Array of Emote Wheel Item VM Object References))
   ) -> (then)
  ]}
 UpdateSelectedItem {nodes [
   n1 = UpdateSelectedItem!() -> (then, Item:Emote Wheel Item VM Object Reference) {is=title K2Node_FunctionEntry}
   > n2 = Cast To GizmosWheelItemVM!(execute, Object=@n1) -> (then, CastFailed, AsGizmos Wheel Item VM:Gizmos Wheel Item VM Object Reference) {is=title K2Node_DynamicCast}
   > Set WheelItemVM(WheelItemVM:Gizmos Wheel Item VM Object Reference=@n2) -> (Output_Get:Gizmos Wheel Item VM Object Reference)
   n3 = Get ItemDefinition!(self:Gizmos Wheel Item VM Object Reference=@n2) -> (ItemDefinition:Fort World Item Definition Object Reference)
   > n4 = Is Valid(InputObject=@n3)
   > Set ItemDefinition(ItemDefinition:Fort World Item Definition Object Reference=@n3) -> (Output_Get:Fort World Item Definition Object Reference)
   n5 = GetRarity!(ItemDefinition:Item Definition Base Object Reference=@n3) -> (ReturnValue:EFortRarity Enum=Common)
   n6 = "Equal (Enum)"(A:EFortRarity Enum=@n5, B:EFortRarity Enum=Common)
   n7 = "Equal (Enum)"(A=@n5, B=Unattainable)
   > SetVectorParameterValue(
    self=Get MID_BackgroundWithSegments(self=Get WBP_RadialWheel())
    ParameterName=SubSegmentColor
    Value=SelectColor(
     A=(R=1,G=1,B=1,A=1)
     B=GetRarityOrSeriesColors!(ItemDefinition:Item Definition Base Object Reference=@n3) -> (
      ReturnValue_Color1:Linear Color Structure
      ReturnValue_Color2:Linear Color Structure
      ReturnValue_Color3:Linear Color Structure
      ReturnValue_Color4:Linear Color Structure
      ReturnValue_Color5:Linear Color Structure
     ).ReturnValue_Color2
     bPickA=OR Boolean(A=@n6, B=@n7)
    )
   )
   > Set Text(
    Text=Get DisplayName!(self:Emote Wheel Item VM Object Reference=@n2) -> (DisplayName:Text)
    self=Get Text_ItemName() -> (Text_ItemName:Common Text Object Reference)
   )
   > Set Text(
    Text=GetItemDescription!(self:Item Definition Base Object Reference=Get ItemDefinition(self=@n2)) -> (ReturnValue:Text)
    self=Get Text_ItemDescription() -> (Text_ItemDescription:Common Text Object Reference)
   )
   > UpdateItemExpirationView!(
    execute
    self
    Has Expiration Date:bool=HasExpirationDate!(self:Gizmos Wheel Item VM Object Reference=@n2) -> (ReturnValue:bool)
    Time Until Item Expires:Timespan Structure=GetTimeUntilItemExpires!(self:Gizmos Wheel Item VM Object Reference=@n2) -> (ReturnValue:Timespan Structure)
   ) -> (then) {is=title K2Node_CallFunction}
   > n8 = GetRarityData!(execute, Rarity:EFortRarity Enum=@n5) -> (then, ReturnValue:Fort Rarity Item Data Structure)
   n9 = Get Text_RarityLabel() -> (Text_RarityLabel:Common Text Object Reference)
   n10 = Break Fort Rarity Item Data!(FortRarityItemData:"Fort Rarity Item Data Structure (by ref)"=@n8) -> (
    Name:Text
    Color1:Linear Color Structure
    Color2:Linear Color Structure
    Color3:Linear Color Structure
    Color4:Linear Color Structure
    Color5:Linear Color Structure
    Radius:float=~
    Falloff:float=~
    Brightness:float=~
    Roughness:float=~
    Glow:float=~
   ) {is=title K2Node_BreakStruct}
   > "SetText (Text)"(self=@n9, InText=@n10)
   > n11 = GetDynamicFontMaterial!(execute, self:Text Object Reference=@n9) -> (then, ReturnValue:Material Instance Dynamic Object Reference)
   > SetVectorParameterValue(self=@n11, ParameterName=Color_1, Value=@n10.Color1)
   > SetVectorParameterValue(self=@n11, ParameterName=Color_2, Value=@n10.Color2)
   > n12 = GetDynamicMaterial(self:Border Object Reference=Get RarityLabel() -> (RarityLabel:Border Object Reference))
   > SetVectorParameterValue(self=@n12, ParameterName=Color, Value=@n10.Color4)
   n13 = Get ItemDefinition!() -> (ItemDefinition)
   n14 = Get WheelItemVM() -> (WheelItemVM:Gizmos Wheel Item VM Object Reference)
   > "SetText (Text)"(
    self=Get CommonText_ItemType() -> (CommonText_ItemType:Common Text Object Reference)
    InText=Select(
     Option 0:Text="LOCTEXT(\"Gizmo\")"
     Option 1:Text=GetItemTypeName!(ItemDefinition:Item Definition Base Object Reference=@n13, bUsePlural:bool=false) -> (ReturnValue:Text)
     Index=IsPickaxeItem!(self:Gizmos Wheel Item VM Object Reference=@n14) -> (ReturnValue:bool)
    ) -> (ReturnValue:Text)
   )
   n15 = GetItemType!(ItemDefinition:Item Definition Base Object Reference=Get ItemDefinition()) -> (ReturnValue:EFortItemType Enum=WorldItem)
   n16 = Get ExcludeShowingItemTypes() -> (ExcludeShowingItemTypes:Array of EFortItemType Enums=WorldItem)
   > SetVisibility(
    self=Get HorizontalBox_ItemTypeRarity() -> (HorizontalBox_ItemTypeRarity:Horizontal Box Object Reference)
    InVisibility=Select(
     Option 0:ESlateVisibility Enum=HitTestInvisible
     Option 1:ESlateVisibility Enum=Collapsed
     Index=Contains Item(TargetArray:Array of EFortItemType Enums=@n16, ItemToFind:"EFortItemType Enum (by ref)"=@n15)
    ) -> (ReturnValue:ESlateVisibility Enum)
   )
   SetVectorParameterValue(
    execute=@n4.Is Not Valid
    self=Get MID_BackgroundWithSegments(self=Get WBP_RadialWheel())
    ParameterName=SubSegmentColor
    Value=(R=1,G=1,B=1,A=1)
   )
  ]}
 UpdateSlotIndicatorVisibility {nodes [
   n1 = UpdateSlotIndicatorVisibility!() -> (then, CategoryDisplayName:Text) {is=title K2Node_FunctionEntry}
   n2 = Get Source String from Text!(Text:Text=@n1) -> (ReturnValue:String) {is=title K2Node_CallFunction}
   n3 = Get SlotIndicatorHideStringArray() -> (SlotIndicatorHideStringArray:Array of Strings)
   > UpdateSlotIndicatorVisibility!(
    execute
    self:WBP Emote Radial Wheel Object Reference=Get WBP_RadialWheel()
    ShouldShow:bool="Equal (Integer)"(A=Find Item!(TargetArray=@n3, ItemToFind:"String (by ref)"=@n2) -> (ReturnValue), B=-1)
   ) -> (then) {is=title K2Node_CallFunction}
  ]}
 UpdateWheelSlotCount {nodes [
   UpdateWheelSlotCount!() -> (then, SlotCount:int) {is=title K2Node_FunctionEntry}
   > n1 = GetDynamicMaterial(self:Image Object Reference=Get BackgroundWithSegments!(self:WBP Emote Radial Wheel Object Reference=Get WBP_RadialWheel()) -> (BackgroundWithSegments:Image Object Reference))
   > SetScalarParameterValue(self=@n1, ParameterName=Subdivisions, Value="To Float (Integer)"(InInt=Get SlotCount() -> (SlotCount:int=0)))
   n2 = Get WBP_RadialWheel()
   > n3 = GetDynamicMaterial(self=Get EmotePosition_Indicator!(self:WBP Emote Indicator Object Reference=Get WBP_EmoteIndicator!(self:WBP Emote Radial Wheel Object Reference=@n2) -> (WBP_EmoteIndicator:WBP Emote Indicator Object Reference)) -> (EmotePosition_Indicator:Image Object Reference))
   > SetScalarParameterValue(self=@n3, ParameterName=Subdivisions, Value="To Float (Integer)"(InInt=Get SlotCount() -> (SlotCount=0)))
  ]}
 UseItem {nodes [
   UseItem() {is=title K2Node_FunctionEntry}
   > n1 = Get!(execute, self:Emote Wheel Item Selection VM Object Reference=Get EmoteWheelItemSelectionVM()) -> (then, else, SelectedItem:Emote Wheel Item VM Object Reference)
   > n2 = Play(self:Emote Wheel Item VM Object Reference=@n1)
   CloseMenu(execute=[@n2.then, @n1.else])
  ]}
]
variables [
 ClickedCollectionButton {type=CommonButtonBase, default=null}
 ClickedUseButton {type=CommonButtonBase, default=null}
 CollectionClass {type=Class, default=&/WBP_EW_Collection.WBP_EW_Collection_C}
 CollectionCount {type=integer}
 CurrentOverlayWidget {type=UserWidget, default=null}
 ExcludeShowingItemTypes {type="array<EFortItemType>", default [WeaponHarvest]}
 GamepadReleaseButtonToHide {type=WBP_BottomBarDecoyButton_C, default=null}
 GamepadReleaseButtonToUse {type=WBP_BottomBarDecoyButton_C, default=null}
 InputAction_Back {type=DataTableRowHandle, default {dataTable=&InputActionDataTable, rowName=Back}}
 InputAction_Close {type=DataTableRowHandle, default {dataTable=&InputActionDataTable, rowName=Close}}
 Is Collection Expanded {type=boolean}
 IsCollectionExpanded {type=boolean}
 IsMenuSticky {type=boolean}
 IsUsingRightStick {type=boolean}
 LerpTopToBottomOfScreenHeight {type=number}
 LockerCollectionWidget {type=CommonActivatableWidget, default=null}
 ReleaseInputBindings {type="array<Key>", default []}
 SlotIndicatorHideStringArray {type="array<string>", default [Jam Tracks, Comms, Follower]}
 SoftCollectionClass {type=Class, default=&WBP_EW_Collection_C}
 SubTabInterface {default=None}
]
widgets {
 info {inheritedWidgetCount=0, parentClass=/Script/EmoteWheelRuntime.EmoteWheelPickerScreen, rootWidgetClass=&GridPanel, widgetCount=53}
 widgets [
  GridPanel_Main {widgetClassPath=&GridPanel, widget=GridPanel_Main, properties {columnFill [1], rowFill [1]}}
  BG_Darkener {
   widgetClassPath=&Image
   bIsVariable=true
   parent=GridPanel_Main
   widget=BG_Darkener
   properties {colorAndOpacity=<rgba 0.04 0.04 0.04 0.4>}
   slot {class=GridSlot, properties {layer=-101}}
  }
  PostBufferUpdate_BGBlur {
   widgetClassPath=/Script/UMG.PostBufferUpdate
   bIsVariable=true
   parent=GridPanel_Main
   widget=PostBufferUpdate_BGBlur
   properties {
    visibility=HitTestInvisible
    updateBufferInfos [{
      bufferToUpdate=ESlatePostRT_0
      postParamUpdater=&self:WidgetTree.PostBufferUpdate_BGBlur.PostBufferBlurUpdater_0
     }]
   }
   slot {class=GridSlot, properties {layer=-100}}
  }
  Background_BlockClick_BlurOnlyHighEndDevice {
   widgetClassPath=&Image
   bIsVariable=true
   parent=GridPanel_Main
   widget=Background_BlockClick_BlurOnlyHighEndDevice
   properties {brush {
     bIsDynamicallyLoaded=false
     drawAs=Image
     imageType=FullColor
     mirroring=NoMirror
     resourceName=None
     resourceObject=/EmoteWheel/MaterialInstances/MI_UI_EmoteScreen_Background
     tiling=NoTile
     imageSize=<32 32>
     margin=<ltrb 0 0 0 0>
     outlineSettings {
      bUseBrushTransparency=false
      roundingType=HalfHeightRadius
      width=0
      color {colorUseRule=UseColor_Specified, specifiedColor=<rgba 0 0 0 0>}
      cornerRadii=<0 0 0 0>
     }
     tintColor {colorUseRule=UseColor_Specified, specifiedColor=<rgba 1 1 1 1>}
     uVRegion {bIsValid=false, max=<0 0>, min=<0 0>}
    }}
   slot {class=GridSlot, properties {layer=-50}}
  }
  WBP_AthenaBottomBarWidget {
   widgetClassPath=/Game/UI/Frontend/BottomBar/WBP_AthenaBottomBarWidget
   bIsVariable=true
   parent=GridPanel_Main
   widget=WBP_AthenaBottomBarWidget
   slot {class=GridSlot, properties {horizontalAlignment=HAlign_Right, layer=100, verticalAlignment=VAlign_Bottom}}
  }
  SafeZone {widgetClassPath=&SafeZone, bIsVariable=true, parent=GridPanel_Main, widget=SafeZone, slot {class=GridSlot}}
  GridPanel_18 {
   widgetClassPath=&GridPanel
   parent=SafeZone
   widget=GridPanel_18
   properties {columnFill [1], rowFill [1]}
   slot {class=SafeZoneSlot}
  }
  RadialWheel_ScaleBox {
   widgetClassPath=&ScaleBox
   parent=GridPanel_18
   widget=RadialWheel_ScaleBox
   properties {stretchDirection=DownOnly}
   slot {class=GridSlot, properties {horizontalAlignment=HAlign_Center, verticalAlignment=VAlign_Center, padding=<ltrb 0 200 0 200>}}
  }
  WBP_RadialWheel {
   widgetClassPath=&/WBP_Gizmo_RadialWheel
   bIsVariable=true
   parent=RadialWheel_ScaleBox
   widget=WBP_RadialWheel
   properties {
    navigation=&self:WidgetTree.WBP_RadialWheel.WidgetNavigation_0
    radialBoxSettings {angleBetweenItems=45, bClockwiseOrder=true, bDistributeItemsEvenly=true, marginSize=0, sectorCentralAngle=360, startingAngle=90}
   }
   slot {class=ScaleBoxSlot, properties {horizontalAlignment=HAlign_Fill, verticalAlignment=VAlign_Fill}}
  }
  AnimationGrid_ForDynamicScreenHeight {
   widgetClassPath=&GridPanel
   bIsVariable=true
   parent=GridPanel_18
   widget=AnimationGrid_ForDynamicScreenHeight
   properties {bOverride_Cursor=true, columnFill [1], rowFill [1,0]}
   slot {class=GridSlot, properties {layer=30}}
  }
  Spacer_Row0 {
   widgetClassPath=&Spacer
   parent=AnimationGrid_ForDynamicScreenHeight
   widget=Spacer_Row0
   slot {class=GridSlot, properties {rowSpan=1}}
  }
  VisualAttachment_Row1 {
   widgetClassPath=&CommonVisualAttachment
   parent=AnimationGrid_ForDynamicScreenHeight
   widget=VisualAttachment_Row1
   properties {contentAnchor=<0.5 0>}
   slot {class=GridSlot, properties {horizontalAlignment=HAlign_Center, row=1, rowSpan=1}}
  }
  VerticalBox_0 {
   widgetClassPath=&VerticalBox
   parent=GridPanel_18
   widget=VerticalBox_0
   slot {class=GridSlot, properties {horizontalAlignment=HAlign_Left, verticalAlignment=VAlign_Bottom, padding=<ltrb 80 0 0 35>}}
  }
  HorizontalBox_ItemTypeRarity {
   widgetClassPath=&HorizontalBox
   bIsVariable=true
   parent=VerticalBox_0
   widget=HorizontalBox_ItemTypeRarity
   slot {class=VerticalBoxSlot}
  }
  CommonText_ItemType {
   widgetClassPath=&CommonTextBlock
   bIsVariable=true
   parent=HorizontalBox_ItemTypeRarity
   widget=CommonText_ItemType
   properties {
    bOverrideAccessibleDefaults=true
    text=Gizmo
    textTransformPolicy=ToUpper
    colorAndOpacity {colorUseRule=UseColor_Specified, specifiedColor=<rgba 0.75 0.75 0.75 0.8>}
    font {
     bForceMonospaced=false
     bMaterialIsStencil=false
     fontMaterial=null
     fontObject=&HeadingNow
     letterSpacing=0
     monospacedWidth=1
     size=15.75
     skewAmount=0
     typefaceFontName=86 Bold Italic
     outlineSettings {
      bApplyOutlineToDropShadows=false
      bMiteredCorners=false
      bSeparateFillAlpha=false
      outlineMaterial=null
      outlineSize=0
      outlineColor=<rgba 0 0 0 1>
     }
    }
    shadowOffset=<0 0>
   }
   slot {class=HorizontalBoxSlot, properties {horizontalAlignment=HAlign_Left, verticalAlignment=VAlign_Center, padding=<ltrb 0 0 10 0>}}
  }
  ScaleBox_0 {
   widgetClassPath=&ScaleBox
   parent=HorizontalBox_ItemTypeRarity
   widget=ScaleBox_0
   properties {stretch=UserSpecified, userSpecifiedScale=1.3}
   slot {class=HorizontalBoxSlot, properties {horizontalAlignment=HAlign_Right, verticalAlignment=VAlign_Center, padding=<ltrb 0 0 2 0>}}
  }
  RarityLabel {
   widgetClassPath=&Border
   bIsVariable=true
   parent=ScaleBox_0
   widget=RarityLabel
   properties {
    clipping=ClipToBounds
    visibility=SelfHitTestInvisible
    background {
     bIsDynamicallyLoaded=false
     drawAs=Image
     imageType=FullColor
     mirroring=NoMirror
     resourceName=None
     resourceObject=/EmoteWheel/Materials/M_UI_RarityLabel_Backing_GizmoWheel
     tiling=NoTile
     imageSize=<32 32>
     margin=<ltrb 0 0 0 0>
     outlineSettings {
      bUseBrushTransparency=false
      roundingType=HalfHeightRadius
      width=0
      color {colorUseRule=UseColor_Specified, specifiedColor=<rgba 0 0 0 0>}
      cornerRadii=<0 0 0 0>
     }
     tintColor {colorUseRule=UseColor_Specified, specifiedColor=<rgba 1 1 1 1>}
     uVRegion {bIsValid=false, max=<0 0>, min=<0 0>}
    }
    padding=<ltrb 3 1 3 1>
   }
   slot {class=ScaleBoxSlot}
  }
  Text_RarityLabel {
   widgetClassPath=&CommonTextBlock
   bIsVariable=true
   parent=RarityLabel
   widget=Text_RarityLabel
   properties {
    justification=Right
    mobileFontSizeMultiplier=1.35
    text="{Rarity}"
    textTransformPolicy=ToUpper
    font {
     bForceMonospaced=false
     bMaterialIsStencil=false
     fontMaterial=/EmoteWheel/Materials/M_UI_RarityTag_Font_GizmoWheel
     fontObject=&BurbankSmall
     letterSpacing=0
     monospacedWidth=1
     size=9
     skewAmount=0
     typefaceFontName=Black
     outlineSettings {
      bApplyOutlineToDropShadows=false
      bMiteredCorners=false
      bSeparateFillAlpha=false
      outlineMaterial=null
      outlineSize=0
      outlineColor=<rgba 0 0 0 1>
     }
    }
    shadowOffset=<0 0>
   }
   slot {class=BorderSlot, properties {padding=<ltrb 3 1 3 1>}}
  }
  Text_ItemName {
   widgetClassPath=&CommonTextBlock
   bIsVariable=true
   parent=VerticalBox_0
   widget=Text_ItemName
   properties {
    bAutoCollapseWithEmptyText=true
    mobileFontSizeMultiplier=1.2
    scrollStyle=&TextScrollStyle-Base_C
    style=/FortUILibrary/Styles/Typography/_TextStyle_SubH2._TextStyle_SubH2_C
    text="[ITEMNAME]"
    textTransformPolicy=ToUpper
    wrapTextAt=600
    font {
     bForceMonospaced=false
     bMaterialIsStencil=false
     fontMaterial=null
     fontObject=&Burbank
     letterSpacing=0
     monospacedWidth=1
     size=33
     skewAmount=0
     typefaceFontName=BurbankRegular-Black
     outlineSettings {
      bApplyOutlineToDropShadows=false
      bMiteredCorners=false
      bSeparateFillAlpha=false
      outlineMaterial=null
      outlineSize=0
      outlineColor=<rgba 0 0 0 1>
     }
    }
    shadowOffset=<0 0>
    strikeBrush {
     bIsDynamicallyLoaded=false
     drawAs=NoDrawType
     imageType=NoImage
     mirroring=NoMirror
     resourceName=None
     resourceObject=null
     tiling=NoTile
     imageSize=<32 32>
     margin=<ltrb 0 0 0 0>
     outlineSettings {
      bUseBrushTransparency=false
      roundingType=HalfHeightRadius
      width=0
      color {colorUseRule=UseColor_Specified, specifiedColor=<rgba 0 0 0 0>}
      cornerRadii=<0 0 0 0>
     }
     tintColor {colorUseRule=UseColor_Specified, specifiedColor=<rgba 1 1 1 1>}
     uVRegion {bIsValid=false, max=<0 0>, min=<0 0>}
    }
   }
   slot {class=VerticalBoxSlot}
  }
  Text_ItemDescription {
   widgetClassPath=&CommonTextBlock
   bIsVariable=true
   parent=VerticalBox_0
   widget=Text_ItemDescription
   properties {
    bAutoCollapseWithEmptyText=true
    lineHeightPercentage=1.29
    mobileFontSizeMultiplier=1.2
    scrollStyle=&TextScrollStyle-Base_C
    style=/FortUILibrary/Styles/Typography/_TextStyle_Body1._TextStyle_Body1_C
    text=Gizmo description that is multiple lines to ensure it takes enough space when hidden.
    wrapTextAt=600
    font {
     bForceMonospaced=false
     bMaterialIsStencil=false
     fontMaterial=null
     fontObject=&BurbankSmall
     letterSpacing=0
     monospacedWidth=1
     size=18
     skewAmount=0
     typefaceFontName=Bold
     outlineSettings {
      bApplyOutlineToDropShadows=false
      bMiteredCorners=false
      bSeparateFillAlpha=false
      outlineMaterial=null
      outlineSize=0
      outlineColor=<rgba 0 0 0 1>
     }
    }
    shadowOffset=<0 0>
    strikeBrush {
     bIsDynamicallyLoaded=false
     drawAs=NoDrawType
     imageType=NoImage
     mirroring=NoMirror
     resourceName=None
     resourceObject=null
     tiling=NoTile
     imageSize=<32 32>
     margin=<ltrb 0 0 0 0>
     outlineSettings {
      bUseBrushTransparency=false
      roundingType=HalfHeightRadius
      width=0
      color {colorUseRule=UseColor_Specified, specifiedColor=<rgba 0 0 0 0>}
      cornerRadii=<0 0 0 0>
     }
     tintColor {colorUseRule=UseColor_Specified, specifiedColor=<rgba 1 1 1 1>}
     uVRegion {bIsValid=false, max=<0 0>, min=<0 0>}
    }
   }
   slot {class=VerticalBoxSlot, properties {padding=<ltrb 0 8 0 0>, size {sizeRule=Fill, value=2}}}
  }
  Overlay_TimeLeft {
   widgetClassPath=&Overlay
   bIsVariable=true
   parent=VerticalBox_0
   widget=Overlay_TimeLeft
   slot {class=VerticalBoxSlot, properties {horizontalAlignment=HAlign_Left, verticalAlignment=VAlign_Center, padding=<ltrb 0 8 0 0>}}
  }
  Image_TimerBackground {
   widgetClassPath=&Image
   parent=Overlay_TimeLeft
   widget=Image_TimerBackground
   properties {brush {
     bIsDynamicallyLoaded=false
     drawAs=RoundedBox
     imageType=FullColor
     mirroring=NoMirror
     resourceName=None
     resourceObject=/Game/Athena/UI/Frontend/ActivityBrowser/Art/M_UI_Clock_WhiteCirlce
     tiling=NoTile
     imageSize=<32 32>
     margin=<ltrb 0.5 0.5 0.5 0.5>
     outlineSettings {
      bUseBrushTransparency=false
      roundingType=HalfHeightRadius
      width=0
      color {colorUseRule=UseColor_Specified, specifiedColor=<rgba 0 0 0 0>}
      cornerRadii=<0 0 0 0>
     }
     tintColor {colorUseRule=UseColor_Specified, specifiedColor=<rgba 0 0 0 0.5611>}
     uVRegion {bIsValid=false, max=<0 0>, min=<0 0>}
    }}
   slot {class=OverlaySlot, properties {horizontalAlignment=HAlign_Fill, verticalAlignment=VAlign_Fill}}
  }
  HorizontalBox_0 {
   widgetClassPath=&HorizontalBox
   parent=Overlay_TimeLeft
   widget=HorizontalBox_0
   slot {class=OverlaySlot, properties {verticalAlignment=VAlign_Center, padding=<ltrb 20 10 20 10>}}
  }
  Image_Timer {
   widgetClassPath=&Image
   bIsVariable=true
   parent=HorizontalBox_0
   widget=Image_Timer
   properties {brush {
     bIsDynamicallyLoaded=false
     drawAs=Image
     imageType=FullColor
     mirroring=NoMirror
     resourceName=None
     resourceObject=/Game/UI/Icons/M_UI_Timer_Icon_ClassSelector
     tiling=NoTile
     imageSize=<40 40>
     margin=<ltrb 0 0 0 0>
     outlineSettings {
      bUseBrushTransparency=false
      roundingType=HalfHeightRadius
      width=0
      color {colorUseRule=UseColor_Specified, specifiedColor=<rgba 0 0 0 0>}
      cornerRadii=<0 0 0 0>
     }
     tintColor {colorUseRule=UseColor_Specified, specifiedColor=<rgba 1 1 1 1>}
     uVRegion {bIsValid=false, max=<0 0>, min=<0 0>}
    }}
   slot {class=HorizontalBoxSlot, properties {horizontalAlignment=HAlign_Center, verticalAlignment=VAlign_Center, padding=<ltrb 0 0 5 0>}}
  }
  CommonText_EventEndsIn {
   widgetClassPath=&CommonTextBlock
   parent=HorizontalBox_0
   widget=CommonText_EventEndsIn
   properties {
    lineHeightPercentage=1.29
    style=&_TextStyle_Body1_C
    text="Expires in "
    font {
     bForceMonospaced=false
     bMaterialIsStencil=false
     fontMaterial=null
     fontObject=&BurbankSmall
     letterSpacing=0
     monospacedWidth=1
     size=18
     skewAmount=0
     typefaceFontName=Bold
     outlineSettings {
      bApplyOutlineToDropShadows=false
      bMiteredCorners=false
      bSeparateFillAlpha=false
      outlineMaterial=null
      outlineSize=0
      outlineColor=<rgba 0 0 0 1>
     }
    }
    shadowOffset=<0 0>
    strikeBrush {
     bIsDynamicallyLoaded=false
     drawAs=NoDrawType
     imageType=NoImage
     mirroring=NoMirror
     resourceName=None
     resourceObject=null
     tiling=NoTile
     imageSize=<32 32>
     margin=<ltrb 0 0 0 0>
     outlineSettings {
      bUseBrushTransparency=false
      roundingType=HalfHeightRadius
      width=0
      color {colorUseRule=UseColor_Specified, specifiedColor=<rgba 0 0 0 0>}
      cornerRadii=<0 0 0 0>
     }
     tintColor {colorUseRule=UseColor_Specified, specifiedColor=<rgba 1 1 1 1>}
     uVRegion {bIsValid=false, max=<0 0>, min=<0 0>}
    }
   }
   slot {class=HorizontalBoxSlot, properties {horizontalAlignment=HAlign_Center, verticalAlignment=VAlign_Center, padding=<ltrb 0 5 0 0>}}
  }
  Text_ItemTimeLeft {
   widgetClassPath=&CommonTextBlock
   bIsVariable=true
   parent=HorizontalBox_0
   widget=Text_ItemTimeLeft
   properties {
    lineHeightPercentage=1.29
    style=&_TextStyle_Body1_C
    text=5d 12hrs
    font {
     bForceMonospaced=false
     bMaterialIsStencil=false
     fontMaterial=null
     fontObject=&BurbankSmall
     letterSpacing=0
     monospacedWidth=1
     size=18
     skewAmount=0
     typefaceFontName=Bold
     outlineSettings {
      bApplyOutlineToDropShadows=false
      bMiteredCorners=false
      bSeparateFillAlpha=false
      outlineMaterial=null
      outlineSize=0
      outlineColor=<rgba 0 0 0 1>
     }
    }
    shadowOffset=<0 0>
    strikeBrush {
     bIsDynamicallyLoaded=false
     drawAs=NoDrawType
     imageType=NoImage
     mirroring=NoMirror
     resourceName=None
     resourceObject=null
     tiling=NoTile
     imageSize=<32 32>
     margin=<ltrb 0 0 0 0>
     outlineSettings {
      bUseBrushTransparency=false
      roundingType=HalfHeightRadius
      width=0
      color {colorUseRule=UseColor_Specified, specifiedColor=<rgba 0 0 0 0>}
      cornerRadii=<0 0 0 0>
     }
     tintColor {colorUseRule=UseColor_Specified, specifiedColor=<rgba 1 1 1 1>}
     uVRegion {bIsValid=false, max=<0 0>, min=<0 0>}
    }
   }
   slot {class=HorizontalBoxSlot, properties {verticalAlignment=VAlign_Center, padding=<ltrb 0 5 0 0>}}
  }
  BottomBar_Left {
   widgetClassPath=&HorizontalBox
   bIsVariable=true
   parent=VerticalBox_0
   widget=BottomBar_Left
   properties {renderOpacity=0}
   slot {class=VerticalBoxSlot}
  }
  BottomBarLeft_UseRelease {
   widgetClassPath=&WBP_UIKit_Button_Generic
   bIsVariable=true
   parent=BottomBar_Left
   widget=BottomBarLeft_UseRelease
   properties {
    bDisplayInputActionWhenNotInteractable=true
    bIsFocusable=false
    bLocked=true
    text="Use It (Release)"
    useBackgroundBlock=false
    useIconBlock=false
    font {
     bForceMonospaced=false
     bMaterialIsStencil=false
     fontMaterial=/FortUILibrary/UIKit/Buttons/MaterialInstances/CTAButtons/MI_UIKit_TextQuiet_Prompt
     fontObject=&HeadingNow
     letterSpacing=0
     monospacedWidth=1
     size=15
     skewAmount=0
     typefaceFontName=75 Medium
     outlineSettings {
      bApplyOutlineToDropShadows=false
      bMiteredCorners=false
      bSeparateFillAlpha=false
      outlineMaterial=null
      outlineSize=0
      outlineColor=<rgba 0 0 0 1>
     }
    }
    promptNudge=<-12 0>
    triggeringInputAction {dataTable=&InputActionDataTable, rowName=EWUseRelease}
   }
   slot {class=HorizontalBoxSlot, properties {horizontalAlignment=HAlign_Left, verticalAlignment=VAlign_Center}}
  }
  BottomButtonLeft_Divider {
   widgetClassPath=&Image
   bIsVariable=true
   parent=BottomBar_Left
   widget=BottomButtonLeft_Divider
   properties {
    renderOpacity=0.2
    brush {
     bIsDynamicallyLoaded=false
     drawAs=Image
     imageType=NoImage
     mirroring=NoMirror
     resourceName=None
     resourceObject=null
     tiling=NoTile
     imageSize=<2 28>
     margin=<ltrb 0 0 0 0>
     outlineSettings {
      bUseBrushTransparency=false
      roundingType=HalfHeightRadius
      width=0
      color {colorUseRule=UseColor_Specified, specifiedColor=<rgba 0 0 0 0>}
      cornerRadii=<0 0 0 0>
     }
     tintColor {colorUseRule=UseColor_Specified, specifiedColor=<rgba 1 1 1 1>}
     uVRegion {bIsValid=false, max=<0 0>, min=<0 0>}
    }
   }
   slot {class=HorizontalBoxSlot, properties {horizontalAlignment=HAlign_Left, padding=<ltrb 0 10 24 10>}}
  }
  QuickUseInput1 {
   widgetClassPath=&CommonActionWidget
   bIsVariable=true
   parent=BottomBar_Left
   widget=QuickUseInput1
   properties {
    progressMaterialParam=Percentage
    visibility=HitTestInvisible
    inputActions [{dataTable=&InputActionDataTable, rowName=EW_Slot_1}]
    progressMaterialBrush {
     bIsDynamicallyLoaded=false
     drawAs=Image
     imageType=NoImage
     mirroring=NoMirror
     resourceName=None
     resourceObject=/FortUILibrary/UIKit/Buttons/MaterialInstances/CTAButtons/MI_UIKit_EmptyProgress
     tiling=NoTile
     imageSize=<32 32>
     margin=<ltrb 0 0 0 0>
     outlineSettings {
      bUseBrushTransparency=false
      roundingType=HalfHeightRadius
      width=0
      color {colorUseRule=UseColor_Specified, specifiedColor=<rgba 0 0 0 0>}
      cornerRadii=<0 0 0 0>
     }
     tintColor {colorUseRule=UseColor_Specified, specifiedColor=<rgba 1 1 1 1>}
     uVRegion {bIsValid=false, max=<0 0>, min=<0 0>}
    }
   }
   slot {class=HorizontalBoxSlot, properties {horizontalAlignment=HAlign_Center, verticalAlignment=VAlign_Center, padding=<ltrb 0 0 2 0>}}
  }
  ActionPromptDivider {
   widgetClassPath=&Image
   bIsVariable=true
   parent=BottomBar_Left
   widget=ActionPromptDivider
   properties {
    renderOpacity=0.2
    brush {
     bIsDynamicallyLoaded=false
     drawAs=Image
     imageType=NoImage
     mirroring=NoMirror
     resourceName=None
     resourceObject=null
     tiling=NoTile
     imageSize=<2 6>
     margin=<ltrb 0 0 0 0>
     outlineSettings {
      bUseBrushTransparency=false
      roundingType=HalfHeightRadius
      width=0
      color {colorUseRule=UseColor_Specified, specifiedColor=<rgba 0 0 0 0>}
      cornerRadii=<0 0 0 0>
     }
     tintColor {colorUseRule=UseColor_Specified, specifiedColor=<rgba 1 1 1 1>}
     uVRegion {bIsValid=false, max=<0 0>, min=<0 0>}
    }
   }
   slot {class=HorizontalBoxSlot, properties {horizontalAlignment=HAlign_Left, verticalAlignment=VAlign_Center, padding=<ltrb 0 0 14 0>}}
  }
  BottomBarLeft_QuickUse {
   widgetClassPath=&WBP_UIKit_Button_Generic
   bIsVariable=true
   parent=BottomBar_Left
   widget=BottomBarLeft_QuickUse
   properties {
    bDisplayInputActionWhenNotInteractable=true
    bIsFocusable=false
    bLocked=true
    text=Quick Use
    useBackgroundBlock=false
    useIconBlock=false
    font {
     bForceMonospaced=false
     bMaterialIsStencil=false
     fontMaterial=&MI_UIKit_TextQuiet_Prompt
     fontObject=&HeadingNow
     letterSpacing=0
     monospacedWidth=1
     size=15
     skewAmount=0
     typefaceFontName=75 Medium
     outlineSettings {
      bApplyOutlineToDropShadows=false
      bMiteredCorners=false
      bSeparateFillAlpha=false
      outlineMaterial=null
      outlineSize=0
      outlineColor=<rgba 0 0 0 1>
     }
    }
    promptNudge=<-12 0>
    triggeringInputAction {dataTable=&InputActionDataTable, rowName=EW_Slot_8}
   }
   slot {class=HorizontalBoxSlot, properties {horizontalAlignment=HAlign_Left, verticalAlignment=VAlign_Center}}
  }
  DecoyBottomButtons {
   widgetClassPath=&HorizontalBox
   parent=GridPanel_18
   widget=DecoyBottomButtons
   slot {class=GridSlot, properties {horizontalAlignment=HAlign_Right, verticalAlignment=VAlign_Bottom}}
  }
  WBP_BottomBarDecoy_PM {
   widgetClassPath=/Game/UI/Frontend/BottomBar/WBP_BottomBarDecoyButton
   bIsVariable=true
   parent=DecoyBottomButtons
   widget=WBP_BottomBarDecoy_PM
   properties {visibility=Collapsed, triggeringInputAction {dataTable=/FortUILibrary/Input/ReplayInputActionDataTable, rowName=EnterPhotoMode}}
   slot {class=HorizontalBoxSlot}
  }
  WBP_BottomBarDecoy_Use {
   widgetClassPath=&WBP_BottomBarDecoyButton
   bIsVariable=true
   parent=DecoyBottomButtons
   widget=WBP_BottomBarDecoy_Use
   properties {triggeringInputAction {dataTable=&InputActionDataTable, rowName=EWUse}}
   slot {class=HorizontalBoxSlot, properties {horizontalAlignment=HAlign_Right, verticalAlignment=VAlign_Bottom}}
  }
  WBP_BottomBarDecoy_UseRelease {
   widgetClassPath=&WBP_BottomBarDecoyButton
   bIsVariable=true
   parent=DecoyBottomButtons
   widget=WBP_BottomBarDecoy_UseRelease
   properties {pressMethod=ButtonRelease, triggeringInputAction {dataTable=&/GizmosInputActionDataTable, rowName=UseItRelease}}
   slot {class=HorizontalBoxSlot, properties {horizontalAlignment=HAlign_Right, verticalAlignment=VAlign_Bottom}}
  }
  WBP_BottomBarDecoy_UseCommsRelease {
   widgetClassPath=&WBP_BottomBarDecoyButton
   bIsVariable=true
   parent=DecoyBottomButtons
   widget=WBP_BottomBarDecoy_UseCommsRelease
   properties {pressMethod=ButtonRelease, triggeringInputAction {dataTable=&GizmosInputActionDataTable, rowName=UseItRelease}}
   slot {class=HorizontalBoxSlot, properties {horizontalAlignment=HAlign_Right, verticalAlignment=VAlign_Bottom}}
  }
  WBP_BottomBarDecoy_Close {
   widgetClassPath=&WBP_BottomBarDecoyButton
   bIsVariable=true
   parent=DecoyBottomButtons
   widget=WBP_BottomBarDecoy_Close
   properties {triggeringInputAction {dataTable=&InputActionDataTable, rowName=EW_Close}}
   slot {class=HorizontalBoxSlot}
  }
  WBP_BottomBarDecoy_CloseEsc {
   widgetClassPath=&WBP_BottomBarDecoyButton
   bIsVariable=true
   parent=DecoyBottomButtons
   widget=WBP_BottomBarDecoy_CloseEsc
   properties {bDisplayInActionBar=false, triggeringInputAction {dataTable=&InputActionDataTable, rowName=Back}}
   slot {class=HorizontalBoxSlot}
  }
  WBP_BottomBarDecoy_GamepadUse {
   widgetClassPath=&WBP_BottomBarDecoyButton
   bIsVariable=true
   parent=DecoyBottomButtons
   widget=WBP_BottomBarDecoy_GamepadUse
   properties {bDisplayInActionBar=false, triggeringInputAction {dataTable=&InputActionDataTable, rowName=EWUse}}
   slot {class=HorizontalBoxSlot}
  }
  OverlayPanel {
   widgetClassPath=&Overlay
   bIsVariable=true
   parent=GridPanel_18
   widget=OverlayPanel
   slot {class=GridSlot, properties {layer=40}}
  }
  TouchSwipe_Border {
   widgetClassPath=&CommonHardwareVisibilityBorder
   parent=GridPanel_18
   widget=TouchSwipe_Border
   properties {visibilityQuery {
     autoDescription=" ANY( Input.Touch )"
     tokenStreamVersion=0
     userDescription=""
     queryTokenStream [0,1,1,1,0]
     tagDictionary [{tagName=Input.Touch}]
    }}
   slot {class=GridSlot, properties {layer=-1}}
  }
  Overlay_0 {
   widgetClassPath=&Overlay
   parent=TouchSwipe_Border
   widget=Overlay_0
   slot {class=BorderSlot, properties {padding=<ltrb 0 0 0 0>}}
  }
  TouchSwipePanel {
   widgetClassPath=/Script/FortniteUI.FortSwipePanel
   bIsVariable=true
   parent=Overlay_0
   widget=TouchSwipePanel
   properties {bAllowSwipeFromChildTouchMove=false, bBeginSwipeOnPointerEnter=false, bConsumePointerInput=false, swipeThreshold=<20 0>}
   slot {class=OverlaySlot, properties {horizontalAlignment=HAlign_Fill, verticalAlignment=VAlign_Fill}}
  }
  BottomButtons {
   widgetClassPath=&Overlay
   parent=GridPanel_18
   widget=BottomButtons
   properties {visibility=Collapsed}
   slot {class=GridSlot, properties {horizontalAlignment=HAlign_Right, verticalAlignment=VAlign_Bottom, padding=<ltrb 0 0 128 48>}}
  }
  Image_ButtonBackground {
   widgetClassPath=&Image
   parent=BottomButtons
   widget=Image_ButtonBackground
   properties {brush {
     bIsDynamicallyLoaded=false
     drawAs=RoundedBox
     imageType=FullColor
     mirroring=NoMirror
     resourceName=None
     resourceObject=&M_UI_Clock_WhiteCirlce
     tiling=NoTile
     imageSize=<32 32>
     margin=<ltrb 0.5 0.5 0.5 0.5>
     outlineSettings {
      bUseBrushTransparency=false
      roundingType=HalfHeightRadius
      width=0
      color {colorUseRule=UseColor_Specified, specifiedColor=<rgba 0 0 0 0>}
      cornerRadii=<0 0 0 0>
     }
     tintColor {colorUseRule=UseColor_Specified, specifiedColor=<rgba 0 0 0 0.5611>}
     uVRegion {bIsValid=false, max=<0 0>, min=<0 0>}
    }}
   slot {class=OverlaySlot, properties {horizontalAlignment=HAlign_Fill, verticalAlignment=VAlign_Fill}}
  }
  HorizontalBoxButtons {
   widgetClassPath=&HorizontalBox
   parent=BottomButtons
   widget=HorizontalBoxButtons
   slot {class=OverlaySlot, properties {horizontalAlignment=HAlign_Center, verticalAlignment=VAlign_Center, padding=<ltrb 20 10 20 10>}}
  }
  HorizontalBox1 {
   widgetClassPath=&HorizontalBox
   parent=HorizontalBoxButtons
   widget=HorizontalBox1
   slot {class=HorizontalBoxSlot, properties {padding=<ltrb 0 0 30 0>}}
  }
  KeybindWidget_Equip {
   widgetClassPath=/FortUILibrary/Input/KeybindWidget
   bIsVariable=true
   parent=HorizontalBox1
   widget=KeybindWidget_Equip
   properties {brushSize=Small}
   slot {class=HorizontalBoxSlot}
  }
  CommonTextBlock_42 {
   widgetClassPath=&CommonTextBlock
   parent=HorizontalBox1
   widget=CommonTextBlock_42
   properties {
    style=&TextStyle-Base_C
    text=Equip
    font {
     bForceMonospaced=false
     bMaterialIsStencil=false
     fontMaterial=null
     fontObject=&NotoSans
     letterSpacing=0
     monospacedWidth=1
     size=24
     skewAmount=0
     typefaceFontName=NotoSans-Regular
     outlineSettings {
      bApplyOutlineToDropShadows=false
      bMiteredCorners=false
      bSeparateFillAlpha=false
      outlineMaterial=null
      outlineSize=0
      outlineColor=<rgba 0 0 0 1>
     }
    }
    shadowOffset=<0 0>
   }
   slot {class=HorizontalBoxSlot, properties {horizontalAlignment=HAlign_Center, verticalAlignment=VAlign_Center}}
  }
  HorizontalBox2 {widgetClassPath=&HorizontalBox, parent=HorizontalBoxButtons, widget=HorizontalBox2, slot {class=HorizontalBoxSlot}}
  KeybindWidget_Close {
   widgetClassPath=&KeybindWidget
   bIsVariable=true
   parent=HorizontalBox2
   widget=KeybindWidget_Close
   properties {brushSize=Small}
   slot {class=HorizontalBoxSlot}
  }
  CommonTextBlock_117 {
   widgetClassPath=&CommonTextBlock
   parent=HorizontalBox2
   widget=CommonTextBlock_117
   properties {
    style=&TextStyle-Base_C
    text=Close
    font {
     bForceMonospaced=false
     bMaterialIsStencil=false
     fontMaterial=null
     fontObject=&NotoSans
     letterSpacing=0
     monospacedWidth=1
     size=24
     skewAmount=0
     typefaceFontName=NotoSans-Regular
     outlineSettings {
      bApplyOutlineToDropShadows=false
      bMiteredCorners=false
      bSeparateFillAlpha=false
      outlineMaterial=null
      outlineSize=0
      outlineColor=<rgba 0 0 0 1>
     }
    }
    shadowOffset=<0 0>
   }
   slot {class=HorizontalBoxSlot, properties {horizontalAlignment=HAlign_Center, verticalAlignment=VAlign_Center}}
  }
 ]
}
